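# Niagara camera-lens effect for the intro Rift Portal's green-room burst, built from two active mesh emitters.
# CameraLensHeat001 fades in a blue heat layer, feeding flicker, panning time and an orange-to-blue secondary color from user parameters.
# LensEffectLoop wobbles a camera-offset mesh, visible only when the direction to User.PortalPosition opposes the owner's X axis, via a dot-product alpha curve.
# Speed-line and fade-to-black transition emitters are present but disabled.
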NiagaraSystem /SparksEffects/Effects/Intro/Rift_Portal/Particles/NS_Camera_Lens_GreenRoom_Portal_Burst
dependencies {
 usedDataInterfaces [&NiagaraDataInterfaceCurve, &NiagaraDataInterfaceCamera]
 usedDynamicInputs [&RemapRange, &VectorLength]
 usedModules [
  &self:SystemSpawnScript.SystemScriptSource.SystemScriptGraph.NiagaraNodeAssignment_3.SetVariables_#1
  &self:SystemSpawnScript.SystemScriptSource.SystemScriptGraph.NiagaraNodeAssignment_5.SetVariables_#2
  &SystemState
  &self:SystemSpawnScript.SystemScriptSource.SystemScriptGraph.NiagaraNodeAssignment_4.SetVariables_#3
  &self:SystemSpawnScript.SystemScriptSource.SystemScriptGraph.NiagaraNodeAssignment_7.SetVariables_#4
  &self:SystemSpawnScript.SystemScriptSource.SystemScriptGraph.NiagaraNodeAssignment_0.SetVariables_#5
  &self:SystemSpawnScript.SystemScriptSource.SystemScriptGraph.NiagaraNodeAssignment_2.SetVariables_#6
  &self:SystemSpawnScript.SystemScriptSource.SystemScriptGraph.NiagaraNodeAssignment_1.SetVariables_#7
 ]
}
emitters {
 CameraLensHeat001 {
  emitterSpawnScript [/SetVariables(Emitter.CameraLensHeatOpacity=1) {isSetParameters=true}]
  emitterUpdateScript [
   &EmitterState(Inactive Response=::NewEnumerator1, Life Cycle Mode=::NewEnumerator1, Loop Behavior=::NewEnumerator1, Loop Duration=10, UseLoopDelay=false)
   &SpawnBurst_Instantaneous(Use Spawn Probability=false, UseLoopCountLimit=false)
   /SetVariables(Emitter.CameraLensHeatOpacity:NiagaraFloat=&FloatFromCurve(
     FloatCurve=DataInterface() {propertyValues {Curve {keys [{tangentMode=RCTM_User, time=0, value=0}, {tangentMode=RCTM_User, time=2, value=1}]}}}
     CurveIndex=$Emitter.Age
    )) {isSetParameters=true}
  ]
  particleSpawnScript [
   &InitializeParticle(
    Color=<rgba 0.2867 0.2555 1.5 0.1>
    Lifetime=3.35
    Mesh Scale=<2 6 6>
    Mesh Scale Mode=::NewEnumerator4
    Position Offset=<700 0 0>
    UsePositionOffset=true
    Write Material Random=true
   )
   /SetVariables(Particles.Position2:NiagaraPosition=<-50 0 0>) {isSetParameters=true}
   /SetVariables() {isSetParameters=true}
   &InitialMeshOrientation(Mesh Orientation Mode=::NewEnumerator2, Orientation Vector=<-1 0 0>, Use Rotation Vector=true)
  ]
  particleUpdateScript [
   &ParticleState(Kill Particles When Lifetime Has Elapsed=false, Let Infinitely Lived Particles Die When Emitter Deactivates=true)
   &ScaleColor(Scale Alpha=$User.HeatOpacity, ScaleA=true, ScaleRGB=false)
   /SetVariables(Particles.TimeAdd:NiagaraFloat=&Add_Float(A=$Particles.TimeAdd, B=&Multiply_Float(A=$Engine.DeltaTime, B=$User.PanSpeed))) {isSetParameters=true}
   /SetVariables(Particles.Flicker:NiagaraFloat=&Waveform("Waveform [1]"=::NewEnumerator10, "[1] Frequency"=15, "[1] Amplitude Min/Max"=<0 1>)) {isSetParameters=true}
   &DynamicMaterialParameters(
    Index 0 Param 2=2
    Index 0 Param 3=0.5
    Index 0 Param 4=$Particles.TimeAdd
    Index 1 Param 2=&Lerp_Float(A=0.7, B=0.25, Alpha=$Particles.Flicker)
    Index 1 Param 3=1
    Index 1 Param 4=$Particles.TimeAdd
    Module.Param0WriteEnabled=true
    Module.Param0WriteEnabled1=true
    Module.Param1WriteEnabled=true
    Module.Param1WriteEnabled1=true
    Module.Param2WriteEnabled=true
    Module.Param2WriteEnabled1=true
    Module.Param3WriteEnabled=true
    Module.Param3WriteEnabled1=true
    Write Parameter Index 1=true
   )
   /SetVariables(Particles.Color2=&LerpLinearColors(
     StartColor=&MultiplyLinearColor(
      Linear Color=<rgba 5 0.517 0 0.25>
      Scale Factor=&MakeFloat4FromVectorAndFloat(
       "Vector (RGB)"=&VectorFromFloat(Value=&Lerp_Float(A=0.25, B=1.5, Alpha=$Particles.Flicker))
       "Float (Alpha)"=&FloatFromCurve(
        FloatCurve=DataInterface() {propertyValues {Curve {keys [
            {interpMode=RCIM_Cubic, tangentMode=RCTM_User, time=0.2, value=0}
            {arriveTangent=1.6796, interpMode=RCIM_Cubic, leaveTangent=1.6796, tangentMode=RCTM_User, time=0.7, value=0.5}
            {interpMode=RCIM_Cubic, tangentMode=RCTM_User, time=1, value=1}
           ]}}}
        CurveIndex=$User.HeatOpacity
       )
      )
     )
     EndColor=&MultiplyLinearColor(
      Linear Color=<rgba 0 6.6834 20 0.25>
      Scale Factor=&MakeFloat4FromVectorAndFloat(
       "Vector (RGB)"=&VectorFromFloat(Value=&Lerp_Float(A=0.25, B=1.5, Alpha=$Particles.Flicker))
       "Float (Alpha)"=&FloatFromCurve(
        FloatCurve=DataInterface() {propertyValues {Curve {keys [
            {interpMode=RCIM_Cubic, tangentMode=RCTM_User, time=0.2, value=0}
            {arriveTangent=1.6796, interpMode=RCIM_Cubic, leaveTangent=1.6796, tangentMode=RCTM_User, time=0.7, value=0.5}
            {interpMode=RCIM_Cubic, tangentMode=RCTM_User, time=1, value=1}
           ]}}}
        CurveIndex=$User.HeatOpacity
       )
      )
     )
     LerpFactor=$User.HeatScale
    )) {isSetParameters=true}
   /SetVariables(Particles.Scale=&self:StretchtoViewportWidth_01(Default:Vector3f=$Particles.Initial.Scale, New Camera Query:NiagaraDataInterfaceCamera=DataInterface())) {enabled=false, isSetParameters=true, module=SetVariables_#8}
   /SetVariables(Particles.Scale=&Multiply_Vector(A=$Particles.Initial.Scale, B=&MakeVector(X=1, Y=$User.ScreenSizScalar, Z=1))) {isSetParameters=true}
   /SetVariables(Particles.Scale2:Vector3f=&Multiply_Vector(A=$Particles.Scale, B=&VectorFromFloat(Value=0.2))) {isSetParameters=true}
   /SetVariables(Particles.Color=&MultiplyLinearColor(
     Linear Color=<rgba 0 1.6708 5 0.1>
     Scale Factor=&MakeVector4_FromFloat(Float=$Emitter.CameraLensHeatOpacity)
    )) {isSetParameters=true}
   &ScaleColor(
    Scale Alpha=&FloatFromCurve(
     FloatCurve=DataInterface() {propertyValues {Curve {keys [
         <at 0.35 1>
         {arriveTangent=-1.5385, interpMode=RCIM_Cubic, leaveTangent=-1.5385, time=0.6, value=1}
         {interpMode=RCIM_Cubic, tangentMode=RCTM_User, time=1, value=0}
        ]}}}
     CurveIndex=&NormalizeDistanceRange(
      Start Position:NiagaraPosition=&TransformPosition(Position=$Engine.Emitter.SimulationPosition, Source Space=::Simulation, Destination Space=::World)
      End Position:NiagaraPosition=$User.PortalPosition
      Distance:NiagaraFloat=1500
      Invert Normalized Range:NiagaraBool=false
     )
     Scale Curve=5
    )
    ScaleA=true
    ScaleRGB=true
   ) {enabled=false, module=ScaleColor001}
  ]
  rendererClasses [&NiagaraMeshRendererProperties]
  renderers [{rendererClass=&NiagaraMeshRendererProperties, rendererIndex=0}]
 }
 LensEffectLoop {
  emitterUpdateScript [
   &EmitterState(Inactive Response=::NewEnumerator1, Life Cycle Mode=::NewEnumerator1, Loop Behavior=::NewEnumerator1, Loop Duration=10, UseLoopDelay=false)
   &SpawnBurst_Instantaneous(Use Spawn Probability=false, UseLoopCountLimit=false)
  ]
  particleSpawnScript [
   /SetVariables(Particles.MeshOrientation=&MakeQuatFromAxisAngle(Axis=<0 0 1>, AngleInDegrees=90)) {isSetParameters=true}
   /Niagara/Modules/Spawn/Initialization/InitializeParticle(
    Lifetime=1.35
    Mesh Scale=<2 3 2>
    Module.Write Color=true
    Module.Write Lifetime=true
    Module.Write Mass=true
    Module.Write Material Random=true
    Module.Write Position=true
    Module.Write Scale=true
    Module.Write SpriteRotation=false
    Module.Write SpriteSize=false
    Position=<-100 0 0>
   ) {enabled=false}
   /Niagara/Modules/Spawn/Initialization/V2/InitializeParticle(
    Color=<rgba 0.0108 0.4659 2 0.6>
    Lifetime=3.35
    Mass Mode=::NewEnumerator1
    Mesh Scale=<2 3 2>
    Mesh Scale Mode=::NewEnumerator4
    Position Offset=<-250 0 0>
    Sprite Size Mode=::NewEnumerator4
    UsePositionOffset=false
    Write Material Random=true
   ) {module=InitializeParticle001}
   &Color(Color=&MakeLinearColorFromVectorAndFloat("Vector (RGB)"=<1 1 1>, "Float (Alpha)"=&ClampFloat(Float=1))) {enabled=false}
  ]
  particleUpdateScript [
   &ParticleState(Kill Particles When Lifetime Has Elapsed=false, Let Infinitely Lived Particles Die When Emitter Deactivates=true)
   &Color(Color=&MakeLinearColorFromVectorAndFloat(
     "Vector (RGB)"=<1 1 1>
     "Float (Alpha)"=&ClampFloat(Float=&FloatFromCurve(FloatCurve=DataInterface() {propertyValues {Curve {keys [<at 0 0 RCIM_Cubic>, <at 0.1086 1>, {arriveTangent=-1.1218, interpMode=RCIM_Cubic, leaveTangent=-1.1218, time=0.8251, value=1}, <at 1 0>]}}}))
    )) {enabled=false, module=Color001}
   &DynamicMaterialParameters(
    Index 0 Param 2=1
    Index 0 Param 3=1
    Index 0 Param 4=1
    Module.Param0WriteEnabled=true
    Module.Param1WriteEnabled=true
    Module.Param2WriteEnabled=true
    Module.Param3WriteEnabled=true
   ) {enabled=false}
   &UpdateMeshOrientation(
    Rotation Rate=0.15
    Rotation Vector=&MakeVector(X=0, Y=&Waveform("Waveform [1]"=::NewEnumerator12, "[1] Frequency"=1, "[1] Amplitude Min/Max"=<-1 1>), Z=0)
   )
   &ScaleColor(
    Scale Alpha=&FloatFromCurve(FloatCurve=DataInterface() {propertyValues {Curve {keys [
         {arriveTangent=0.3357, interpMode=RCIM_Cubic, leaveTangent=0.3357, tangentMode=RCTM_User, time=0.05, value=0}
         {arriveTangent=2.4269, interpMode=RCIM_Cubic, leaveTangent=2.4269, tangentMode=RCTM_User, time=0.15, value=0.5}
         {arriveTangent=1.7192, tangentMode=RCTM_User, time=1, value=1}
        ]}}})
    ScaleA=true
    ScaleRGB=true
   )
   &DynamicMaterialParameters(
    Index 2 Param 1=&FloatFromCurve(FloatCurve=DataInterface() {propertyValues {Curve {keys [
         {arriveTangent=0.2659, interpMode=RCIM_Cubic, leaveTangent=0.2659, tangentMode=RCTM_User, time=0, value=0}
         {tangentMode=RCTM_User, time=1, value=1}
        ]}}})
    Index 2 Param 2=&FloatFromCurve(FloatCurve=DataInterface() {propertyValues {Curve {keys [
         {arriveTangent=0.067, interpMode=RCIM_Cubic, leaveTangent=0.067, tangentMode=RCTM_User, time=0.5, value=0}
         {interpMode=RCIM_Cubic, tangentMode=RCTM_User, time=1, value=1}
        ]}}})
    Module.Index2Param0Write=true
    Module.Index2Param1Write=false
    Module.Index2Param2Write=true
    Module.Index2Param3Write=true
    Write Parameter Index 0=false
    Write Parameter Index 2=true
   ) {module=DynamicMaterialParameters001}
   &ScaleColor(
    Scale Alpha=&FloatFromCurve(
     FloatCurve=DataInterface() {propertyValues {Curve {keys [<at -1 1 RCIM_Cubic>, <at -0.5 0>, {interpMode=RCIM_Cubic, tangentMode=RCTM_User, time=1, value=0}]}}}
     CurveIndex=&DotProduct(
      Vector A=&NormalizeVector(
       Vector To Normalize=&SubtractVectorFromPosition(
        Position=$User.PortalPosition
        Vector=&TransformVector(
         Vector=Linked() {linkedVariable {name=Engine.Emitter.SimulationPosition, type=NiagaraPosition}}
         Source Space=::Simulation
         Destination Space=::World
         Transform Vector as Position=true
        )
       )
       Safe Normalize=true
       Safe Fallback Threshold=0
       Safe Fallback Vector=<0 0 0.001>
      )
      Vector B=&NormalizeVector(Vector To Normalize=$Engine.Owner.SystemXAxis, Safe Normalize=true, Safe Fallback Threshold=0, Safe Fallback Vector=<0 0 0.001>)
     )
    )
    ScaleA=true
    ScaleRGB=false
   ) {module=ScaleColor002}
   &CameraOffset(Camera Offset Amount=-400)
  ]
  rendererClasses [&NiagaraMeshRendererProperties]
  renderers [{rendererClass=&NiagaraMeshRendererProperties, rendererIndex=0}]
 }
 SpeedLinesBurst {
  bEnabled=false
  emitterUpdateScript [
   &EmitterState(Life Cycle Mode=::NewEnumerator1, Loop Behavior=::NewEnumerator1, Loop Duration=1, UseLoopDelay=false)
   &SpawnBurst_Instantaneous(Spawn Count=250, Use Spawn Probability=false, UseLoopCountLimit=false)
   &SpawnRate(
    Spawn Probability=&FloatFromCurve(
     FloatCurve=DataInterface() {propertyValues {Curve {keys [
         {arriveTangent=0.2238, interpMode=RCIM_Cubic, leaveTangent=0.2238, tangentMode=RCTM_User, time=0, value=1}
         {arriveTangent=-5.834, interpMode=RCIM_Cubic, leaveTangent=-5.834, tangentMode=RCTM_User, time=0.5, value=0}
        ]}}}
     CurveIndex=$Emitter.NormalizedLoopAge
    )
    SpawnRate=1000
    Use Spawn Probability=true
   )
  ]
  particleSpawnScript [
   /SetVariables(Particles.Intensity:NiagaraFloat=&FloatFromCurve(
     FloatCurve=DataInterface() {propertyValues {Curve {keys [
         {interpMode=RCIM_Cubic, tangentMode=RCTM_User, time=0, value=0}
         {arriveTangent=1.6995, interpMode=RCIM_Cubic, leaveTangent=1.6995, tangentMode=RCTM_User, time=0.5215, value=0.3748}
         {interpMode=RCIM_Cubic, tangentMode=RCTM_User, time=0.9, value=1}
        ]}}}
     CurveIndex=&RandomRangeFloat(Minimum=0, Maximum=1)
    )) {isSetParameters=true}
   &InitializeParticle(
    Color=&LerpLinearColors(StartColor=<rgba 0 0.8651 1 0.6>, EndColor=<rgba 0.2188 0.5108 1 0.8>, LerpFactor=$Particles.Intensity)
    Lifetime=&Multiply_Float(
     A=&RandomRangeFloat(Minimum=0.6, Maximum=1.3)
     B=&Lerp_Float(A=1, B=0.35, Alpha=$System.NormalizedSpeedScale)
    )
    Sprite Size=&Multiply_Vector2D_ByFloat(Float=0.25, Vector2D=&Lerp_Vector2(A=<70 60>, B=<150 75>, Alpha=$Particles.Intensity))
    Sprite Size Mode=::NewEnumerator4
    UsePositionOffset=false
    Write Material Random=true
   )
   &ShapeLocation(
    Cylinder Height=250
    Cylinder Radius=150
    MaskSpawnGroup=false
    Offset=&Multiply_VectorByFloat(Vector=$System.TravelSpeed, Float=1)
    Offset Mode=::NewEnumerator0
    Override Seed=false
    Rotation Axis=$System.TravelSpeed
    Rotation Coordinate Space=::Simulation
    Rotation Mode=::NewEnumerator1
    Shape Primitive=::NewEnumerator1
    Surface Only=true
    Use Endcaps In Surface Only Mode=false
    UseOwnerScale=true
   )
   &AddVelocity(Velocity=$System.TravelSpeed, Velocity Speed Scale=5)
  ]
  particleUpdateScript [
   &ParticleState()
   &ScaleColor(
    Scale Alpha=&FloatFromCurve(FloatCurve=DataInterface() {propertyValues {Curve {keys [
         {interpMode=RCIM_Cubic, tangentMode=RCTM_User, time=0, value=0}
         <at 0.25 1>
         {interpMode=RCIM_Cubic, tangentMode=RCTM_User, time=0.75, value=1}
         {tangentMode=RCTM_User, time=1, value=0}
        ]}}})
    Scale RGB=&VectorFromFloat(Value=&FloatFromCurve(FloatCurve=DataInterface() {propertyValues {Curve {keys [<at 0.25 0.3 RCIM_Cubic>,<at 1 2>]}}}, CurveIndex=$System.NormalizedSpeedScale))
    ScaleA=true
    ScaleRGB=true
   )
   &DynamicMaterialParameters(
    Index 0 Param 1=&Lerp_Float(
     A=5
     B=2
     Alpha=&FloatFromCurve(FloatCurve=DataInterface() {propertyValues {Curve {keys [
          {interpMode=RCIM_Cubic, tangentMode=RCTM_User, time=0, value=0}
          {interpMode=RCIM_Cubic, tangentMode=RCTM_User, time=0.7, value=1}
          {tangentMode=RCTM_User, time=1, value=0}
         ]}}})
    )
    Index 0 Param 2=500
    Index 0 Param 3=1
    Module.Param0WriteEnabled=true
    Module.Param1WriteEnabled=true
    Module.Param2WriteEnabled=true
    Module.Param3WriteEnabled=false
   )
   &Drag(Drag=4, Use Linear Drag=true, Use Rotational Drag=false)
   &WindForce(Air Resistance=0.1, Wind Speed=$System.TravelSpeed, Wind Speed Scale=-20)
   &SolveForcesAndVelocity(Clamp Velocity=false, Limit Acceleration=false)
   &ScaleSpriteSizeBySpeed(
    Max Scale Factor=<5 180>
    Min Scale Factor=<5 0.7>
    Sample Scale Factor By Curve=true
    Scale Factor Curve=DataInterface() {propertyValues {Curve {keys [<at 0 0>, {arriveTangent=1, interpMode=RCIM_Cubic, leaveTangent=1, time=0.6, value=0.4}, <at 1 1>]}}}
    Velocity Threshold=14000
   )
  ]
  rendererClasses [&NiagaraSpriteRendererProperties]
  renderers [{rendererClass=&NiagaraSpriteRendererProperties}]
 }
 WhiteOrBlackTransition_CamParticle {
  bEnabled=false
  emitterUpdateScript [
   &EmitterState(
    Inactive Response=::NewEnumerator1
    Life Cycle Mode=::NewEnumerator1
    Loop Behavior=::NewEnumerator1
    Loop Duration Mode=::NewEnumerator1
    UseLoopDelay=false
   )
   &SpawnBurst_Instantaneous(Spawn Count=1, Use Spawn Probability=false)
  ]
  particleSpawnScript [
   &CameraOffset(Camera Offset Amount=-30) {enabled=false}
   &InitializeParticle(
    Color=<rgba 0 0 0 1>
    Lifetime=3.35
    Mass Mode=::NewEnumerator1
    Mesh Scale=<2 3 2>
    Mesh Scale Mode=::NewEnumerator4
    Position Offset=<-250 0 0>
    Sprite Size Mode=::NewEnumerator4
    UsePositionOffset=false
    Write Material Random=true
   ) {module=InitializeParticle001}
  ]
  particleUpdateScript [
   &ParticleState(Kill Particles When Lifetime Has Elapsed=false, Let Infinitely Lived Particles Die When Emitter Deactivates=true)
   &SolveForcesAndVelocity(Clamp Velocity=false, Limit Acceleration=false)
   &DynamicMaterialParameters(
    Index 0 Param 1=1
    Index 0 Param 2=&FloatFromCurve(FloatCurve=DataInterface() {propertyValues {Curve {keys [
         {arriveTangent=0.067, interpMode=RCIM_Cubic, leaveTangent=0.067, tangentMode=RCTM_User, time=1, value=0}
         {interpMode=RCIM_Cubic, tangentMode=RCTM_User, time=1.5, value=1}
        ]}}})
    Module.Param0WriteEnabled=true
    Module.Param1WriteEnabled=true
    Module.Param2WriteEnabled=true
    Module.Param3WriteEnabled=true
   )
   &ScaleColor(
    Scale RGB=&VectorFromFloat(Value=&FloatFromCurve(FloatCurve=DataInterface() {propertyValues {Curve {keys [<at 1 1 RCIM_Cubic>,<at 1.35 0>]}}}, CurveIndex=$Engine.System.Age))
    ScaleA=true
    ScaleRGB=true
   )
  ]
  rendererClasses [&NiagaraMeshRendererProperties]
  renderers [
   {rendererClass=&NiagaraMeshRendererProperties, rendererIndex=0}
   {rendererClass=&NiagaraMeshRendererProperties, rendererIndex=1}
  ]
 }
}
summary {
 systemName=NS_Camera_Lens_GreenRoom_Portal_Burst
 emitters [
  {bEnabled=false, emitterName=SpeedLinesBurst, rendererClasses [&NiagaraSpriteRendererProperties]}
  {emitterName=LensEffectLoop, rendererClasses [&NiagaraMeshRendererProperties]}
  {bEnabled=false, emitterName=WhiteOrBlackTransition_CamParticle, rendererClasses [&NiagaraMeshRendererProperties]}
  {emitterName=CameraLensHeat001, rendererClasses [&NiagaraMeshRendererProperties]}
 ]
 userVariables [
  User.HeatOpacity {type=NiagaraFloat, defaultValue=0.4, description=""}
  User.HeatScale {type=NiagaraFloat, defaultValue=5, description=""}
  User.PanSpeed {type=NiagaraFloat, defaultValue=5, description=""}
  User.PortalPosition {type=NiagaraPosition, description="", defaultValue {}}
  User.ScreenSizScalar {type=NiagaraFloat, defaultValue=1, description=""}
  User.SpawnRocks {type=NiagaraFloat, defaultValue=0.4, description=""}
  User.SpeedlineLength {type=NiagaraFloat, defaultValue=1, description=""}
  User.SpeedLineMulti {type=NiagaraFloat, defaultValue=1, description=""}
 ]
}
