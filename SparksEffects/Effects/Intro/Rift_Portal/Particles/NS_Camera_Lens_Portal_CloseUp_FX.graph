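# Camera lens Niagara effect for close-ups of the intro rift portal, layering heat distortion and cyan glow.
# A lens mesh pans and flickers its material and ramps in over emitter age, beside a glow sprite and a screen glow mesh.
# All three fade with distance from User.PortalPosition, and both glows also fade by a dot product between portal direction and the system X axis.
# Shard and triangle burst emitters are present but disabled.
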NiagaraSystem /SparksEffects/Effects/Intro/Rift_Portal/Particles/NS_Camera_Lens_Portal_CloseUp_FX
dependencies {
 usedDataInterfaces [
  &NiagaraDataInterfaceCurve
  &NiagaraDataInterfaceArrayInt32
  &NiagaraDataInterfaceArrayFloat
  &NiagaraDataInterfaceCamera
 ]
 usedDynamicInputs [&SelectFloatFromArray]
 usedModules [&SystemState]
}
emitters {
 CameraLensHeat001 {
  emitterSpawnScript [/SetVariables(Emitter.CameraLensHeatOpacity=1) {isSetParameters=true}]
  emitterUpdateScript [
   &EmitterState(
    Inactive Response=::NewEnumerator1
    Life Cycle Mode=::NewEnumerator1
    Loop Behavior=::NewEnumerator1
    Loop Duration Mode=::NewEnumerator1
    UseLoopDelay=false
   )
   &SpawnBurst_Instantaneous(Use Spawn Probability=false, UseLoopCountLimit=false)
   /SetVariables(Emitter.CameraLensHeatOpacity:NiagaraFloat=&FloatFromCurve(
     FloatCurve=DataInterface() {propertyValues {Curve {keys [{tangentMode=RCTM_User, time=0, value=0}, {tangentMode=RCTM_User, time=2, value=1}]}}}
     CurveIndex=$Emitter.Age
    )) {isSetParameters=true}
  ]
  particleSpawnScript [
   &InitializeParticle(
    Color=<rgba 0.2867 0.2555 1.5 0.1>
    Lifetime=5
    Mesh Scale=<2 6 6>
    Mesh Scale Mode=::NewEnumerator4
    Position Offset=<700 0 0>
    UsePositionOffset=true
    Write Material Random=true
   )
   /SetVariables(Particles.Position2:NiagaraPosition=<-50 0 0>) {isSetParameters=true}
   /SetVariables() {isSetParameters=true}
   &InitialMeshOrientation(Mesh Orientation Mode=::NewEnumerator2, Orientation Vector=<-1 0 0>, Use Rotation Vector=true)
  ]
  particleUpdateScript [
   &ParticleState(Kill Particles When Lifetime Has Elapsed=false, Let Infinitely Lived Particles Die When Emitter Deactivates=true)
   &ScaleColor(Scale Alpha=$User.HeatOpacity, ScaleA=true, ScaleRGB=false)
   /SetVariables(Particles.TimeAdd:NiagaraFloat=&Add_Float(A=$Particles.TimeAdd, B=&Multiply_Float(A=$Engine.DeltaTime, B=$User.PanSpeed))) {isSetParameters=true}
   /SetVariables(Particles.Flicker:NiagaraFloat=&Waveform("Waveform [1]"=::NewEnumerator10, "[1] Frequency"=15, "[1] Amplitude Min/Max"=<0 1>)) {isSetParameters=true}
   &DynamicMaterialParameters(
    Index 0 Param 2=2
    Index 0 Param 3=0.5
    Index 0 Param 4=$Particles.TimeAdd
    Index 1 Param 2=&Lerp_Float(A=0.7, B=0.25, Alpha=$Particles.Flicker)
    Index 1 Param 3=1
    Index 1 Param 4=$Particles.TimeAdd
    Module.Param0WriteEnabled=true
    Module.Param0WriteEnabled1=true
    Module.Param1WriteEnabled=true
    Module.Param1WriteEnabled1=true
    Module.Param2WriteEnabled=true
    Module.Param2WriteEnabled1=true
    Module.Param3WriteEnabled=true
    Module.Param3WriteEnabled1=true
    Write Parameter Index 1=true
   )
   /SetVariables(Particles.Color2=&LerpLinearColors(
     StartColor=&MultiplyLinearColor(
      Linear Color=<rgba 5 0.517 0 0.25>
      Scale Factor=&MakeFloat4FromVectorAndFloat(
       "Vector (RGB)"=&VectorFromFloat(Value=&Lerp_Float(A=0.25, B=1.5, Alpha=$Particles.Flicker))
       "Float (Alpha)"=&FloatFromCurve(
        FloatCurve=DataInterface() {propertyValues {Curve {keys [
            {interpMode=RCIM_Cubic, tangentMode=RCTM_User, time=0.2, value=0}
            {arriveTangent=1.6796, interpMode=RCIM_Cubic, leaveTangent=1.6796, tangentMode=RCTM_User, time=0.7, value=0.5}
            {interpMode=RCIM_Cubic, tangentMode=RCTM_User, time=1, value=1}
           ]}}}
        CurveIndex=$User.HeatOpacity
       )
      )
     )
     EndColor=&MultiplyLinearColor(
      Linear Color=<rgba 0 6.6834 20 0.25>
      Scale Factor=&MakeFloat4FromVectorAndFloat(
       "Vector (RGB)"=&VectorFromFloat(Value=&Lerp_Float(A=0.25, B=1.5, Alpha=$Particles.Flicker))
       "Float (Alpha)"=&FloatFromCurve(
        FloatCurve=DataInterface() {propertyValues {Curve {keys [
            {interpMode=RCIM_Cubic, tangentMode=RCTM_User, time=0.2, value=0}
            {arriveTangent=1.6796, interpMode=RCIM_Cubic, leaveTangent=1.6796, tangentMode=RCTM_User, time=0.7, value=0.5}
            {interpMode=RCIM_Cubic, tangentMode=RCTM_User, time=1, value=1}
           ]}}}
        CurveIndex=$User.HeatOpacity
       )
      )
     )
     LerpFactor=$User.HeatScale
    )) {isSetParameters=true}
   /SetVariables(Particles.Scale=&self:StretchtoViewportWidth(Default:Vector3f=$Particles.Initial.Scale, New Camera Query:NiagaraDataInterfaceCamera=DataInterface())) {enabled=false, isSetParameters=true, module=SetVariables_#1}
   /SetVariables(Particles.Scale=&Multiply_Vector(A=$Particles.Initial.Scale, B=&MakeVector(X=1, Y=$User.ScreenSizScalar, Z=1))) {isSetParameters=true}
   /SetVariables(Particles.Scale2:Vector3f=&Multiply_Vector(A=$Particles.Scale, B=&VectorFromFloat(Value=0.2))) {isSetParameters=true}
   /SetVariables(Particles.Color=&MultiplyLinearColor(
     Linear Color=<rgba 0 1.6708 5 0.1>
     Scale Factor=&MakeVector4_FromFloat(Float=$Emitter.CameraLensHeatOpacity)
    )) {isSetParameters=true}
   &ScaleColor(
    Scale Alpha=&FloatFromCurve(
     FloatCurve=DataInterface() {propertyValues {Curve {keys [
         <at 0.35 1>
         {arriveTangent=-1.5385, interpMode=RCIM_Cubic, leaveTangent=-1.5385, time=0.6, value=1}
         {interpMode=RCIM_Cubic, tangentMode=RCTM_User, time=1, value=0}
        ]}}}
     CurveIndex=&NormalizeDistanceRange(
      Start Position:NiagaraPosition=&TransformPosition(Position=$Engine.Emitter.SimulationPosition, Source Space=::Simulation, Destination Space=::World)
      End Position:NiagaraPosition=$User.PortalPosition
      Distance:NiagaraFloat=1500
      Invert Normalized Range:NiagaraBool=false
     )
     Scale Curve=5
    )
    ScaleA=true
    ScaleRGB=true
   ) {module=ScaleColor001}
  ]
  rendererClasses [&NiagaraMeshRendererProperties]
  renderers [{rendererClass=&NiagaraMeshRendererProperties, rendererIndex=0}]
 }
 LargeGlowBurst {
  emitterUpdateScript [
   &EmitterState(
    Inactive Response=::NewEnumerator1
    Life Cycle Mode=::NewEnumerator1
    Loop Behavior=::NewEnumerator1
    Loop Duration Mode=::NewEnumerator1
    UseLoopDelay=false
   )
   &SpawnBurst_Instantaneous(Use Spawn Probability=false, UseLoopCountLimit=true)
  ]
  particleSpawnScript [
   &InitializeParticle(
    Color=<rgba 0 0.8562 1 1>
    Lifetime=0.5
    Mesh Scale Mode=::NewEnumerator3
    Mesh Uniform Scale=30
    Position Offset=<0 0 35>
    Sprite Size Mode=::NewEnumerator3
    Uniform Sprite Size=500
    UsePositionOffset=false
    Write Material Random=true
   ) {module=InitializeParticle001}
   &InitialMeshOrientation(Orientation Axis=<0 0 1>, Use Orientation Vector=true, Use Rotation Vector=false) {enabled=false, module=InitialMeshOrientation001}
   &CameraOffset(Camera Offset Amount=50)
   &Color(Color=&MultiplyLinearColor(Linear Color=$Particles.Color, Scale Factor=<1 1 1 0.25>))
  ]
  particleUpdateScript [
   &ParticleState(
    Kill Particles When Lifetime Has Elapsed=false
    Let Infinitely Lived Particles Die When Emitter Deactivates=true
    Loop Particles Lifetime=true
   )
   &ScaleSpriteSize(
    Uniform Curve Index=&DotProduct(
     Vector A=&SubtractVectorFromPosition(
      Position=$User.PortalPosition
      Vector=&TransformVector(
       Vector=Linked() {linkedVariable {name=Engine.Emitter.SimulationPosition, type=NiagaraPosition}}
       Source Space=::Simulation
       Destination Space=::World
       Transform Vector as Position=true
      )
     )
     Vector B=Linked() {linkedVariable {name=Engine.Owner.SystemXAxis, type=NiagaraPosition}}
    )
    Uniform Curve Sprite Scale=DataInterface() {propertyValues {Curve {keys [<at -1 1 RCIM_Cubic>, <at 0.8 0>, {interpMode=RCIM_Cubic, tangentMode=RCTM_User, time=1, value=0}]}}}
   ) {enabled=false, module=ScaleSpriteSize001}
   &ScaleColor(Scale RGB=<5 5 5>, ScaleA=true, ScaleRGB=true) {module=ScaleColor001}
   /SetVariables(Particles.MeshDynParams:Vector4f=&MakeVector4(
     X=&FloatFromCurve(
      FloatCurve=DataInterface() {propertyValues {Curve {keys [
          {arriveTangent=1.7005, interpMode=RCIM_Cubic, leaveTangent=1.7005, tangentMode=RCTM_User, time=0, value=0.4}
          {arriveTangent=0.0334, interpMode=RCIM_Cubic, leaveTangent=0.0334, tangentMode=RCTM_User, time=1, value=1}
         ]}}}
      CurveIndex=$User.DistanceBlend
     )
     Y=0
     Z=0
     W=0
    )) {isSetParameters=true}
   &ScaleSpriteSize(Scale Factor=&Vector2DFromFloat(Value=&FloatFromCurve(
      FloatCurve=DataInterface() {propertyValues {Curve {keys [
          {arriveTangent=1.2491, interpMode=RCIM_Cubic, leaveTangent=1.2491, tangentMode=RCTM_User, time=0, value=0.1}
          {arriveTangent=0.3527, interpMode=RCIM_Cubic, leaveTangent=0.3527, tangentMode=RCTM_User, time=1, value=1}
         ]}}}
      CurveIndex=$User.DistanceBlend
      Scale Curve=0.5
     )))
   &DynamicMaterialParameters(
    Index 0 Param 1=1
    Index 0 Param 2=&FloatFromCurve(FloatCurve=DataInterface() {propertyValues {Curve {keys [
         {arriveTangent=-1.3736, time=0, value=1}
         {arriveTangent=-1.3273, interpMode=RCIM_Cubic, leaveTangent=-1.3273, tangentMode=RCTM_User, time=1, value=0.1}
        ]}}})
    Module.Param0WriteEnabled=true
    Module.Param1WriteEnabled=true
    Module.Param2WriteEnabled=false
    Module.Param3WriteEnabled=false
   )
   &ScaleColor(
    Scale Alpha=&FloatFromCurve(
     FloatCurve=DataInterface() {propertyValues {Curve {keys [{interpMode=RCIM_Cubic, tangentMode=RCTM_User, time=0, value=1}, {interpMode=RCIM_Cubic, tangentMode=RCTM_User, time=1, value=0}]}}}
     CurveIndex=&NormalizeDistanceRange(
      Start Position:NiagaraPosition=&TransformPosition(Position=$Engine.Emitter.SimulationPosition, Source Space=::Simulation, Destination Space=::World)
      End Position:NiagaraPosition=$User.PortalPosition
      Distance:NiagaraFloat=1500
      Invert Normalized Range:NiagaraBool=false
     )
    )
    ScaleA=true
    ScaleRGB=true
   )
   &ScaleColor(
    Scale Alpha=&FloatFromCurve(
     FloatCurve=DataInterface() {propertyValues {Curve {keys [<at -1 1 RCIM_Cubic>, <at -0.5 0>, {interpMode=RCIM_Cubic, tangentMode=RCTM_User, time=1, value=0}]}}}
     CurveIndex=&DotProduct(
      Vector A=&NormalizeVector(
       Vector To Normalize=&SubtractVectorFromPosition(
        Position=$User.PortalPosition
        Vector=&TransformVector(
         Vector=Linked() {linkedVariable {name=Engine.Emitter.SimulationPosition, type=NiagaraPosition}}
         Source Space=::Simulation
         Destination Space=::World
         Transform Vector as Position=true
        )
       )
       Safe Normalize=true
       Safe Fallback Threshold=0
       Safe Fallback Vector=<0 0 0.001>
      )
      Vector B=&NormalizeVector(
       Vector To Normalize=Linked() {linkedVariable {name=Engine.Owner.SystemXAxis, type=NiagaraPosition}}
       Safe Normalize=true
       Safe Fallback Threshold=0
       Safe Fallback Vector=<0 0 0.001>
      )
     )
    )
    ScaleA=true
    ScaleRGB=false
   ) {module=ScaleColor002}
  ]
  rendererClasses [&NiagaraSpriteRendererProperties]
  renderers [{rendererClass=&NiagaraSpriteRendererProperties}]
 }
 RiftShards {
  bEnabled=false
  emitterUpdateScript [
   &EmitterState(Life Cycle Mode=::NewEnumerator1, Loop Behavior=::NewEnumerator1, Loop Delay=$System.CrackDelay, UseLoopDelay=true)
   /SetVariables(Emitter.ShardSpawnArrayFriendly=DataInterface() {propertyValues {GpuSyncMode=SyncCpuToGpu, MaxElements=0, IntData [0,1,0]}}) {isSetParameters=true}
   &SpawnBurst_Instantaneous(
    Spawn Count=&Multiply_Int(
     A=25
     B=&SelectIntFromArray(Direct Array Index=$User.StateSelect, Int Selection Array=$Emitter.ShardSpawnArrayFriendly)
    )
    Use Spawn Probability=false
    UseLoopCountLimit=false
   )
   &SpawnBurst_Instantaneous(
    Spawn Count=&Multiply_Int(
     A=50
     B=&SelectIntFromArray(Direct Array Index=$User.StateSelect, Int Selection Array=$Emitter.ShardSpawnArrayFriendly)
    )
    Spawn Time=0.06
    Use Spawn Probability=false
    UseLoopCountLimit=false
   ) {module=SpawnBurst_Instantaneous001}
   &SpawnBurst_Instantaneous(
    Spawn Count=&Multiply_Int(
     A=50
     B=&SelectIntFromArray(Direct Array Index=$User.StateSelect, Int Selection Array=$Emitter.ShardSpawnArrayFriendly)
    )
    Spawn Time=0.15
    Use Spawn Probability=false
    UseLoopCountLimit=false
   ) {module=SpawnBurst_Instantaneous002}
  ]
  particleSpawnScript [
   &FN_SetSpawnGroup()
   /SetVariables(Particles.VisToggle:NiagaraInt32=&MakeIntFromFloat(Float=&ClampFloat(Float=&RandomRangeFloat(Minimum=0, Maximum=3)))) {isSetParameters=true, module=SetVariables_#2}
   &InitializeParticle(
    Color=&SelectLinearColorFromArray(Array Sampling Mode=::NewEnumerator1, Direct Array Index=$User.StateSelect, Color Selection Array=$System.Color Array, Random Seed=?)
    Lifetime Max=4
    Lifetime Min=0.2
    Lifetime Mode=::NewEnumerator1
    Mesh Scale Max=<0.7 0.7 0.7>
    Mesh Scale Min=<0.25 0.5 0.5>
    Mesh Scale Mode=::NewEnumerator2
    Sprite Size Mode=::NewEnumerator1
    Uniform Sprite Size Max=12
    UsePositionOffset=false
    Write Material Random=true
   )
   &SphereLocation(
    Hemisphere Z=true
    MaskSpawnGroup=false
    Offset=<0 0 -30>
    Override Seed=false
    Sphere Radius=&Multiply_Float(A=100, B=&Lerp_Float(A=0, B=$Particles.MeshBaseSize, Alpha=$User.AOESizeMult))
    Surface Only=false
    UseNonUniformScale=false
    UseOffset=true
    UseSphereOrientation=false
   ) {enabled=false, module=SphereLocation001}
   &SphereLocation(
    MaskSpawnGroup=false
    Non Uniform Scale=<0.25 1 1>
    Override Seed=false
    Sphere Radius=0
    Surface Only=false
    UseNonUniformScale=true
    UseOffset=false
    UseSphereOrientation=false
   )
   &AddVelocity(
    Cone Angle=180
    Cone Axis=<0 0 1>
    Module.UseConeFalloff=false
    Module.UseVelDistribution=false
    Override Seed=false
    Velocity Mode=::NewEnumerator2
    Velocity Speed=&RandomRangeFloat(Minimum=500, Maximum=1000)
   ) {module=AddVelocity001}
   &AddVelocity(Velocity=&RandomRangeVector(Minimum=<0 0 50>, Maximum=<1 1 350>)) {module=AddVelocity002}
   &AddVelocityFromPoint(
    Module.Use Falloff Distance=false
    Velocity Strength=&Multiply_Float(A=2500, B=&RandomRangeFloat(Minimum=0.5, Maximum=1))
   ) {enabled=false}
   &AddVelocity(Velocity=&Multiply_VectorByFloat(Vector=<1 0 0>, Float=&RandomRangeFloat(Minimum=-1500, Maximum=2000))) {enabled=false}
   &InitialMeshOrientation(Orientation Vector=&RandomVector(Vector Scale=1), Use Orientation Vector=true, Use Rotation Vector=false)
   /SetVariables(Particles.MeshIndex=&RandomRangeInt(Minimum=0, Maximum=2)) {isSetParameters=true, module=SetVariables_#3}
   &CameraOffset(Camera Offset Amount=-25)
  ]
  particleUpdateScript [
   &ParticleState()
   &UpdateMeshOrientation(
    Rotation Coordinate Space=::NewEnumerator3
    Rotation Rate=&FloatFromCurve(
     FloatCurve=DataInterface() {propertyValues {Curve {keys [
         {arriveTangent=-2.5411, interpMode=RCIM_Cubic, leaveTangent=-2.5411, tangentMode=RCTM_User, time=0, value=1}
         {arriveTangent=-0.173, interpMode=RCIM_Cubic, leaveTangent=-0.173, tangentMode=RCTM_User, time=1, value=0.1}
        ]}}}
     Scale Curve=&FloatFromCurve(
      FloatCurve=DataInterface() {propertyValues {Curve {keys [
          {interpMode=RCIM_Cubic, tangentMode=RCTM_User, time=0, value=-1}
          {arriveTangent=9.6958, interpMode=RCIM_Cubic, leaveTangent=9.6958, tangentMode=RCTM_User, time=0.5, value=0}
          {interpMode=RCIM_Cubic, tangentMode=RCTM_User, time=1, value=1}
         ]}}}
      CurveIndex=&RandomRangeFloat(Minimum=0, Maximum=1)
      Scale Curve=2
     )
    )
    Rotation Vector=&RandomRangeVector(Minimum=<0 1 0>, Maximum=<1 0 0>)
   )
   &Drag(Drag=5, Use Linear Drag=true, Use Rotational Drag=false)
   &SolveForcesAndVelocity(Clamp Velocity=false, Limit Acceleration=false)
   &ScaleColor(
    Scale Alpha=&FloatFromCurve(FloatCurve=DataInterface() {propertyValues {Curve {keys [
         {arriveTangent=-0.0405, interpMode=RCIM_Cubic, leaveTangent=-0.0405, tangentMode=RCTM_User, time=0.6, value=1}
         {interpMode=RCIM_Cubic, leaveTangent=-4.3194, time=1, value=0}
        ]}}})
    Scale RGB=&VectorFromFloat(Value=&Add_Float(
      A=1
      B=&FloatFromCurve(
       FloatCurve=DataInterface() {propertyValues {Curve {keys [
           {arriveTangent=-5.4733, interpMode=RCIM_Cubic, leaveTangent=-5.4733, tangentMode=RCTM_User, time=0, value=1}
           {interpMode=RCIM_Cubic, tangentMode=RCTM_User, time=0.5, value=0}
          ]}}}
       Scale Curve=20
      )
     ))
    ScaleA=true
    ScaleRGB=true
   )
   /SetVariables(Particles.MeshColor:LinearColor=&MakeLinearColorFromVectorAndFloat(
     "Vector (RGB)"=&VectorFromFloat(Value=&Add_Float(
       A=5
       B=&FloatFromCurve(
        FloatCurve=DataInterface() {propertyValues {Curve {keys [
            {arriveTangent=-3.7813, interpMode=RCIM_Cubic, leaveTangent=-3.7813, tangentMode=RCTM_User, time=0, value=1}
            {interpMode=RCIM_Cubic, tangentMode=RCTM_User, time=0.7, value=0}
           ]}}}
        Scale Curve=20
       )
      ))
     "Float (Alpha)"=&FloatFromCurve(
      FloatCurve=DataInterface() {propertyValues {Curve {keys [<at 0.7 1 RCIM_Cubic>, {arriveTangent=-5.6246, interpMode=RCIM_Cubic, leaveTangent=-5.6246, tangentMode=RCTM_User, time=1, value=0}]}}}
      Scale Curve=2.5
     )
    )) {isSetParameters=true, module=SetVariables_#4}
   &ScaleMeshSize(Scale Factor=&VectorFromFloat(Value=&FloatFromCurve(FloatCurve=DataInterface() {propertyValues {Curve {keys [
          {arriveTangent=-0.0405, interpMode=RCIM_Cubic, leaveTangent=-0.0405, tangentMode=RCTM_User, time=0.6, value=1}
          {arriveTangent=-5.4415, interpMode=RCIM_Cubic, leaveTangent=-5.4415, tangentMode=RCTM_User, time=1, value=0}
         ]}}})))
  ]
  rendererClasses [&NiagaraMeshRendererProperties, &NiagaraSpriteRendererProperties]
  renderers [
   {rendererClass=&NiagaraMeshRendererProperties, rendererIndex=0}
   {rendererClass=&NiagaraSpriteRendererProperties, rendererIndex=1}
  ]
 }
 RiftShards001 {
  bEnabled=false
  emitterUpdateScript [
   /SetVariables(Emitter.ShardSpawnArrayEnemy=DataInterface() {propertyValues {GpuSyncMode=SyncCpuToGpu, MaxElements=0, IntData [0,0,1]}}) {isSetParameters=true}
   &EmitterState(Life Cycle Mode=::NewEnumerator1, Loop Behavior=::NewEnumerator1, Loop Delay=$System.CrackDelay, UseLoopDelay=true)
   &SpawnBurst_Instantaneous(
    Spawn Count=&Multiply_Int(
     A=25
     B=&SelectIntFromArray(Direct Array Index=$User.StateSelect, Int Selection Array=$Emitter.ShardSpawnArrayEnemy)
    )
    Use Spawn Probability=false
    UseLoopCountLimit=false
   )
   &SpawnBurst_Instantaneous(
    Spawn Count=&Multiply_Int(
     A=50
     B=&SelectIntFromArray(Direct Array Index=$User.StateSelect, Int Selection Array=$Emitter.ShardSpawnArrayEnemy)
    )
    Spawn Time=0.06
    Use Spawn Probability=false
    UseLoopCountLimit=false
   ) {module=SpawnBurst_Instantaneous001}
   &SpawnBurst_Instantaneous(
    Spawn Count=&Multiply_Int(
     A=50
     B=&SelectIntFromArray(Direct Array Index=$User.StateSelect, Int Selection Array=$Emitter.ShardSpawnArrayEnemy)
    )
    Spawn Time=0.15
    Use Spawn Probability=false
    UseLoopCountLimit=false
   ) {module=SpawnBurst_Instantaneous002}
  ]
  particleSpawnScript [
   &FN_SetSpawnGroup()
   /SetVariables(Particles.VisToggle:NiagaraInt32=&MakeIntFromFloat(Float=&ClampFloat(Float=&RandomRangeFloat(Minimum=0, Maximum=3)))) {isSetParameters=true, module=SetVariables_#2}
   &InitializeParticle(
    Color=&SelectLinearColorFromArray(Array Sampling Mode=::NewEnumerator1, Direct Array Index=$User.StateSelect, Color Selection Array=$System.Color Array, Random Seed=?)
    Lifetime Max=4
    Lifetime Min=0.2
    Lifetime Mode=::NewEnumerator1
    Mesh Scale Max=<0.7 0.7 0.7>
    Mesh Scale Min=<0.25 0.5 0.5>
    Mesh Scale Mode=::NewEnumerator2
    Sprite Size Mode=::NewEnumerator1
    Uniform Sprite Size Max=12
    UsePositionOffset=false
    Write Material Random=true
   )
   &SphereLocation(
    Hemisphere Z=true
    MaskSpawnGroup=false
    Offset=<0 0 -30>
    Override Seed=false
    Sphere Radius=&Multiply_Float(A=100, B=&Lerp_Float(A=0, B=$Particles.MeshBaseSize, Alpha=$User.AOESizeMult))
    Surface Only=false
    UseNonUniformScale=false
    UseOffset=true
    UseSphereOrientation=false
   ) {enabled=false, module=SphereLocation001}
   &SphereLocation(
    MaskSpawnGroup=false
    Non Uniform Scale=<0.25 1 1>
    Override Seed=false
    Sphere Radius=0
    Surface Only=false
    UseNonUniformScale=true
    UseOffset=false
    UseSphereOrientation=false
   )
   &AddVelocity(
    Cone Angle=180
    Cone Axis=<0 0 1>
    Module.UseConeFalloff=false
    Module.UseVelDistribution=false
    Override Seed=false
    Velocity Mode=::NewEnumerator2
    Velocity Speed=&RandomRangeFloat(Minimum=500, Maximum=1000)
   ) {module=AddVelocity001}
   &AddVelocity(Velocity=&RandomRangeVector(Minimum=<0 0 50>, Maximum=<1 1 350>)) {module=AddVelocity002}
   &AddVelocityFromPoint(
    Module.Use Falloff Distance=false
    Velocity Strength=&Multiply_Float(A=2500, B=&RandomRangeFloat(Minimum=0.5, Maximum=1))
   ) {enabled=false}
   &AddVelocity(Velocity=&Multiply_VectorByFloat(Vector=<1 0 0>, Float=&RandomRangeFloat(Minimum=-1500, Maximum=2000))) {enabled=false}
   &InitialMeshOrientation(Orientation Vector=&RandomVector(Vector Scale=1), Use Orientation Vector=true, Use Rotation Vector=false)
   /SetVariables(Particles.MeshIndex=&RandomRangeInt(Minimum=0, Maximum=2)) {isSetParameters=true, module=SetVariables_#3}
   &CameraOffset(Camera Offset Amount=-25)
  ]
  particleUpdateScript [
   &ParticleState()
   &UpdateMeshOrientation(
    Rotation Coordinate Space=::NewEnumerator3
    Rotation Rate=&FloatFromCurve(
     FloatCurve=DataInterface() {propertyValues {Curve {keys [
         {arriveTangent=-2.5411, interpMode=RCIM_Cubic, leaveTangent=-2.5411, tangentMode=RCTM_User, time=0, value=1}
         {arriveTangent=-0.173, interpMode=RCIM_Cubic, leaveTangent=-0.173, tangentMode=RCTM_User, time=1, value=0.1}
        ]}}}
     Scale Curve=&FloatFromCurve(
      FloatCurve=DataInterface() {propertyValues {Curve {keys [
          {interpMode=RCIM_Cubic, tangentMode=RCTM_User, time=0, value=-1}
          {arriveTangent=9.6958, interpMode=RCIM_Cubic, leaveTangent=9.6958, tangentMode=RCTM_User, time=0.5, value=0}
          {interpMode=RCIM_Cubic, tangentMode=RCTM_User, time=1, value=1}
         ]}}}
      CurveIndex=&RandomRangeFloat(Minimum=0, Maximum=1)
      Scale Curve=2
     )
    )
    Rotation Vector=&RandomRangeVector(Minimum=<0 1 0>, Maximum=<1 0 0>)
   )
   &Drag(Drag=5, Use Linear Drag=true, Use Rotational Drag=false)
   &SolveForcesAndVelocity(Clamp Velocity=false, Limit Acceleration=false)
   &ScaleColor(
    Scale Alpha=&FloatFromCurve(FloatCurve=DataInterface() {propertyValues {Curve {keys [
         {arriveTangent=-0.0405, interpMode=RCIM_Cubic, leaveTangent=-0.0405, tangentMode=RCTM_User, time=0.6, value=1}
         {interpMode=RCIM_Cubic, leaveTangent=-4.3194, time=1, value=0}
        ]}}})
    Scale RGB=&VectorFromFloat(Value=&Add_Float(
      A=1
      B=&FloatFromCurve(
       FloatCurve=DataInterface() {propertyValues {Curve {keys [
           {arriveTangent=-5.4733, interpMode=RCIM_Cubic, leaveTangent=-5.4733, tangentMode=RCTM_User, time=0, value=1}
           {interpMode=RCIM_Cubic, tangentMode=RCTM_User, time=0.5, value=0}
          ]}}}
       Scale Curve=20
      )
     ))
    ScaleA=true
    ScaleRGB=true
   )
   /SetVariables(Particles.MeshColor:LinearColor=&MakeLinearColorFromVectorAndFloat(
     "Vector (RGB)"=&VectorFromFloat(Value=&Add_Float(
       A=5
       B=&FloatFromCurve(
        FloatCurve=DataInterface() {propertyValues {Curve {keys [
            {arriveTangent=-3.7813, interpMode=RCIM_Cubic, leaveTangent=-3.7813, tangentMode=RCTM_User, time=0, value=1}
            {interpMode=RCIM_Cubic, tangentMode=RCTM_User, time=0.7, value=0}
           ]}}}
        Scale Curve=20
       )
      ))
     "Float (Alpha)"=&FloatFromCurve(
      FloatCurve=DataInterface() {propertyValues {Curve {keys [<at 0.7 1 RCIM_Cubic>, {arriveTangent=-5.6246, interpMode=RCIM_Cubic, leaveTangent=-5.6246, tangentMode=RCTM_User, time=1, value=0}]}}}
      Scale Curve=2.5
     )
    )) {isSetParameters=true, module=SetVariables_#4}
   &ScaleMeshSize(Scale Factor=&VectorFromFloat(Value=&FloatFromCurve(FloatCurve=DataInterface() {propertyValues {Curve {keys [
          {arriveTangent=-0.0405, interpMode=RCIM_Cubic, leaveTangent=-0.0405, tangentMode=RCTM_User, time=0.6, value=1}
          {arriveTangent=-5.4415, interpMode=RCIM_Cubic, leaveTangent=-5.4415, tangentMode=RCTM_User, time=1, value=0}
         ]}}})))
  ]
  rendererClasses [&NiagaraMeshRendererProperties, &NiagaraSpriteRendererProperties]
  renderers [
   {rendererClass=&NiagaraMeshRendererProperties, rendererIndex=0}
   {rendererClass=&NiagaraSpriteRendererProperties, rendererIndex=1}
  ]
 }
 Screen_Camera_Glow {
  emitterUpdateScript [
   &EmitterState(
    Inactive Response=::NewEnumerator1
    Life Cycle Mode=::NewEnumerator1
    Loop Behavior=::NewEnumerator1
    Loop Duration Mode=::NewEnumerator1
    UseLoopDelay=false
   )
   &SpawnBurst_Instantaneous(Spawn Count=1, Use Spawn Probability=false)
  ]
  particleSpawnScript [
   &InitializeParticle(
    Color=<rgba 0 0.5795 1 1>
    Lifetime=100
    Mesh Scale Mode=::NewEnumerator3
    Mesh Uniform Scale=5
    Position Offset=<0 0 35>
    Sprite Size=<250 250>
    Sprite Size Mode=::NewEnumerator4
    UsePositionOffset=true
    Write Material Random=true
   ) {module=InitializeParticle001}
   &CameraOffset(Camera Offset Amount=-30) {enabled=false}
  ]
  particleUpdateScript [
   &ParticleState(Kill Particles When Lifetime Has Elapsed=false, Let Infinitely Lived Particles Die When Emitter Deactivates=true)
   &SolveForcesAndVelocity(Clamp Velocity=false, Limit Acceleration=false)
   &ScaleColor(
    Scale Alpha=&FloatFromCurve(
     FloatCurve=DataInterface() {propertyValues {Curve {keys [{interpMode=RCIM_Cubic, tangentMode=RCTM_User, time=0.35, value=1}, {interpMode=RCIM_Cubic, tangentMode=RCTM_User, time=1, value=0}]}}}
     CurveIndex=&NormalizeDistanceRange(
      Start Position:NiagaraPosition=&TransformPosition(Position=$Engine.Emitter.SimulationPosition, Source Space=::Simulation, Destination Space=::World)
      End Position:NiagaraPosition=$User.PortalPosition
      Distance:NiagaraFloat=1500
      Invert Normalized Range:NiagaraBool=false
     )
    )
    ScaleA=true
    ScaleRGB=true
   )
   &ScaleColor(
    Scale Alpha=&FloatFromCurve(
     FloatCurve=DataInterface() {propertyValues {Curve {keys [<at -1 1 RCIM_Cubic>, <at -0.5 0>, {interpMode=RCIM_Cubic, tangentMode=RCTM_User, time=1, value=0}]}}}
     CurveIndex=&DotProduct(
      Vector A=&NormalizeVector(
       Vector To Normalize=&SubtractVectorFromPosition(
        Position=$User.PortalPosition
        Vector=&TransformVector(
         Vector=Linked() {linkedVariable {name=Engine.Emitter.SimulationPosition, type=NiagaraPosition}}
         Source Space=::Simulation
         Destination Space=::World
         Transform Vector as Position=true
        )
       )
       Safe Normalize=true
       Safe Fallback Threshold=0
       Safe Fallback Vector=<0 0 0.001>
      )
      Vector B=&NormalizeVector(Vector To Normalize=$Engine.Owner.SystemXAxis, Safe Normalize=true, Safe Fallback Threshold=0, Safe Fallback Vector=<0 0 0.001>)
     )
    )
    ScaleA=true
    ScaleRGB=false
   ) {module=ScaleColor002}
   &DynamicMaterialParameters(
    Index 0 Param 1=1
    Module.Param0WriteEnabled=true
    Module.Param1WriteEnabled=true
    Module.Param2WriteEnabled=true
    Module.Param3WriteEnabled=true
   )
  ]
  rendererClasses [&NiagaraMeshRendererProperties]
  renderers [{rendererClass=&NiagaraMeshRendererProperties, rendererIndex=0}]
 }
 TriangleInitialBurst {
  bEnabled=false
  emitterUpdateScript [
   &EmitterState(Inactive Response=::NewEnumerator1, Life Cycle Mode=::NewEnumerator1, Loop Behavior=::NewEnumerator1, UseLoopDelay=false)
   &SpawnBurst_Instantaneous(Spawn Count=75, Use Spawn Probability=false, UseLoopCountLimit=false)
   &SpawnRate(Spawn Probability=0.8, SpawnRate=25, Use Spawn Probability=true) {enabled=false}
  ]
  particleSpawnScript [
   &InitializeParticle(
    Color=&SelectLinearColorFromArray(Array Sampling Mode=::NewEnumerator1, Direct Array Index=$User.StateSelect, Color Selection Array=$System.Color Array, Random Seed=?)
    Lifetime Max=4
    Lifetime Min=0.2
    Lifetime Mode=::NewEnumerator1
    Sprite Size Mode=::NewEnumerator1
    Uniform Sprite Size Max=15
    Uniform Sprite Size Min=10
    UsePositionOffset=false
    Write Material Random=true
   ) {module=InitializeParticle001}
   /SetVariables(Particles.MeshBaseSize=0.64) {isSetParameters=true}
   &SphereLocation(
    Hemisphere Z=true
    MaskSpawnGroup=false
    Offset=<0 0 -30>
    Override Seed=false
    Sphere Radius=&Multiply_Float(A=100, B=&Lerp_Float(A=0, B=$Particles.MeshBaseSize, Alpha=$User.AOESizeMult))
    Surface Only=false
    UseNonUniformScale=false
    UseOffset=true
    UseSphereOrientation=false
   )
   &AddVelocity(
    Cone Angle=180
    Cone Axis=<0 0 1>
    Module.UseConeFalloff=false
    Module.UseVelDistribution=false
    Override Seed=false
    Velocity Mode=::NewEnumerator2
    Velocity Speed=&RandomRangeFloat(Minimum=500, Maximum=1000)
   ) {module=AddVelocity001}
   &AddVelocity(Velocity=&RandomRangeVector(Minimum=<0 0 50>, Maximum=<1 1 350>))
  ]
  particleUpdateScript [
   &ParticleState()
   &SpriteRotationRate(Rotation Rate=&RandomRangeFloat(Minimum=100, Maximum=300))
   /Niagara/Modules/Update/Forces/CurlNoiseForce(Noise Frequency=500, Noise Strength=150, Pan Noise=false) {module=CurlNoiseForce001}
   &Drag(Drag=3, Use Linear Drag=true, Use Rotational Drag=false)
   &ScaleColor(
    Scale Alpha=&FloatFromCurve(FloatCurve=DataInterface() {propertyValues {Curve {keys [
         <at -0.0092 0.0144 RCIM_Cubic>
         {arriveTangent=-2.5894, leaveTangent=-2.5894, time=0.0967, value=0.9767}
         {arriveTangent=-1.0777, interpMode=RCIM_Cubic, leaveTangent=-1.0777, time=0.8, value=1}
         <at 1.0011 0.002>
        ]}}})
    ScaleA=true
    ScaleRGB=true
   )
   &ScaleSpriteSize(Scale Factor=&Vector2DFromFloat(Value=&FloatFromCurve(FloatCurve=DataInterface() {propertyValues {Curve {keys [
          <at 0 0 RCIM_Cubic>
          {arriveTangent=-1.8533, leaveTangent=-1.8533, time=0.1, value=1}
          {arriveTangent=-1.7911, interpMode=RCIM_Cubic, leaveTangent=-1.7911, tangentMode=RCTM_User, time=0.5, value=0.8}
          <at 1 0>
         ]}}})))
   &ScaleColor(Scale Alpha=$User.GlobalAlpha, ScaleA=true, ScaleRGB=true) {enabled=false, module=ScaleColor001}
   &SolveForcesAndVelocity(Clamp Velocity=false, Limit Acceleration=false, Manually Enable Rotational Solver=true)
   &DynamicMaterialParameters(
    Index 0 Param 1=&RandomRangeFloat(Minimum=0.45, Maximum=0.7)
    Module.Param0WriteEnabled=true
    Module.Param1WriteEnabled=true
    Module.Param2WriteEnabled=true
    Module.Param3WriteEnabled=true
   )
  ]
  rendererClasses [&NiagaraSpriteRendererProperties]
  renderers [
   {rendererClass=&NiagaraSpriteRendererProperties}
   {rendererClass=&NiagaraSpriteRendererProperties, rendererIndex=1}
  ]
 }
}
summary {
 systemName=NS_Camera_Lens_Portal_CloseUp_FX
 emitters [
  {emitterName=LargeGlowBurst, rendererClasses [&NiagaraSpriteRendererProperties]}
  {bEnabled=false, emitterName=TriangleInitialBurst, rendererClasses [&NiagaraSpriteRendererProperties]}
  {
   bEnabled=false
   emitterName=RiftShards
   rendererClasses [&NiagaraMeshRendererProperties, &NiagaraSpriteRendererProperties]
  }
  {
   bEnabled=false
   emitterName=RiftShards001
   rendererClasses [&NiagaraMeshRendererProperties, &NiagaraSpriteRendererProperties]
  }
  {emitterName=Screen_Camera_Glow, rendererClasses [&NiagaraMeshRendererProperties]}
  {emitterName=CameraLensHeat001, rendererClasses [&NiagaraMeshRendererProperties]}
 ]
 userVariables [
  User.AOESizeMult {type=NiagaraFloat, defaultValue=0, description=""}
  User.DirectionBlend {type=NiagaraFloat, defaultValue=0.1, description=""}
  User.DistanceBlend {type=NiagaraFloat, defaultValue=1, description=""}
  User.HeatOpacity {type=NiagaraFloat, defaultValue=0.4, description=""}
  User.HeatScale {type=NiagaraFloat, defaultValue=5, description=""}
  User.PanSpeed {type=NiagaraFloat, defaultValue=5, description=""}
  User.PortalPosition {type=NiagaraPosition, description="", defaultValue {}}
  User.ScreenSizScalar {type=NiagaraFloat, defaultValue=1, description=""}
  User.StateSelect {type=NiagaraInt32}
 ]
}
