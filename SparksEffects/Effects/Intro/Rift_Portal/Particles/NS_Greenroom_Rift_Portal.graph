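# Niagara system for the greenroom rift portal in the intro effects: a glowing, mostly cyan-blue portal with smoke and a large shard mesh.
# Looping corona, flare, glint, smoke and GPU ember emitters surround a persistent blackout sprite, with distance culling on several.
# User.Open drives the pulsing light's color and intensity, and User.TriggerLargeShard gates the shard burst.
# Crack, dark gradient, electric arc, flash and light-shaft emitters are present but disabled.
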NiagaraSystem /SparksEffects/Effects/Intro/Rift_Portal/Particles/NS_Greenroom_Rift_Portal
dependencies {
 usedDataInterfaces [&NiagaraDataInterfaceCurve, &NiagaraDataInterfaceVectorCurve, &NiagaraDataInterfaceArrayFloat]
 usedDynamicInputs []
 usedModules [&SystemState]
}
emitters {
 Corona {
  emitterUpdateScript [
   &EmitterState(
    Life Cycle Mode=::NewEnumerator1
    Loop Delay=$User.InitialStartDelay
    Loop Duration=&RandomRangeFloat(Minimum=2.5, Maximum=2.7, Evaluation Type=::NewEnumerator1)
    Recalculate Duration Each Loop=true
    UseLoopDelay=false
   )
   &SpawnBurst_Instantaneous(Spawn Count=1, Use Spawn Probability=false)
  ]
  particleSpawnScript [
   /Niagara/Modules/Spawn/Initialization/InitializeParticle(
    Color=<rgba 0.2326 0.7931 1.2 1>
    Lifetime=&RandomRangeFloat(Minimum=2.7, Maximum=3.2, Evaluation Type=::NewEnumerator1)
    Module.Write Color=true
    Module.Write Lifetime=true
    Module.Write Mass=true
    Module.Write Material Random=true
    Module.Write Position=false
    Module.Write Scale=false
    Module.Write SpriteRotation=false
    Module.Write SpriteSize=true
    Sprite Size=<150 150>
   ) {enabled=false}
   /Niagara/Modules/Spawn/Initialization/V2/InitializeParticle(
    Color=<rgba 0.5816 1.9826 3 1>
    Lifetime=&RandomRangeFloat(Minimum=2.7, Maximum=3.2)
    Mass Mode=::NewEnumerator1
    Sprite Size=<150 150>
    Sprite Size Mode=::NewEnumerator4
    UsePositionOffset=false
    Write Material Random=true
   ) {module=InitializeParticle001}
   &DynamicMaterialParameters(
    Index 0 Param 1=&RandomRangeFloat(Minimum=0.8, Maximum=1.2, Evaluation Type=::NewEnumerator1)
    Index 0 Param 2=&RandomRangeFloat(Minimum=0.2, Maximum=0.4, Evaluation Type=::NewEnumerator1)
    Index 0 Param 3=&RandomRangeFloat(Minimum=0.4, Maximum=0.6, Evaluation Type=::NewEnumerator1)
    Module.Param0WriteEnabled=true
    Module.Param1WriteEnabled=true
    Module.Param2WriteEnabled=true
    Module.Param3WriteEnabled=true
   )
  ]
  particleUpdateScript [
   &ParticleState()
   &ScaleColor(
    Scale Alpha=&FloatFromCurve(FloatCurve=DataInterface() {propertyValues {Curve {keys [<at 0 0>,<at 0.25 1>,<at 0.75 1>,<at 1 0>]}}})
    ScaleA=true
    ScaleRGB=true
   )
   &DynamicMaterialParameters(
    Index 0 Param 1=&RandomRangeFloat(Minimum=0.8, Maximum=1.2)
    Index 0 Param 2=&RandomRangeFloat(Minimum=0.2, Maximum=0.4)
    Index 0 Param 3=&RandomRangeFloat(Minimum=0.4, Maximum=0.6)
    Index 0 Param 4=1
    Module.Param0WriteEnabled=true
    Module.Param1WriteEnabled=true
    Module.Param2WriteEnabled=true
    Module.Param3WriteEnabled=true
   ) {module=DynamicMaterialParameters001}
   &CameraOffset(Camera Offset Amount=150)
  ]
  rendererClasses [&NiagaraSpriteRendererProperties]
  renderers [{rendererClass=&NiagaraSpriteRendererProperties}]
 }
 Crack {
  bEnabled=false
  emitterUpdateScript [
   &EmitterState(
    Inactive Response=::NewEnumerator1
    Life Cycle Mode=::NewEnumerator1
    Loop Behavior=::NewEnumerator1
    Loop Delay=$User.InitialStartDelay
    Loop Duration Mode=::NewEnumerator1
    UseLoopDelay=true
   )
   &SpawnBurst_Instantaneous(Spawn Count=3, Use Spawn Probability=false)
  ]
  particleSpawnScript [
   /Niagara/Modules/Spawn/Initialization/InitializeParticle(
    Lifetime=1.2
    Mesh Scale=<3.2 3.2 3.2>
    Module.Write Color=true
    Module.Write Lifetime=true
    Module.Write Mass=true
    Module.Write Material Random=true
    Module.Write Position=false
    Module.Write Scale=true
    Module.Write SpriteRotation=false
    Module.Write SpriteSize=false
   ) {enabled=false}
   /Niagara/Modules/Spawn/Initialization/V2/InitializeParticle(
    Lifetime=1.2
    Mass Mode=::NewEnumerator1
    Mesh Scale=<8 8 8>
    Mesh Scale Mode=::NewEnumerator4
    Sprite Size Mode=::NewEnumerator4
    UsePositionOffset=false
    Write Material Random=true
   ) {module=InitializeParticle001}
   &Color(Color=&MakeLinearColorFromVectorAndFloat("Vector (RGB)"=<1.25 1.25 1.25>, "Float (Alpha)"=&ClampFloat(Float=1)))
   &InitialMeshOrientation(Use Rotation Vector=false)
  ]
  particleUpdateScript [
   &ParticleState(Kill Particles When Lifetime Has Elapsed=false, Loop Particles Lifetime=true)
   &ScaleColor(
    Scale Alpha=&FloatFromCurve(FloatCurve=DataInterface() {propertyValues {Curve {keys [<at 0 0 RCIM_Cubic>, <at 0.2 1>, {arriveTangent=-1.25, interpMode=RCIM_Cubic, leaveTangent=-1.25, time=0.8, value=1}, <at 1 0>]}}})
    ScaleA=true
    ScaleRGB=true
   ) {enabled=false}
   &DynamicMaterialParameters(
    Index 0 Param 1=&self:ValueMinMax(Value:NiagaraFloat=$User.Open, MinIn:NiagaraFloat=0, MaxIn:NiagaraFloat=1, MinOut:NiagaraFloat=0.2, MaxOut:NiagaraFloat=1)
    Index 0 Param 2=&self:ValueMinMax(Value:NiagaraFloat=$User.Open, MinIn:NiagaraFloat=0, MaxIn:NiagaraFloat=0.7, MinOut:NiagaraFloat=0, MaxOut:NiagaraFloat=0.4)
    Index 0 Param 3=&self:ValueMinMax(Value:NiagaraFloat=$User.Open, MinIn:NiagaraFloat=0, MaxIn:NiagaraFloat=1, MinOut:NiagaraFloat=-0.05, MaxOut:NiagaraFloat=0)
    Module.Param0WriteEnabled=true
    Module.Param1WriteEnabled=true
    Module.Param2WriteEnabled=true
    Module.Param3WriteEnabled=true
   )
  ]
  rendererClasses [&NiagaraMeshRendererProperties]
  renderers [{rendererClass=&NiagaraMeshRendererProperties, rendererIndex=0}]
 }
 DarkGradient {
  bEnabled=false
  emitterUpdateScript [
   &EmitterState(Inactive Response=::NewEnumerator1, Life Cycle Mode=::NewEnumerator1, Loop Duration=1, UseLoopDelay=true)
   &SpawnBurst_Instantaneous(Spawn Count=1, Use Spawn Probability=false)
  ]
  particleSpawnScript [
   /Niagara/Modules/Spawn/Initialization/InitializeParticle(
    Lifetime=1.7
    Module.Write Color=true
    Module.Write Lifetime=true
    Module.Write Mass=true
    Module.Write Material Random=true
    Module.Write Position=false
    Module.Write Scale=false
    Module.Write SpriteRotation=false
    Module.Write SpriteSize=true
    Sprite Size=<340 340>
   ) {enabled=false}
   /Niagara/Modules/Spawn/Initialization/V2/InitializeParticle(
    Lifetime=1.7
    Mass Mode=::NewEnumerator1
    Sprite Size=<640 640>
    Sprite Size Mode=::NewEnumerator4
    UsePositionOffset=false
    Write Material Random=true
   ) {module=InitializeParticle001}
   &Color(Color=&MakeLinearColorFromVectorAndFloat("Vector (RGB)"=<0.009 0.0041 0.02>, "Float (Alpha)"=&ClampFloat(Float=0.85)))
  ]
  particleUpdateScript [
   &ParticleState()
   &ScaleColor(
    Scale Alpha=&FloatFromCurve(FloatCurve=DataInterface() {propertyValues {Curve {keys [
         <at 0 0 RCIM_Cubic>
         {arriveTangent=1.6667, interpMode=RCIM_Cubic, leaveTangent=1.6667, time=0.4, value=1}
         {arriveTangent=-1.6667, interpMode=RCIM_Cubic, leaveTangent=-1.6667, time=0.6, value=1}
         <at 1 0>
        ]}}})
    ScaleA=true
    ScaleRGB=true
   )
   &DynamicMaterialParameters(
    Index 0 Param 1=3
    Index 0 Param 2=1
    Index 0 Param 3=1
    Index 0 Param 4=1
    Module.Param0WriteEnabled=true
    Module.Param1WriteEnabled=true
    Module.Param2WriteEnabled=true
    Module.Param3WriteEnabled=true
   )
  ]
  rendererClasses [&NiagaraSpriteRendererProperties]
  renderers [{rendererClass=&NiagaraSpriteRendererProperties}]
 }
 ElectricArcsResidual {
  bEnabled=false
  emitterUpdateScript [
   &EmitterState(Life Cycle Mode=::NewEnumerator1, Loop Delay=1, UseLoopDelay=true)
   &EmitterState(Life Cycle Mode=::NewEnumerator1, Loop Delay=8, Loop Duration=7, UseLoopDelay=true) {enabled=false, module=EmitterState001}
   &SpawnRate(Spawn Probability=0.15, SpawnRate=25, Use Spawn Probability=true)
   &SpawnBurst_Instantaneous(Spawn Count=5, Spawn Time=2.2, Use Spawn Probability=false) {enabled=false}
   &SpawnBurst_Instantaneous(Spawn Count=6, Spawn Time=3, Use Spawn Probability=false) {enabled=false, module=SpawnBurst_Instantaneous001}
   &SpawnBurst_Instantaneous(Spawn Count=4, Spawn Time=3.05, Use Spawn Probability=false) {enabled=false, module=SpawnBurst_Instantaneous002}
  ]
  particleSpawnScript [
   /SetVariables(Particles.MeshOrientation=&MakeQuatFromAxisAngle(
     Axis=&RandomRangeVector(Minimum=<0 0 0>, Maximum=<1 0 0>, Evaluation Type=::NewEnumerator1)
     AngleInDegrees=&RandomRangeFloat(Minimum=0, Maximum=360)
    )) {isSetParameters=true}
   /Niagara/Modules/Spawn/Initialization/InitializeParticle(
    Lifetime=&RandomRangeFloat(Minimum=0.08, Maximum=0.2, Evaluation Type=::NewEnumerator1)
    Mesh Scale=&RandomRangeVector(Minimum=<8 6 6>, Maximum=<15 20 20>, Evaluation Type=::NewEnumerator1)
    Module.Write Color=true
    Module.Write Lifetime=true
    Module.Write Mass=true
    Module.Write Material Random=true
    Module.Write Position=false
    Module.Write Scale=true
    Module.Write SpriteRotation=false
    Module.Write SpriteSize=false
   ) {enabled=false}
   /Niagara/Modules/Spawn/Initialization/V2/InitializeParticle(
    Lifetime=&RandomRangeFloat(Minimum=0.08, Maximum=0.2)
    Mass Mode=::NewEnumerator1
    Mesh Scale=&RandomRangeVector(Minimum=<8 6 6>, Maximum=<15 20 20>)
    Mesh Scale Mode=::NewEnumerator4
    Sprite Size Mode=::NewEnumerator4
    UsePositionOffset=false
    Write Material Random=true
   ) {module=InitializeParticle001}
   &DynamicMaterialParameters(
    Index 0 Param 1=&RandomRangeFloat(Minimum=2, Maximum=3, Evaluation Type=::NewEnumerator1)
    Index 0 Param 3=&RandomRangeFloat(Minimum=0.2, Maximum=2, Evaluation Type=::NewEnumerator1)
    Index 0 Param 4=&RandomRangeFloat(Minimum=0, Maximum=1, Evaluation Type=::NewEnumerator1)
    Module.Param0WriteEnabled=true
    Module.Param1WriteEnabled=true
    Module.Param2WriteEnabled=true
    Module.Param3WriteEnabled=true
   )
   &Color(Color=&MakeLinearColorFromVectorAndFloat("Vector (RGB)"=<0.0986 1 0.6632>, "Float (Alpha)"=&ClampFloat(Float=2)))
   /Niagara/Modules/Spawn/Location/SphereLocation(MaskSpawnGroup=false, Module.Override Randomness=false, Sphere Radius=10, Surface Only=true)
   &AddVelocityFromPoint(Module.Use Falloff Distance=false)
  ]
  particleUpdateScript [
   &ParticleState()
   &DynamicMaterialParameters(
    Index 0 Param 2=&FloatFromCurve(FloatCurve=DataInterface() {propertyValues {Curve {keys [{arriveTangent=-1.8894, interpMode=RCIM_Cubic, time=0, value=1}, <at 1 -0.25>]}}})
    Module.Param0WriteEnabled=true
    Module.Param1WriteEnabled=true
    Module.Param2WriteEnabled=true
    Module.Param3WriteEnabled=true
   ) {module=DynamicMaterialParameters001}
   &ScaleColor(
    Scale RGB=&VectorFromCurve(VectorCurve=DataInterface() {propertyValues {
       XCurve {keys [<at 0 50 RCIM_Cubic>, {arriveTangent=-45, interpMode=RCIM_Cubic, leaveTangent=-45, time=0.5, value=30}, <at 1 5>]}
       YCurve {keys [<at 0 19.1123 RCIM_Cubic>, {arriveTangent=-18.2463, interpMode=RCIM_Cubic, leaveTangent=-18.2463, time=0.5, value=13.6826}, <at 1 0.8661>]}
       ZCurve {keys [<at 0 22.1016 RCIM_Cubic>, {arriveTangent=-21.4843, interpMode=RCIM_Cubic, leaveTangent=-21.4843, time=0.5, value=12.787}, <at 1 0.6173>]}
      }})
    ScaleA=true
    ScaleRGB=true
   )
   &ScaleMeshSize(Scale Factor=&VectorFromCurve(
     VectorCurve=DataInterface() {propertyValues {
       XCurve {keys [<at 0.0033 0.5 RCIM_Constant>,<at 0.2054 0.7619 RCIM_Constant>,<at 0.4101 0.8575 RCIM_Constant>,<at 0.6967 1>]}
       YCurve {keys [<at 0.0033 0.6004 RCIM_Constant>,<at 0.2054 0.6716 RCIM_Constant>,<at 0.4101 0.3646 RCIM_Constant>,<at 0.6967 0.2468>]}
       ZCurve {keys [<at 0.0033 0.7767 RCIM_Constant>,<at 0.2054 0.5215 RCIM_Constant>,<at 0.4101 0.4457 RCIM_Constant>,<at 0.6967 0.6282>]}
      }}
     Scale Curve=&VectorFromFloat(Value=1.5)
    ))
   &ScaleMeshSize(Scale Factor=&VectorFromFloat(Value=&FloatFromCurve(FloatCurve=DataInterface() {propertyValues {Curve {keys [<at 0 1>,<at 15000 1.785>]}}}, CurveIndex=$Engine.Owner.LODDistance))) {module=ScaleMeshSize001}
   &ScaleMeshSize(Scale Factor=&MakeVector(
     X=&Lerp_Float(
      A=0.3
      B=1
      Alpha=&FloatFromCurve(FloatCurve=DataInterface() {propertyValues {Curve {keys [
           {interpMode=RCIM_Cubic, tangentMode=RCTM_User, time=0, value=0}
           {arriveTangent=0.7289, interpMode=RCIM_Cubic, leaveTangent=0.7289, tangentMode=RCTM_User, time=0.65, value=0.2}
           {arriveTangentWeight=0.096, interpMode=RCIM_Cubic, tangentMode=RCTM_User, time=0.9, value=1}
          ]}}})
     )
     Y=1
     Z=1
    )) {module=ScaleMeshSize002}
   &CameraOffset(Camera Offset Amount=10)
   &SolveForcesAndVelocity(Clamp Velocity=false, Limit Acceleration=false)
  ]
  rendererClasses [&NiagaraMeshRendererProperties]
  renderers [{rendererClass=&NiagaraMeshRendererProperties, rendererIndex=0}]
 }
 Embers {
  simTarget=GPUComputeSim
  emitterUpdateScript [
   &EmitterState(
    Enable Distance Culling=true
    Life Cycle Mode=::NewEnumerator1
    Loop Delay=1.5
    Loop Duration=8
    MaxDistance=4500
    Module.bUseMaxDistance=true
    Module.bUseMinDistance=false
    Scalability Mode=::NewEnumerator1
    UseLoopDelay=false
   )
   &SpawnRate(
    SpawnRate=&Multiply_Float(
     A=90
     B=&FloatFromCurve(
      FloatCurve=DataInterface() {propertyValues {Curve {keys [
          <at 0 0.25 RCIM_Cubic>
          {arriveTangent=0.0521, interpMode=RCIM_Cubic, leaveTangent=0.0521, time=1.2828, value=1.6362}
          {arriveTangent=-0.3703, interpMode=RCIM_Cubic, leaveTangent=-0.3703, time=2.4472, value=0.3776}
          {arriveTangent=-0.11, interpMode=RCIM_Cubic, leaveTangent=-0.11, time=3.6302, value=0.7671}
          {arriveTangent=0.2907, interpMode=RCIM_Cubic, leaveTangent=0.2907, time=5.034, value=0.0931}
          {arriveTangent=0.3058, interpMode=RCIM_Cubic, leaveTangent=0.3058, time=6.8459, value=1.702}
          <at 8 1>
         ]}}}
      CurveIndex=$Emitter.NormalizedLoopAge
     )
    )
    Use Spawn Probability=false
   )
  ]
  particleSpawnScript [
   /SetVariables(Particles.SpriteRotationRate=&RandomRangeFloat(Minimum=-1, Maximum=1, Evaluation Type=::NewEnumerator1)) {isSetParameters=true}
   /Niagara/Modules/Spawn/Initialization/InitializeParticle(
    Lifetime=&RandomRangeFloat(Minimum=0.6, Maximum=3, Evaluation Type=::NewEnumerator1)
    Module.Write Color=true
    Module.Write Lifetime=true
    Module.Write Mass=true
    Module.Write Material Random=true
    Module.Write Position=false
    Module.Write Scale=false
    Module.Write SpriteRotation=true
    Module.Write SpriteSize=true
    Sprite Rotation=&RandomRangeFloat(Minimum=0, Maximum=1, Evaluation Type=::NewEnumerator1)
    Sprite Size=&Vector2DFromFloat(Value=&RandomRangeFloat(Minimum=3, Maximum=7, Evaluation Type=::NewEnumerator1))
   ) {enabled=false}
   /Niagara/Modules/Spawn/Initialization/V2/InitializeParticle(
    Lifetime=&RandomRangeFloat(Minimum=0.6, Maximum=3)
    Mass Mode=::NewEnumerator1
    Sprite Rotation Angle=&RandomRangeFloat(Minimum=0, Maximum=1)
    Sprite Rotation Mode=::NewEnumerator0
    Sprite Size=&Vector2DFromFloat(Value=&RandomRangeFloat(Minimum=3, Maximum=7))
    Sprite Size Mode=::NewEnumerator4
    UsePositionOffset=false
    Write Material Random=true
   ) {module=InitializeParticle001}
   &Color(Color=&MakeLinearColorFromVectorAndFloat("Vector (RGB)"=<2.6656 15.8408 16>, "Float (Alpha)"=&ClampFloat(Float=1)))
   &SphereLocation(
    MaskSpawnGroup=false
    Module.Override Randomness=false
    Non Uniform Scale=<2.5 1 1>
    Sphere Radius=&RandomRangeFloat(Minimum=100, Maximum=280, Evaluation Type=::NewEnumerator1)
    Surface Only=true
   )
   &AddVelocity(Velocity=&Multiply_VectorByFloat(
     Vector=$Output.SphereLocation.SphereVector
     Float=&RandomRangeFloat(Minimum=-0.04, Maximum=0.02, Evaluation Type=::NewEnumerator1)
    ))
  ]
  particleUpdateScript [
   &ParticleState()
   &Drag(Use Linear Drag=true, Use Rotational Drag=false)
   &CurlNoiseForce(Noise Strength=45, Pan Noise=false)
   &SolveForcesAndVelocity(Clamp Velocity=false, Limit Acceleration=false)
   &ScaleSpriteSize(Scale Factor=&Vector2DFromFloat(Value=&FloatFromCurve(FloatCurve=DataInterface() {propertyValues {Curve {keys [
          <at -0.0044 0.2654>
          {arriveTangent=0.7558, interpMode=RCIM_Cubic, leaveTangent=0.7558, time=0.157, value=0.5261}
          {arriveTangent=-0.6241, interpMode=RCIM_Cubic, leaveTangent=-0.6241, time=0.7034, value=0.8003}
          <at 1 0>
         ]}}})))
   &SpriteRotationRate(Rotation Rate=$Particles.SpriteRotationRate)
  ]
  rendererClasses [&NiagaraSpriteRendererProperties]
  renderers [{rendererClass=&NiagaraSpriteRendererProperties}]
 }
 Flash {
  bEnabled=false
  emitterUpdateScript [
   &EmitterState(Life Cycle Mode=::NewEnumerator1, Loop Behavior=::NewEnumerator1, Loop Delay=$User.InitialStartDelay, Loop Duration=1, UseLoopDelay=true)
   &SpawnBurst_Instantaneous(Use Spawn Probability=false, UseLoopCountLimit=false)
  ]
  particleSpawnScript [
   &InitializeParticle(
    Color=<rgba 0.5207 0.6194 1 1>
    Lifetime=0.5
    Sprite Size Mode=::NewEnumerator3
    Uniform Sprite Size=5000
    UsePositionOffset=false
    Write Material Random=true
   )
   &ShapeLocation(MaskSpawnGroup=false, Offset=$System.Offset, Offset Mode=::NewEnumerator0, Override Seed=false, Sphere Radius=0, UseOwnerScale=true)
   &DynamicMaterialParameters(
    Index 1 Param 1=150
    Module.Param0WriteEnabled=true
    Module.Param0WriteEnabled1=true
    Module.Param1WriteEnabled=true
    Module.Param1WriteEnabled1=true
    Module.Param2WriteEnabled=true
    Module.Param2WriteEnabled1=true
    Module.Param3WriteEnabled=true
    Module.Param3WriteEnabled1=true
    Write Parameter Index 1=true
   )
  ]
  particleUpdateScript [
   &ParticleState()
   &CameraOffset(Camera Offset Amount=&FloatFromCurve(
     FloatCurve=DataInterface() {propertyValues {Curve {keys [
         {arriveTangent=4.0899, interpMode=RCIM_Cubic, leaveTangent=4.0899, tangentMode=RCTM_User, time=0, value=0.5}
         {arriveTangent=-3.6426, interpMode=RCIM_Cubic, leaveTangent=-3.6426, tangentMode=RCTM_User, time=0.2, value=1}
         {arriveTangent=5.165, interpMode=RCIM_Cubic, leaveTangent=5.165, tangentMode=RCTM_User, time=0.4, value=0.3}
         <at 0.75 0>
        ]}}}
     Scale Curve=200
    ))
   &ScaleSpriteSize(
    Uniform Curve Scale=0.85
    Uniform Curve Sprite Scale=DataInterface() {propertyValues {Curve {keys [
        {interpMode=RCIM_Cubic, leaveTangentWeight=0.5401, tangentMode=RCTM_User, time=0, value=0.1}
        {arriveTangent=2.0921, interpMode=RCIM_Cubic, leaveTangent=2.0921, tangentMode=RCTM_User, time=0.1, value=0.2}
        {interpMode=RCIM_Cubic, tangentMode=RCTM_User, time=0.75, value=1}
        {interpMode=RCIM_Cubic, tangentMode=RCTM_User, time=1, value=0.4}
       ]}}}
   )
   &ScaleColor(
    Scale Alpha=&FloatFromCurve(FloatCurve=DataInterface() {propertyValues {Curve {keys [
         {arriveTangent=2.8296, interpMode=RCIM_Cubic, leaveTangent=2.8296, tangentMode=RCTM_User, time=0, value=0.4}
         {interpMode=RCIM_Cubic, tangentMode=RCTM_User, time=0.7, value=1}
         {arriveTangent=-8.4365, interpMode=RCIM_Cubic, leaveTangent=-8.4365, tangentMode=RCTM_User, time=1, value=0}
        ]}}})
    ScaleA=true
    ScaleRGB=true
   )
  ]
  rendererClasses [&NiagaraSpriteRendererProperties]
  renderers [{rendererClass=&NiagaraSpriteRendererProperties}]
 }
 Glints {
  emitterUpdateScript [
   &EmitterState(
    Enable Distance Culling=true
    Life Cycle Mode=::NewEnumerator1
    Loop Delay=$User.InitialStartDelay
    Loop Duration=&RandomRangeFloat(Minimum=0.8, Maximum=2, Evaluation Type=::NewEnumerator1)
    MaxDistance=3200
    Module.bUseMaxDistance=true
    Module.bUseMinDistance=false
    Recalculate Duration Each Loop=true
    Scalability Mode=::NewEnumerator1
    UseLoopDelay=false
   )
   &SpawnRate(
    SpawnRate=&RandomRangeFloat(Minimum=0.5, Maximum=5, Evaluation Type=::NewEnumerator1)
    Use Spawn Probability=false
   )
   &SpawnBurst_Instantaneous(Spawn Count=&RandomRangeInt(Minimum=0, Maximum=2), Spawn Time=0.15, Use Spawn Probability=false)
   &SpawnBurst_Instantaneous(Spawn Count=&RandomRangeInt(Minimum=0, Maximum=2), Spawn Time=0.9, Use Spawn Probability=false) {module=SpawnBurst_Instantaneous001}
  ]
  particleSpawnScript [
   /Niagara/Modules/Spawn/Initialization/InitializeParticle(
    Lifetime=&RandomRangeFloat(Minimum=0.12, Maximum=1.5, Evaluation Type=::NewEnumerator1)
    Module.Write Color=true
    Module.Write Lifetime=true
    Module.Write Mass=true
    Module.Write Material Random=true
    Module.Write Position=false
    Module.Write Scale=false
    Module.Write SpriteRotation=false
    Module.Write SpriteSize=true
    Sprite Size=&RandomRangeVector2D(Minimum=<120 80>, Maximum=<200 140>, Evaluation Type=::NewEnumerator1)
   ) {enabled=false}
   /Niagara/Modules/Spawn/Initialization/V2/InitializeParticle(
    Lifetime=&RandomRangeFloat(Minimum=0.12, Maximum=1.5)
    Mass Mode=::NewEnumerator1
    Sprite Size=&RandomRangeVector2D(Minimum=<120 80>, Maximum=<200 140>)
    Sprite Size Mode=::NewEnumerator4
    UsePositionOffset=false
    Write Material Random=true
   ) {module=InitializeParticle001}
   &Color(Color=&MakeLinearColorFromVectorAndFloat("Vector (RGB)"=<0.2398 1.0892 1.5>, "Float (Alpha)"=&ClampFloat(Float=1)))
   &ShapeLocation(MaskSpawnGroup=false, Override Seed=false, Sphere Radius=300, UseOwnerScale=true)
  ]
  particleUpdateScript [
   &ParticleState()
   &ScaleSpriteSize(Scale Factor=&Vector2DFromFloat(Value=&FloatFromCurve(FloatCurve=DataInterface() {propertyValues {Curve {keys [
          <at 0 0 RCIM_Cubic>
          {arriveTangent=1.4286, interpMode=RCIM_Cubic, leaveTangent=1.4286, time=0.5754, value=0.3493}
          {arriveTangent=-0.8228, interpMode=RCIM_Cubic, leaveTangent=-0.8228, time=0.7, value=1}
          <at 1 0>
         ]}}})))
   &Drag(Drag=0.2, Use Linear Drag=true, Use Rotational Drag=false)
   &SolveForcesAndVelocity(Clamp Velocity=false, Limit Acceleration=false)
  ]
  rendererClasses [&NiagaraSpriteRendererProperties]
  renderers [{rendererClass=&NiagaraSpriteRendererProperties}]
 }
 Glints_Center {
  emitterUpdateScript [
   &EmitterState(
    Enable Distance Culling=true
    Life Cycle Mode=::NewEnumerator1
    Loop Delay=$User.InitialStartDelay
    Loop Duration=&RandomRangeFloat(Minimum=0.8, Maximum=2, Evaluation Type=::NewEnumerator1)
    MaxDistance=3200
    Module.bUseMaxDistance=true
    Module.bUseMinDistance=false
    Recalculate Duration Each Loop=true
    Scalability Mode=::NewEnumerator1
    UseLoopDelay=false
   )
   &SpawnRate(
    SpawnRate=&RandomRangeFloat(Minimum=0.5, Maximum=5, Evaluation Type=::NewEnumerator1)
    Use Spawn Probability=false
   )
   &SpawnBurst_Instantaneous(Spawn Count=&RandomRangeInt(Minimum=0, Maximum=2), Spawn Time=0.15, Use Spawn Probability=false)
   &SpawnBurst_Instantaneous(Spawn Count=&RandomRangeInt(Minimum=0, Maximum=2), Spawn Time=0.9, Use Spawn Probability=false) {module=SpawnBurst_Instantaneous001}
  ]
  particleSpawnScript [
   /Niagara/Modules/Spawn/Initialization/InitializeParticle(
    Lifetime=&RandomRangeFloat(Minimum=0.12, Maximum=1.5, Evaluation Type=::NewEnumerator1)
    Module.Write Color=true
    Module.Write Lifetime=true
    Module.Write Mass=true
    Module.Write Material Random=true
    Module.Write Position=false
    Module.Write Scale=false
    Module.Write SpriteRotation=false
    Module.Write SpriteSize=true
    Sprite Size=&RandomRangeVector2D(Minimum=<120 80>, Maximum=<200 140>, Evaluation Type=::NewEnumerator1)
   ) {enabled=false}
   /Niagara/Modules/Spawn/Initialization/V2/InitializeParticle(
    Lifetime=&RandomRangeFloat(Minimum=0.12, Maximum=1.5)
    Mass Mode=::NewEnumerator1
    Sprite Size=&RandomRangeVector2D(Minimum=<120 80>, Maximum=<200 140>)
    Sprite Size Mode=::NewEnumerator4
    UsePositionOffset=false
    Write Material Random=true
   ) {module=InitializeParticle001}
   &Color(Color=&MakeLinearColorFromVectorAndFloat("Vector (RGB)"=<0.2398 1.0892 1.5>, "Float (Alpha)"=&ClampFloat(Float=1)))
   &ShapeLocation(MaskSpawnGroup=false, Override Seed=false, Sphere Radius=300, UseOwnerScale=true) {enabled=false}
  ]
  particleUpdateScript [
   &ParticleState()
   &ScaleSpriteSize(Scale Factor=&Vector2DFromFloat(Value=&FloatFromCurve(FloatCurve=DataInterface() {propertyValues {Curve {keys [
          <at 0 0 RCIM_Cubic>
          {arriveTangent=1.4286, interpMode=RCIM_Cubic, leaveTangent=1.4286, time=0.5754, value=0.3493}
          {arriveTangent=-0.8228, interpMode=RCIM_Cubic, leaveTangent=-0.8228, time=0.7, value=1}
          <at 1 0>
         ]}}})))
   &Drag(Drag=0.2, Use Linear Drag=true, Use Rotational Drag=false)
   &SolveForcesAndVelocity(Clamp Velocity=false, Limit Acceleration=false)
  ]
  rendererClasses [&NiagaraSpriteRendererProperties]
  renderers [{rendererClass=&NiagaraSpriteRendererProperties}]
 }
 Light {
  emitterUpdateScript [
   &EmitterState(
    Life Cycle Mode=::NewEnumerator1
    Loop Delay=$User.InitialStartDelay
    Loop Duration=&RandomRangeFloat(Minimum=2.4, Maximum=2.7, Evaluation Type=::NewEnumerator1)
    Recalculate Duration Each Loop=true
    UseLoopDelay=false
   )
   &SpawnBurst_Instantaneous(Spawn Count=1, Use Spawn Probability=false)
  ]
  particleSpawnScript [
   /Niagara/Modules/Spawn/Initialization/InitializeParticle(
    Color=<rgba 0.147 0.5957 1 1>
    Lifetime=&RandomRangeFloat(Minimum=2.7, Maximum=3.2, Evaluation Type=::NewEnumerator1)
    Module.Write Color=true
    Module.Write Lifetime=true
    Module.Write Mass=true
    Module.Write Material Random=true
    Module.Write Position=false
    Module.Write Scale=false
    Module.Write SpriteRotation=false
    Module.Write SpriteSize=true
    Sprite Size=<500 500>
   ) {enabled=false}
   /Niagara/Modules/Spawn/Initialization/V2/InitializeParticle(
    Color=<rgba 0.147 0.5957 1 1>
    Lifetime=&RandomRangeFloat(Minimum=2.7, Maximum=3.2)
    Mass Mode=::NewEnumerator1
    Sprite Size=<500 500>
    Sprite Size Mode=::NewEnumerator4
    UsePositionOffset=false
    Write Material Random=true
   ) {module=InitializeParticle001}
   /SetVariables(Particles.LightBrightness=1500) {isSetParameters=true}
  ]
  particleUpdateScript [
   /SetVariables(Particles.LightColorInit:LinearColor=&MultiplyLinearColor(
     Linear Color=&LerpLinearColors(StartColor=<rgba 1 4 8 1>, EndColor=<rgba 10 15 30 1>, LerpFactor=$User.Open)
     Scale Factor=&MakeVector4_FromFloat(Float=&FloatFromCurve(
       FloatCurve=DataInterface() {propertyValues {Curve {keys [
           <at 0 0.1 RCIM_Cubic>
           {arriveTangent=1.2, interpMode=RCIM_Cubic, leaveTangent=1.2, time=0.25, value=1}
           {arriveTangent=-1.2, interpMode=RCIM_Cubic, leaveTangent=-1.2, time=0.75, value=1}
           <at 1 0.1>
          ]}}}
       Scale Curve=3
      ))
    )) {isSetParameters=true}
   /SetVariables(Particles.LightColor=&MultiplyLinearColor(
     Linear Color=$Particles.LightColorInit
     Scale Factor=&MakeVector4_FromLinearColor(LinearColor=$Particles.Color)
    )) {isSetParameters=true}
   /SetVariables(Particles.LightExponent=12) {isSetParameters=true}
   /SetVariables(Particles.LightRadius=1500) {isSetParameters=true}
   &ParticleState()
   &ScaleColor(
    Scale Alpha=&FloatFromCurve(FloatCurve=DataInterface() {propertyValues {Curve {keys [
         <at 0 0 RCIM_Cubic>
         {arriveTangent=1.3333, interpMode=RCIM_Cubic, leaveTangent=1.3333, time=0.25, value=1}
         {arriveTangent=-1.3333, interpMode=RCIM_Cubic, leaveTangent=-1.3333, time=0.75, value=1}
         <at 1 0>
        ]}}})
    ScaleA=true
    ScaleRGB=true
   )
   /SetVariables(Particles.LightIntensityRamp:NiagaraFloat=&FloatFromCurve(
     FloatCurve=DataInterface() {propertyValues {Curve {keys [
         <at 0 0 RCIM_Cubic>
         {arriveTangent=1.3333, interpMode=RCIM_Cubic, leaveTangent=1.3333, time=0.25, value=1}
         {arriveTangent=-1.3333, interpMode=RCIM_Cubic, leaveTangent=-1.3333, time=0.75, value=1}
         <at 1 0>
        ]}}}
     Scale Curve=&Lerp_Float(A=$Particles.LightBrightness, B=&Multiply_Float(A=$Particles.LightBrightness, B=5), Alpha=$User.Open)
    )) {isSetParameters=true}
  ]
  rendererClasses [
   &NiagaraSpriteRendererProperties
   &NiagaraLightRendererProperties
   &NiagaraComponentRendererProperties
  ]
  renderers [
   {rendererClass=&NiagaraSpriteRendererProperties}
   {rendererClass=&NiagaraLightRendererProperties, rendererIndex=1}
   {rendererClass=&NiagaraComponentRendererProperties, rendererIndex=2}
  ]
 }
 LightShaftsBurst {
  bEnabled=false
  emitterUpdateScript [
   &EmitterState(Life Cycle Mode=::NewEnumerator1, Loop Behavior=::NewEnumerator1, Loop Delay=$User.InitialStartDelay, Loop Duration=3, UseLoopDelay=true)
   &SpawnBurst_Instantaneous(Use Spawn Probability=false, UseLoopCountLimit=false)
  ]
  particleSpawnScript [
   &InitializeParticle(
    Color=<rgba 1.2748 5.156 10 1>
    Lifetime=3
    Mesh Scale Mode=::NewEnumerator3
    Mesh Uniform Scale=5
    UsePositionOffset=false
    Write Material Random=true
   )
   &InitialMeshOrientation(Orientation Vector=<0.1 0 1>, Use Orientation Vector=true, Use Rotation Vector=false)
  ]
  particleUpdateScript [
   &ParticleState()
   &DynamicMaterialParameters(
    Index 0 Param 1=&FloatFromCurve(FloatCurve=DataInterface() {propertyValues {Curve {keys [{interpMode=RCIM_Cubic, tangentMode=RCTM_User, time=0, value=1}, {interpMode=RCIM_Cubic, tangentMode=RCTM_User, time=1, value=0.25}]}}})
    Index 0 Param 2=&Lerp_Float(
     A=30
     B=44.3192
     Alpha=&FloatFromCurve(FloatCurve=DataInterface() {propertyValues {Curve {keys [
          {arriveTangent=4.7763, interpMode=RCIM_Cubic, leaveTangent=4.7763, tangentMode=RCTM_User, time=0, value=0}
          {arriveTangent=0.4097, interpMode=RCIM_Cubic, leaveTangent=0.4097, tangentMode=RCTM_User, time=0.1, value=0.9}
          <at 1 1>
         ]}}})
    )
    Index 0 Param 3=&FloatFromCurve(
     FloatCurve=DataInterface() {propertyValues {Curve {keys [
         {arriveTangent=20.3138, interpMode=RCIM_Cubic, leaveTangent=20.3138, tangentMode=RCTM_User, time=0, value=0.5}
         {arriveTangent=0.5, interpMode=RCIM_Cubic, leaveTangent=0.5, time=0.05, value=1}
         {interpMode=RCIM_Cubic, tangentMode=RCTM_User, time=1, value=0.25}
        ]}}}
     Scale Curve=0.65
    )
    Index 0 Param 4=200
    Module.Param0WriteEnabled=true
    Module.Param1WriteEnabled=true
    Module.Param2WriteEnabled=true
    Module.Param3WriteEnabled=true
   )
   &ScaleColor(
    Scale RGB=&VectorFromFloat(Value=&FloatFromCurve(FloatCurve=DataInterface() {propertyValues {Curve {keys [{interpMode=RCIM_Cubic, tangentMode=RCTM_User, time=0, value=1}, {interpMode=RCIM_Cubic, tangentMode=RCTM_User, time=1, value=0}]}}}))
    ScaleA=true
    ScaleRGB=true
   ) {enabled=false}
   &ScaleColor(
    Scale RGB=&VectorFromFloat(Value=&FloatFromCurve(FloatCurve=DataInterface() {propertyValues {Curve {keys [
          {arriveTangent=-31.5859, interpMode=RCIM_Cubic, leaveTangent=-31.5859, tangentMode=RCTM_User, time=0, value=7}
          {arriveTangent=-4.5412, interpMode=RCIM_Cubic, leaveTangent=-4.5412, tangentMode=RCTM_User, time=0.25, value=1}
          <at 1 0>
         ]}}}))
    ScaleA=true
    ScaleRGB=true
   ) {module=ScaleColor001}
  ]
  rendererClasses [&NiagaraMeshRendererProperties]
  renderers [{rendererClass=&NiagaraMeshRendererProperties, rendererIndex=0}]
 }
 Portal_Blackout {
  emitterUpdateScript [
   &EmitterState(Life Cycle Mode=::NewEnumerator1, Loop Behavior=::NewEnumerator1, Loop Duration Mode=::NewEnumerator1, UseLoopDelay=false)
   &SpawnBurst_Instantaneous(Use Spawn Probability=false, UseLoopCountLimit=false)
  ]
  particleSpawnScript [
   &InitializeParticle(
    Color=<rgba 0.0055 0.9798 1.25 0.99>
    Position Offset=<-2.5 0 0>
    Sprite Size Mode=::NewEnumerator3
    Uniform Sprite Size=1280
    UsePositionOffset=true
    Write Material Random=true
   )
   /SetVariables() {isSetParameters=true, module=SetVariables_#1}
  ]
  particleUpdateScript [
   &ParticleState(Kill Particles When Lifetime Has Elapsed=false, Let Infinitely Lived Particles Die When Emitter Deactivates=true)
   &DynamicMaterialParameters(
    Index 0 Param 1=1
    Module.Index0Param0Write=true
    Module.Index0Param1Write=true
    Module.Index0Param2Write=true
    Module.Index0Param3Write=true
   )
   &CameraOffset(Camera Offset Amount=-5)
   &ScaleColor(
    Scale RGB=&VectorFromFloat(Value=&FloatFromCurve(
      FloatCurve=DataInterface() {propertyValues {Curve {keys [
          <at 0 0.05 RCIM_Cubic>
          {arriveTangent=0.1727, interpMode=RCIM_Cubic, leaveTangent=0.4923, tangentMode=RCTM_Break, time=2, value=0.125}
          {arriveTangent=0.3252, interpMode=RCIM_Cubic, leaveTangent=0.3252, tangentMode=RCTM_User, time=3, value=1}
         ]}}}
      CurveIndex=$Emitter.Age
     ))
    ScaleA=true
    ScaleRGB=true
   ) {enabled=false}
  ]
  rendererClasses [&NiagaraSpriteRendererProperties]
  renderers [{rendererClass=&NiagaraSpriteRendererProperties}]
 }
 Portal_Smoke {
  emitterUpdateScript [
   &EmitterState(Delay First Loop Only=true, Life Cycle Mode=::NewEnumerator1, Loop Delay=$User.InitialStartDelay, Loop Duration=2, UseLoopDelay=true)
   &SpawnRate(SpawnRate=10, Use Spawn Probability=false)
  ]
  particleSpawnScript [
   &InitializeParticle(
    Color Maximum=<rgba 0 0.2228 0.6042 1>
    Color Minimum=<rgba 0.421 0.2362 0.4427 1>
    Color Mode=::NewEnumerator2
    Lifetime Max=1.75
    Lifetime Min=1.4
    Lifetime Mode=::NewEnumerator1
    Mass Mode=::NewEnumerator3
    Sprite Rotation Mode=::NewEnumerator2
    Sprite Size Mode=::NewEnumerator1
    Uniform Sprite Size Max=450
    Uniform Sprite Size Min=350
    UsePositionOffset=false
    Write Material Random=true
   )
   /SetVariables(Particles.SpriteRotationRate=&RandomRangeFloat(Minimum=-3, Maximum=3)) {isSetParameters=true}
   &ShapeLocation(
    Large Radius=180
    MaskSpawnGroup=false
    Override Seed=false
    Rotation Mode=::NewEnumerator1
    Shape Primitive=::NewEnumerator3
    UseOwnerScale=true
   )
   &AddVelocity(
    Cone Angle=32
    Distribution Along Cone Axis=0.25
    Module.UseConeFalloff=true
    Module.UseVelDistribution=true
    Override Seed=false
    Speed Falloff From Cone Axis=0.5
    Velocity Mode=::NewEnumerator2
    Velocity Speed=&RandomRangeFloat(Minimum=100, Maximum=250)
   )
   &DynamicMaterialParameters(
    Index 0 Param 1=&RandomRangeFloat(Minimum=0, Maximum=1)
    Index 0 Param 2=&RandomRangeFloat(Minimum=0, Maximum=1)
    Module.Param0WriteEnabled=true
    Module.Param1WriteEnabled=true
    Module.Param2WriteEnabled=true
    Module.Param3WriteEnabled=true
   )
  ]
  particleUpdateScript [
   &ParticleState()
   &GravityForce(Gravity=<0 0 -5>)
   &Drag(Drag=0.25, Use Linear Drag=true, Use Rotational Drag=false)
   &ScaleColor(
    Scale Alpha=&Multiply_Float(
     A=&FloatFromCurve(
      FloatCurve=DataInterface() {propertyValues {Curve {keys [
          {interpMode=RCIM_Cubic, tangentMode=RCTM_User, time=0, value=0}
          {interpMode=RCIM_Cubic, tangentMode=RCTM_User, time=0.5, value=1}
          {tangentMode=RCTM_User, time=1, value=0}
         ]}}}
      Scale Curve=0.5
     )
     B=1
    )
    ScaleA=true
    ScaleRGB=true
   )
   &SpriteRotationRate(Delta Time:NiagaraFloat=$Particles.NormalizedAge, Rotation Rate=$Particles.SpriteRotationRate)
   &ScaleSpriteSize(Uniform Curve Sprite Scale=DataInterface() {propertyValues {Curve {keys [<at 0 0.8>, {arriveTangent=-1.1268, leaveTangent=-1.4471, time=1, value=1}]}}})
   &SolveForcesAndVelocity(Clamp Velocity=false, Limit Acceleration=false)
  ]
  rendererClasses [&NiagaraSpriteRendererProperties]
  renderers [
   {rendererClass=&NiagaraSpriteRendererProperties}
   {rendererClass=&NiagaraSpriteRendererProperties, rendererIndex=1}
  ]
 }
 RoundFlare {
  emitterUpdateScript [
   &EmitterState(
    Enable Distance Culling=true
    Life Cycle Mode=::NewEnumerator1
    Loop Delay=1.5
    Loop Duration=&RandomRangeFloat(Minimum=1, Maximum=2.7)
    MaxDistance=10000
    Module.bUseMaxDistance=true
    Module.bUseMinDistance=false
    Recalculate Duration Each Loop=true
    Scalability Mode=::NewEnumerator1
    UseLoopDelay=false
   )
   &SpawnBurst_Instantaneous(Spawn Count=1, Use Spawn Probability=false)
  ]
  particleSpawnScript [
   /Niagara/Modules/Spawn/Initialization/InitializeParticle(
    Color=<rgba 0.2453 0.9334 1.25 1>
    Lifetime=&RandomRangeFloat(Minimum=2.7, Maximum=3.2, Evaluation Type=::NewEnumerator1)
    Module.Write Color=true
    Module.Write Lifetime=true
    Module.Write Mass=true
    Module.Write Material Random=true
    Module.Write Position=false
    Module.Write Scale=false
    Module.Write SpriteRotation=false
    Module.Write SpriteSize=true
    Sprite Size=<120 120>
   ) {enabled=false}
   /Niagara/Modules/Spawn/Initialization/V2/InitializeParticle(
    Color=<rgba 0.7035 4.5515 5 0.45>
    Lifetime=&RandomRangeFloat(Minimum=2.7, Maximum=3.2)
    Mass Mode=::NewEnumerator1
    Sprite Size=<80 80>
    Sprite Size Mode=::NewEnumerator4
    UsePositionOffset=false
    Write Material Random=true
   ) {module=InitializeParticle001}
   &DynamicMaterialParameters(
    Index 0 Param 1=&RandomRangeFloat(Minimum=0.016, Maximum=0.5, Evaluation Type=::NewEnumerator1)
    Index 0 Param 2=1
    Index 0 Param 3=1
    Index 0 Param 4=1
    Module.Param0WriteEnabled=true
    Module.Param1WriteEnabled=true
    Module.Param2WriteEnabled=true
    Module.Param3WriteEnabled=true
   )
  ]
  particleUpdateScript [
   &ParticleState()
   &ScaleColor(
    Scale Alpha=&FloatFromCurve(FloatCurve=DataInterface() {propertyValues {Curve {keys [
         <at 0 0 RCIM_Cubic>
         {arriveTangent=1.3333, interpMode=RCIM_Cubic, leaveTangent=1.3333, time=0.25, value=1}
         {arriveTangent=-1.3333, interpMode=RCIM_Cubic, leaveTangent=-1.3333, time=0.75, value=1}
         <at 1 0>
        ]}}})
    ScaleA=true
    ScaleRGB=true
   )
   &DynamicMaterialParameters(
    Index 0 Param 1=&RandomRangeFloat(Minimum=0.016, Maximum=0.5)
    Index 0 Param 2=1
    Index 0 Param 3=1
    Index 0 Param 4=1
    Module.Param0WriteEnabled=true
    Module.Param1WriteEnabled=true
    Module.Param2WriteEnabled=true
    Module.Param3WriteEnabled=true
   ) {module=DynamicMaterialParameters001}
   &CameraOffset(Camera Offset Amount=10)
   &ScaleSpriteSize(Scale Factor=&Vector2DFromFloat(Value=&FloatFromCurve(FloatCurve=DataInterface() {propertyValues {Curve {keys [<at 0 1>,<at 10000 3>]}}}, CurveIndex=$Engine.Owner.LODDistance)))
  ]
  rendererClasses [&NiagaraSpriteRendererProperties]
  renderers [{rendererClass=&NiagaraSpriteRendererProperties}]
 }
 Shard_Mesh_Large {
  emitterUpdateScript [
   &EmitterState(Life Cycle Mode=::NewEnumerator1, Loop Behavior=::NewEnumerator1, Loop Delay=3, Loop Duration Mode=::NewEnumerator1, UseLoopDelay=false)
   &SpawnBurst_Instantaneous(
    Age:NiagaraFloat=&Multiply_Float(
     A=$Emitter.LoopedAge
     B=&MakeCustomFloatFromBool(Bool=$User.TriggerLargeShard, True Float=1, False Float=0)
    )
    Use Spawn Probability=false
    UseLoopCountLimit=false
   )
  ]
  particleSpawnScript [
   &InitializeParticle(
    Color=<rgba 0 0.6388 1 1>
    Mesh Scale Mode=::NewEnumerator3
    Mesh Uniform Scale=2.25
    Sprite Size Mode=::NewEnumerator3
    Uniform Sprite Size=1450
    UsePositionOffset=false
    Write Material Random=true
   )
   /SetVariables() {isSetParameters=true, module=SetVariables_#1}
   &InitialMeshOrientation(Mesh Orientation Mode=::NewEnumerator2, Rotation=<0.345 0 0>, Use Rotation Vector=true)
  ]
  particleUpdateScript [
   &ParticleState(Kill Particles When Lifetime Has Elapsed=false, Let Infinitely Lived Particles Die When Emitter Deactivates=true)
   &ScaleMeshSize(Scale Factor=&VectorFromFloat(Value=&FloatFromCurve(FloatCurve=DataInterface() {propertyValues {Curve {keys [
          {arriveTangent=5.5712, interpMode=RCIM_Cubic, leaveTangent=5.5712, tangentMode=RCTM_User, time=0, value=0.75}
          {interpMode=RCIM_Cubic, tangentMode=RCTM_User, time=0.25, value=1}
         ]}}})))
   &DynamicMaterialParameters(Module.Index0Param0Write=true, Module.Index0Param1Write=true, Module.Index0Param2Write=true, Module.Index0Param3Write=true)
   &CameraOffset(Camera Offset Amount=5)
   &ScaleColor(
    Scale Alpha=&FloatFromCurve(FloatCurve=DataInterface() {propertyValues {Curve {keys [{interpMode=RCIM_Cubic, tangentMode=RCTM_User, time=0, value=0}, {interpMode=RCIM_Cubic, tangentMode=RCTM_User, time=0.25, value=1}]}}})
    ScaleA=true
    ScaleRGB=true
   )
  ]
  rendererClasses [&NiagaraSpriteRendererProperties, &NiagaraMeshRendererProperties]
  renderers [
   {rendererClass=&NiagaraSpriteRendererProperties}
   {rendererClass=&NiagaraMeshRendererProperties, rendererIndex=1}
   {rendererClass=&NiagaraMeshRendererProperties, rendererIndex=2}
  ]
 }
 Shard_Mesh_Small {
  emitterUpdateScript [
   &EmitterState(
    Life Cycle Mode=::NewEnumerator1
    Loop Behavior=::NewEnumerator1
    Loop Delay=$User.InitialStartDelay
    Loop Duration Mode=::NewEnumerator1
    UseLoopDelay=true
   )
   &SpawnBurst_Instantaneous(Spawn Time=0.1, Use Spawn Probability=false, UseLoopCountLimit=false)
  ]
  particleSpawnScript [
   &PartitionParticles()
   &InitializeParticle(
    Color=<rgba 0.7917 0.9248 1 1>
    Lifetime Mode=::NewEnumerator1
    Mesh Scale Mode=::NewEnumerator3
    Mesh Uniform Scale=&SelectFloatFromArray(Direct Array Index=$Particles.PartitionParticles.Partition, Float Selection Array=DataInterface() {propertyValues {FloatData [3,1.5]}})
    Sprite Size Mode=::NewEnumerator3
    Uniform Sprite Size=1450
    UsePositionOffset=false
    Write Material Random=true
   )
   /SetVariables() {isSetParameters=true, module=SetVariables_#1}
   &InitialMeshOrientation(Mesh Orientation Mode=::NewEnumerator2, Rotation=<0.345 0 0>, Use Rotation Vector=true) {enabled=false}
   &InitialMeshOrientation(
    Mesh Orientation Mode=::NewEnumerator2
    Rotation=&RandomRangeVector(Minimum=<0.345 0 0>, Maximum=<1 0 0>)
    Use Rotation Vector=true
   ) {module=InitialMeshOrientation001}
  ]
  particleUpdateScript [
   &ParticleState(Kill Particles When Lifetime Has Elapsed=false, Loop Particles Lifetime=true)
   &UpdateMeshOrientation(Rotation Rate=&RandomRangeFloat(Minimum=0.01, Maximum=-0.01), Rotation Vector=<1 0 0.1>)
   &ScaleMeshSize(Scale Factor=&VectorFromFloat(Value=&FloatFromCurve(
      FloatCurve=DataInterface() {propertyValues {Curve {keys [
          {arriveTangent=5.5712, interpMode=RCIM_Cubic, leaveTangent=5.5712, tangentMode=RCTM_User, time=0, value=0.75}
          {interpMode=RCIM_Cubic, tangentMode=RCTM_User, time=0.25, value=1}
         ]}}}
      CurveIndex=$User.Open
     )))
   &DynamicMaterialParameters(Module.Index0Param0Write=true, Module.Index0Param1Write=true, Module.Index0Param2Write=true, Module.Index0Param3Write=true)
   &CameraOffset(Camera Offset Amount=10)
   &JitterPosition() {enabled=false}
  ]
  rendererClasses [&NiagaraSpriteRendererProperties, &NiagaraMeshRendererProperties]
  renderers [
   {rendererClass=&NiagaraSpriteRendererProperties}
   {rendererClass=&NiagaraMeshRendererProperties, rendererIndex=1}
   {rendererClass=&NiagaraMeshRendererProperties, rendererIndex=2}
   {rendererClass=&NiagaraMeshRendererProperties, rendererIndex=3}
  ]
 }
 SharpFlare {
  emitterUpdateScript [
   &EmitterState(
    Enable Distance Culling=true
    Life Cycle Mode=::NewEnumerator1
    Loop Delay=$User.InitialStartDelay
    Loop Duration=&RandomRangeFloat(Minimum=2.4, Maximum=2.7, Evaluation Type=::NewEnumerator1)
    MaxDistance=10000
    Module.bUseMaxDistance=true
    Module.bUseMinDistance=false
    Recalculate Duration Each Loop=true
    Scalability Mode=::NewEnumerator1
    UseLoopDelay=false
   )
   &SpawnBurst_Instantaneous(Spawn Count=1, Use Spawn Probability=false)
  ]
  particleSpawnScript [
   /Niagara/Modules/Spawn/Initialization/InitializeParticle(
    Lifetime=&RandomRangeFloat(Minimum=2.7, Maximum=3.2, Evaluation Type=::NewEnumerator1)
    Module.Write Color=true
    Module.Write Lifetime=true
    Module.Write Mass=true
    Module.Write Material Random=true
    Module.Write Position=false
    Module.Write Scale=false
    Module.Write SpriteRotation=false
    Module.Write SpriteSize=true
    Sprite Size=<500 150>
   ) {enabled=false}
   /Niagara/Modules/Spawn/Initialization/V2/InitializeParticle(
    Lifetime=&RandomRangeFloat(Minimum=2.7, Maximum=3.2)
    Mass Mode=::NewEnumerator1
    Sprite Size=<800 250>
    Sprite Size Mode=::NewEnumerator4
    UsePositionOffset=false
    Write Material Random=true
   ) {module=InitializeParticle001}
   &Color(Color=&MakeLinearColorFromVectorAndFloat("Vector (RGB)"=<0.2453 0.9334 1.25>, "Float (Alpha)"=&ClampFloat(Float=1)))
  ]
  particleUpdateScript [
   &ParticleState()
   &ScaleColor(
    Scale Alpha=&FloatFromCurve(FloatCurve=DataInterface() {propertyValues {Curve {keys [
         <at 0 0 RCIM_Cubic>
         {arriveTangent=1.3333, interpMode=RCIM_Cubic, leaveTangent=1.3333, time=0.25, value=1}
         {arriveTangent=-1.3333, interpMode=RCIM_Cubic, leaveTangent=-1.3333, time=0.75, value=1}
         <at 1 0>
        ]}}})
    ScaleA=true
    ScaleRGB=true
   )
   &CameraOffset(Camera Offset Amount=100)
  ]
  rendererClasses [&NiagaraSpriteRendererProperties]
  renderers [{rendererClass=&NiagaraSpriteRendererProperties}]
 }
 Shimmers {
  emitterUpdateScript [
   &EmitterState(
    Enable Distance Culling=true
    Life Cycle Mode=::NewEnumerator1
    Loop Delay=$User.InitialStartDelay
    Loop Duration=1
    MaxDistance=10000
    Module.bUseMaxDistance=true
    Module.bUseMinDistance=false
    Scalability Mode=::NewEnumerator1
    UseLoopDelay=false
   )
   &SpawnRate(SpawnRate=8, Use Spawn Probability=false)
  ]
  particleSpawnScript [
   /SetVariables(Particles.SpriteRotationRate=&RandomRangeFloat(Minimum=-0.03, Maximum=0.03, Evaluation Type=::NewEnumerator1)) {isSetParameters=true}
   /Niagara/Modules/Spawn/Initialization/InitializeParticle(
    Lifetime=&RandomRangeFloat(Minimum=3.7, Maximum=4.3, Evaluation Type=::NewEnumerator1)
    Module.Write Color=true
    Module.Write Lifetime=true
    Module.Write Mass=true
    Module.Write Material Random=true
    Module.Write Position=false
    Module.Write Scale=false
    Module.Write SpriteRotation=true
    Module.Write SpriteSize=true
    Sprite Rotation=&RandomRangeFloat(Minimum=-360, Maximum=360, Evaluation Type=::NewEnumerator1)
    Sprite Size=&RandomRangeVector2D(Minimum=<80 80>, Maximum=<120 120>, Evaluation Type=::NewEnumerator1)
   ) {enabled=false}
   /Niagara/Modules/Spawn/Initialization/V2/InitializeParticle(
    Lifetime=&RandomRangeFloat(Minimum=3.7, Maximum=4.3)
    Mass Mode=::NewEnumerator1
    Sprite Rotation Angle=&RandomRangeFloat(Minimum=-360, Maximum=360)
    Sprite Rotation Mode=::NewEnumerator0
    Sprite Size=&RandomRangeVector2D(Minimum=<150 150>, Maximum=<250 250>)
    Sprite Size Mode=::NewEnumerator4
    UsePositionOffset=false
    Write Material Random=true
   ) {module=InitializeParticle001}
   &Color(Color=&MakeLinearColorFromVectorAndFloat(
     "Vector (RGB)"=&RandomRangeVector(Minimum=<0.3589 0.3009 0.4531>, Maximum=<0.605 0.6866 0.7396>, Evaluation Type=::NewEnumerator1)
     "Float (Alpha)"=&ClampFloat(Float=&RandomRangeFloat(Minimum=0.2, Maximum=0.3, Evaluation Type=::NewEnumerator1))
    ))
   &SphereLocation(
    MaskSpawnGroup=false
    Module.Override Randomness=false
    Non Uniform Scale=&VectorFromFloat(Value=&FloatFromCurve(FloatCurve=DataInterface() {propertyValues {Curve {keys [<at 0 1>,<at 10000 1.4>]}}}, CurveIndex=$Engine.Owner.LODDistance))
    Sphere Radius=&RandomRangeFloat(Minimum=100, Maximum=150, Evaluation Type=::NewEnumerator1)
    Surface Only=true
   )
   &AddVelocity(Velocity=&Multiply_VectorByFloat(
     Vector=$Output.SphereLocation.SphereVector
     Float=&RandomRangeFloat(Minimum=-0.04, Maximum=-0.02, Evaluation Type=::NewEnumerator1)
    ))
   &DynamicMaterialParameters(
    Index 0 Param 1=&RandomRangeFloat(Minimum=0, Maximum=1, Evaluation Type=::NewEnumerator1)
    Index 0 Param 2=&RandomRangeFloat(Minimum=0, Maximum=1, Evaluation Type=::NewEnumerator1)
    Index 0 Param 3=1
    Module.Param0WriteEnabled=true
    Module.Param1WriteEnabled=true
    Module.Param2WriteEnabled=true
    Module.Param3WriteEnabled=true
   )
   &CameraOffset(Camera Offset Amount=&RandomRangeFloat(Minimum=50, Maximum=100, Evaluation Type=::NewEnumerator1))
  ]
  particleUpdateScript [
   &ParticleState()
   &SolveForcesAndVelocity(Clamp Velocity=false, Limit Acceleration=false)
   &DynamicMaterialParameters(
    Index 0 Param 3=1
    Index 0 Param 4=1
    Module.Param0WriteEnabled=true
    Module.Param1WriteEnabled=true
    Module.Param2WriteEnabled=true
    Module.Param3WriteEnabled=true
   ) {module=DynamicMaterialParameters001}
   &SpriteRotationRate(Rotation Rate=$Particles.SpriteRotationRate)
   &ScaleColor(
    Scale Alpha=&FloatFromCurve(FloatCurve=DataInterface() {propertyValues {Curve {keys [<at 0 0 RCIM_Cubic>,<at 0.5 1 RCIM_Cubic>,<at 1 0>]}}})
    ScaleA=true
    ScaleRGB=true
   )
  ]
  rendererClasses [&NiagaraSpriteRendererProperties]
  renderers [{rendererClass=&NiagaraSpriteRendererProperties}]
 }
 Source {
  emitterUpdateScript [
   &EmitterState(
    Enable Distance Culling=true
    Life Cycle Mode=::NewEnumerator1
    Loop Delay=0.25
    Loop Duration=&RandomRangeFloat(Minimum=1, Maximum=2.7)
    MaxDistance=10000
    Module.bUseMaxDistance=true
    Module.bUseMinDistance=false
    Recalculate Duration Each Loop=true
    Scalability Mode=::NewEnumerator1
    UseLoopDelay=false
   )
   &SpawnBurst_Instantaneous(Spawn Count=1, Use Spawn Probability=false)
  ]
  particleSpawnScript [
   /Niagara/Modules/Spawn/Initialization/InitializeParticle(
    Color=<rgba 0.2453 0.9334 1.25 1>
    Lifetime=&RandomRangeFloat(Minimum=2.7, Maximum=3.2, Evaluation Type=::NewEnumerator1)
    Module.Write Color=true
    Module.Write Lifetime=true
    Module.Write Mass=true
    Module.Write Material Random=true
    Module.Write Position=false
    Module.Write Scale=false
    Module.Write SpriteRotation=false
    Module.Write SpriteSize=true
    Sprite Size=<120 120>
   ) {enabled=false}
   /Niagara/Modules/Spawn/Initialization/V2/InitializeParticle(
    Color=<rgba 0 3.9613 5 1>
    Lifetime=&RandomRangeFloat(Minimum=2.7, Maximum=3.2)
    Mass Mode=::NewEnumerator1
    Sprite Size Mode=::NewEnumerator3
    Uniform Sprite Size=75
    UsePositionOffset=false
    Write Material Random=true
   ) {module=InitializeParticle001}
   &DynamicMaterialParameters(
    Index 0 Param 1=&RandomRangeFloat(Minimum=0.016, Maximum=0.5, Evaluation Type=::NewEnumerator1)
    Index 0 Param 2=1
    Index 0 Param 3=1
    Index 0 Param 4=1
    Module.Param0WriteEnabled=true
    Module.Param1WriteEnabled=true
    Module.Param2WriteEnabled=true
    Module.Param3WriteEnabled=true
   )
  ]
  particleUpdateScript [
   &ParticleState()
   &ScaleColor(
    Scale Alpha=&FloatFromCurve(FloatCurve=DataInterface() {propertyValues {Curve {keys [<at 0 0 RCIM_Cubic>, {tangentMode=RCTM_User, time=1, value=1}]}}})
    ScaleA=true
    ScaleRGB=true
   )
   &DynamicMaterialParameters(
    Index 0 Param 1=&RandomRangeFloat(Minimum=0.016, Maximum=0.5)
    Index 0 Param 2=1
    Index 0 Param 3=1
    Index 0 Param 4=1
    Module.Param0WriteEnabled=true
    Module.Param1WriteEnabled=true
    Module.Param2WriteEnabled=true
    Module.Param3WriteEnabled=true
   ) {module=DynamicMaterialParameters001}
   &CameraOffset(Camera Offset Amount=100)
   &ScaleSpriteSize(Scale Factor=&Vector2DFromFloat(Value=&FloatFromCurve(FloatCurve=DataInterface() {propertyValues {Curve {keys [<at 0 1>,<at 10000 3>]}}}, CurveIndex=$Engine.Owner.LODDistance)))
  ]
  rendererClasses [&NiagaraSpriteRendererProperties]
  renderers [{rendererClass=&NiagaraSpriteRendererProperties}]
 }
 Swirl_Flicker_Lightning {
  bEnabled=false
  emitterUpdateScript [
   &EmitterState(
    Life Cycle Mode=::NewEnumerator1
    Loop Behavior=::NewEnumerator2
    Loop Count=5
    Loop Delay=&RandomRangeFloat(Minimum=0, Maximum=0.015)
    Loop Duration=0.45
    UseLoopDelay=false
   )
   &SpawnRate(Use Spawn Probability=true)
   &SpawnRate(Spawn Probability=0.9, Use Spawn Probability=true) {module=SpawnRate001}
   &SpawnBurst_Instantaneous(Spawn Count=3, Use Spawn Probability=false, UseLoopCountLimit=false) {enabled=false}
  ]
  particleSpawnScript [
   &InitializeParticle(
    Color Maximum=<rgba 0.44 0.9101 1 0.8>
    Color Minimum=<rgba 0.5926 1.4207 1.5 1>
    Color Mode=::NewEnumerator2
    Lifetime Min=0.85
    Lifetime Mode=::NewEnumerator1
    Sprite Rotation Mode=::NewEnumerator2
    Sprite Size Mode=::NewEnumerator3
    Uniform Sprite Size=700
    UsePositionOffset=false
    Write Material Random=true
   )
   /SetVariables() {isSetParameters=true, module=SetVariables_#1}
  ]
  particleUpdateScript [
   &ParticleState()
   &DynamicMaterialParameters(
    Index 0 Param 1=&FloatFromCurve(FloatCurve=DataInterface() {propertyValues {Curve {keys [
         {arriveTangent=-2.3838, interpMode=RCIM_Cubic, leaveTangent=-2.3838, tangentMode=RCTM_User, time=0, value=1}
         {arriveTangent=-0.1592, interpMode=RCIM_Cubic, leaveTangent=-0.1592, tangentMode=RCTM_User, time=0.5, value=0.1}
         <at 1 0>
        ]}}})
    Index 0 Param 2=&SelectFloatFromArray(
     Array Sampling Mode=::NewEnumerator0
     Direct Array Index=?
     Float Selection Array=DataInterface() {propertyValues {FloatData [0,1]}}
     Random Seed=0
    )
    Index 0 Param 3=&FloatFromCurve(FloatCurve=DataInterface() {propertyValues {Curve {keys [<at 0 1 RCIM_Cubic>, {arriveTangent=-3.5615, tangentMode=RCTM_User, time=0.45, value=0.05}, <at 0.6 0.05>, <at 0.65 0>]}}})
    Module.Index0Param0Write=true
    Module.Index0Param1Write=true
    Module.Index0Param2Write=true
    Module.Index0Param3Write=true
   )
   &CameraOffset(Camera Offset Amount=-1)
   &ScaleColor(
    Scale Alpha=&FloatFromCurve(FloatCurve=DataInterface() {propertyValues {Curve {keys [
         {arriveTangent=8.8235, interpMode=RCIM_Cubic, leaveTangent=8.8235, tangentMode=RCTM_User, time=0, value=0}
         {arriveTangent=1.88, leaveTangent=-3.102, tangentMode=RCTM_User, time=0.2, value=1}
         {arriveTangent=-0.3279, interpMode=RCIM_Cubic, leaveTangent=-0.3279, tangentMode=RCTM_User, time=0.35, value=1}
         {arriveTangent=-0.3008, interpMode=RCIM_Cubic, leaveTangent=-0.3008, tangentMode=RCTM_User, time=0.75, value=0}
        ]}}})
    ScaleA=true
    ScaleRGB=true
   )
  ]
  rendererClasses [&NiagaraSpriteRendererProperties]
  renderers [{rendererClass=&NiagaraSpriteRendererProperties}]
 }
 Swirl_Flicker_Lightning001 {
  bEnabled=false
  emitterUpdateScript [
   &EmitterState(Life Cycle Mode=::NewEnumerator1, Loop Behavior=::NewEnumerator1, Loop Delay=2, Loop Duration=0.5, UseLoopDelay=true)
   &SpawnRate(Spawn Probability=0.6, SpawnRate=7, Use Spawn Probability=false)
   &SpawnRate(Spawn Probability=0.5, Use Spawn Probability=true) {enabled=false, module=SpawnRate001}
   &SpawnBurst_Instantaneous(Spawn Count=3, Use Spawn Probability=false, UseLoopCountLimit=false) {enabled=false}
  ]
  particleSpawnScript [
   &InitializeParticle(
    Color Maximum=<rgba 0.44 0.9101 1 0.8>
    Color Minimum=<rgba 0.5926 1.4207 1.5 1>
    Color Mode=::NewEnumerator2
    Lifetime Min=0.85
    Lifetime Mode=::NewEnumerator1
    Sprite Rotation Mode=::NewEnumerator2
    Sprite Size Mode=::NewEnumerator3
    Uniform Sprite Size=700
    UsePositionOffset=false
    Write Material Random=true
   )
   /SetVariables() {isSetParameters=true, module=SetVariables_#1}
  ]
  particleUpdateScript [
   &ParticleState()
   &DynamicMaterialParameters(
    Index 0 Param 1=&FloatFromCurve(FloatCurve=DataInterface() {propertyValues {Curve {keys [
         {arriveTangent=-2.3838, interpMode=RCIM_Cubic, leaveTangent=-2.3838, tangentMode=RCTM_User, time=0, value=1}
         {arriveTangent=-0.1592, interpMode=RCIM_Cubic, leaveTangent=-0.1592, tangentMode=RCTM_User, time=0.5, value=0.1}
         <at 1 0>
        ]}}})
    Index 0 Param 2=&SelectFloatFromArray(
     Array Sampling Mode=::NewEnumerator0
     Direct Array Index=?
     Float Selection Array=DataInterface() {propertyValues {FloatData [0,1]}}
     Random Seed=0
    )
    Index 0 Param 3=&FloatFromCurve(FloatCurve=DataInterface() {propertyValues {Curve {keys [<at 0 1 RCIM_Cubic>, {arriveTangent=-3.5615, tangentMode=RCTM_User, time=0.45, value=0.05}, <at 0.6 0.05>, <at 0.65 0>]}}})
    Module.Index0Param0Write=true
    Module.Index0Param1Write=true
    Module.Index0Param2Write=true
    Module.Index0Param3Write=true
   )
   &CameraOffset(Camera Offset Amount=-1)
   &ScaleColor(
    Scale Alpha=&FloatFromCurve(FloatCurve=DataInterface() {propertyValues {Curve {keys [
         {arriveTangent=8.8235, interpMode=RCIM_Cubic, leaveTangent=8.8235, tangentMode=RCTM_User, time=0, value=0}
         {arriveTangent=1.88, leaveTangent=-3.102, tangentMode=RCTM_User, time=0.2, value=1}
         {arriveTangent=-0.3279, interpMode=RCIM_Cubic, leaveTangent=-0.3279, tangentMode=RCTM_User, time=0.35, value=1}
         {arriveTangent=-0.3008, interpMode=RCIM_Cubic, leaveTangent=-0.3008, tangentMode=RCTM_User, time=0.75, value=0}
        ]}}})
    ScaleA=true
    ScaleRGB=true
   )
  ]
  rendererClasses [&NiagaraSpriteRendererProperties]
  renderers [{rendererClass=&NiagaraSpriteRendererProperties}]
 }
 TinyFrags {
  emitterUpdateScript [
   &EmitterState(
    Delay First Loop Only=true
    Enable Distance Culling=true
    Life Cycle Mode=::NewEnumerator1
    Loop Delay=$User.InitialStartDelay
    Loop Duration=&RandomRangeFloat(Minimum=1, Maximum=2, Evaluation Type=::NewEnumerator1)
    MaxDistance=2500
    Module.bUseMaxDistance=true
    Module.bUseMinDistance=false
    Recalculate Duration Each Loop=true
    Scalability Mode=::NewEnumerator1
    UseLoopDelay=true
   )
   &SpawnRate(SpawnRate=2, Use Spawn Probability=false)
   &SpawnBurst_Instantaneous(Spawn Count=&RandomRangeInt(Minimum=1, Maximum=2), Use Spawn Probability=false)
   &SpawnBurst_Instantaneous(Spawn Count=&RandomRangeInt(Minimum=0, Maximum=2), Spawn Time=0.25, Use Spawn Probability=false) {module=SpawnBurst_Instantaneous001}
   &SpawnBurst_Instantaneous(Spawn Count=&RandomRangeInt(Minimum=0, Maximum=2), Spawn Time=1.4, Use Spawn Probability=false) {module=SpawnBurst_Instantaneous002}
  ]
  particleSpawnScript [
   /SetVariables(Particles.MeshOrientation=&MakeQuatFromAxisAngle(
     Axis=&RandomRangeVector(Minimum=<0 0 0>, Maximum=<1 1 1>, Evaluation Type=::NewEnumerator1)
     AngleInDegrees=90
    )) {isSetParameters=true}
   /SetVariables(Particles.SpawnRotationVector=&RandomRangeVector(Minimum=<-0.15 -0.15 -0.15>, Maximum=<0.15 0.15 0.15>, Evaluation Type=::NewEnumerator1)) {isSetParameters=true}
   /Niagara/Modules/Spawn/Initialization/InitializeParticle(
    Lifetime=&RandomRangeFloat(Minimum=5, Maximum=6, Evaluation Type=::NewEnumerator1)
    Mesh Scale=&RandomRangeVector(Minimum=<0.4 0.4 1>, Maximum=<0.7 0.7 1>, Evaluation Type=::NewEnumerator1)
    Module.Write Color=true
    Module.Write Lifetime=true
    Module.Write Mass=true
    Module.Write Material Random=true
    Module.Write Position=false
    Module.Write Scale=true
    Module.Write SpriteRotation=false
    Module.Write SpriteSize=false
   ) {enabled=false}
   /Niagara/Modules/Spawn/Initialization/V2/InitializeParticle(
    Lifetime=&RandomRangeFloat(Minimum=5, Maximum=6)
    Mass Mode=::NewEnumerator1
    Mesh Scale=&RandomRangeVector(Minimum=<0.4 0.4 1>, Maximum=<0.7 0.7 1>)
    Mesh Scale Mode=::NewEnumerator4
    Sprite Size Mode=::NewEnumerator4
    UsePositionOffset=false
    Write Material Random=true
   ) {module=InitializeParticle001}
   &Color(Color=&MakeLinearColorFromVectorAndFloat("Vector (RGB)"=<10 10 10>, "Float (Alpha)"=&ClampFloat(Float=1)))
   &AddVelocity(Velocity=&RandomRangeVector(Minimum=<-8 -8 -8>, Maximum=<8 8 8>, Evaluation Type=::NewEnumerator1))
   &SphereLocation(
    MaskSpawnGroup=false
    Module.Override Randomness=false
    Sphere Radius=&RandomRangeFloat(Minimum=70, Maximum=90, Evaluation Type=::NewEnumerator1)
    Surface Only=false
   )
   &AddVelocity(Velocity=&Multiply_VectorByFloat(
     Vector=$Output.SphereLocation.SphereVector
     Float=&RandomRangeFloat(Minimum=-0.1, Maximum=0.2, Evaluation Type=::NewEnumerator1)
    )) {module=AddVelocity001}
  ]
  particleUpdateScript [
   &ParticleState()
   &Drag(Drag=0.2, Use Linear Drag=true, Use Rotational Drag=false)
   &SolveForcesAndVelocity(Clamp Velocity=false, Limit Acceleration=false)
   &ScaleColor(
    Scale Alpha=&FloatFromCurve(FloatCurve=DataInterface() {propertyValues {Curve {keys [
         <at 0 0 RCIM_Cubic>
         {arriveTangent=1.4286, interpMode=RCIM_Cubic, leaveTangent=1.4286, time=0.2, value=1}
         {arriveTangent=-1.25, interpMode=RCIM_Cubic, leaveTangent=-1.25, time=0.7, value=1}
         <at 1 0>
        ]}}})
    ScaleA=true
    ScaleRGB=true
   )
   &UpdateMeshOrientation(Rotation Vector=$Particles.SpawnRotationVector) {module=UpdateMeshOrientation001}
   &ScaleMeshSize(Scale Factor=&VectorFromFloat(Value=&FloatFromCurve(FloatCurve=DataInterface() {propertyValues {Curve {keys [<at 0 0 RCIM_Cubic>, <at 0.2 1>, {arriveTangent=-1.25, interpMode=RCIM_Cubic, leaveTangent=-1.25, time=0.8, value=1}, <at 1 0>]}}})))
  ]
  rendererClasses [&NiagaraMeshRendererProperties]
  renderers [{rendererClass=&NiagaraMeshRendererProperties, rendererIndex=0}]
 }
}
summary {
 systemName=NS_Greenroom_Rift_Portal
 emitters [
  {emitterName=Portal_Smoke, rendererClasses [&NiagaraSpriteRendererProperties]}
  {emitterName=Portal_Blackout, rendererClasses [&NiagaraSpriteRendererProperties]}
  {bEnabled=false, emitterName=DarkGradient, rendererClasses [&NiagaraSpriteRendererProperties]}
  {bEnabled=false, emitterName=Crack, rendererClasses [&NiagaraMeshRendererProperties]}
  {emitterName=Shimmers, rendererClasses [&NiagaraSpriteRendererProperties]}
  {emitterName=Embers, simTarget=GPUComputeSim, rendererClasses [&NiagaraSpriteRendererProperties]}
  {emitterName=TinyFrags, rendererClasses [&NiagaraMeshRendererProperties]}
  {emitterName=Glints, rendererClasses [&NiagaraSpriteRendererProperties]}
  {
   emitterName=Light
   rendererClasses [
    &NiagaraSpriteRendererProperties
    &NiagaraLightRendererProperties
    &NiagaraComponentRendererProperties
   ]
  }
  {emitterName=RoundFlare, rendererClasses [&NiagaraSpriteRendererProperties]}
  {emitterName=SharpFlare, rendererClasses [&NiagaraSpriteRendererProperties]}
  {emitterName=Corona, rendererClasses [&NiagaraSpriteRendererProperties]}
  {
   emitterName=Shard_Mesh_Large
   rendererClasses [&NiagaraSpriteRendererProperties, &NiagaraMeshRendererProperties]
  }
  {bEnabled=false, emitterName=ElectricArcsResidual, rendererClasses [&NiagaraMeshRendererProperties]}
  {bEnabled=false, emitterName=Swirl_Flicker_Lightning, rendererClasses [&NiagaraSpriteRendererProperties]}
  {
   emitterName=Shard_Mesh_Small
   rendererClasses [&NiagaraSpriteRendererProperties, &NiagaraMeshRendererProperties]
  }
  {bEnabled=false, emitterName=Flash, rendererClasses [&NiagaraSpriteRendererProperties]}
  {bEnabled=false, emitterName=LightShaftsBurst, rendererClasses [&NiagaraMeshRendererProperties]}
  {emitterName=Source, rendererClasses [&NiagaraSpriteRendererProperties]}
  {bEnabled=false, emitterName=Swirl_Flicker_Lightning001, rendererClasses [&NiagaraSpriteRendererProperties]}
  {emitterName=Glints_Center, rendererClasses [&NiagaraSpriteRendererProperties]}
 ]
 userVariables [
  User.InitialStartDelay {type=NiagaraFloat, defaultValue=0, description=""}
  User.Open {type=NiagaraFloat, defaultValue=0, description=""}
  User.TriggerLargeShard {type=NiagaraBool, defaultValue=true}
 ]
}
