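# Niagara system for the player glow in the intro rift portal teleport: a glow that travels to an end position.
# A single Glow sprite lives for User.Lifetime, eases toward User.EndPosition, animates material parameters and generates location events.
# A ribbon emitter with disabled spawn rate, likely event-fed, draws a tapering cyan trail.
# A Fragments emitter, delayed by CrackDelay, throws spinning, dragged mesh and sprite shards that flash bright, then shrink and fade.
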NiagaraSystem /SparksEffects/Effects/Intro/Rift_Portal/Particles/NS_Rift_Teleport_PlayerGlow
dependencies {
 usedDataInterfaces [&NiagaraDataInterfaceCurve, &NiagaraDataInterfaceArrayFloat, &NiagaraDataInterfaceVectorCurve]
 usedDynamicInputs [&SelectFloatFromArray]
 usedModules [
  &self:SystemSpawnScript.SystemScriptSource.SystemScriptGraph.NiagaraNodeAssignment_0.SetVariables_#1
  &SystemState
 ]
}
emitters {
 Detail_Trail_Ribbon_Color {
  emitterUpdateScript [
   &EmitterState()
   &SpawnRate(SpawnRate=25, Use Spawn Probability=false) {enabled=false}
  ]
  particleSpawnScript [
   &InitializeRibbon(
    Color=<rgba 0 1.1298 1.25 0.35>
    Lifetime=&Add_Float(A=$User.Lifetime, B=-0.1)
    Ribbon Width=50
    UsePositionOffset=false
    Write Material Random=true
   )
   &InitializeParticle(Color=<rgba 0 1.24 1.25 0.25>, UsePositionOffset=false, Write Material Random=true) {enabled=false}
  ]
  particleUpdateScript [
   &ParticleState()
   &ScaleColor(
    Scale RGB=&VectorFromCurve(VectorCurve=DataInterface() {propertyValues {XCurve {keys [<at 0 1>,<at 1 0>]}, YCurve {keys [<at 0 1>,<at 1 0>]}, ZCurve {keys [<at 0 1>,<at 1 0>]}}})
    ScaleA=true
    ScaleRGB=true
   ) {enabled=false}
   &ScaleRibbonWidth(Ribbon Width Scale=&FloatFromCurve(FloatCurve=DataInterface() {propertyValues {Curve {keys [
         <at 0 1>
         {arriveTangent=-1, interpMode=RCIM_Cubic, leaveTangent=-1, time=0.8, value=1}
         {interpMode=RCIM_Cubic, tangentMode=RCTM_User, time=1, value=0}
        ]}}}))
  ]
  rendererClasses [&NiagaraRibbonRendererProperties, &NiagaraSpriteRendererProperties]
  renderers [
   {rendererClass=&NiagaraRibbonRendererProperties}
   {rendererClass=&NiagaraRibbonRendererProperties, rendererIndex=1}
   {rendererClass=&NiagaraSpriteRendererProperties, rendererIndex=2}
  ]
 }
 Fragments {
  emitterUpdateScript [&EmitterState(Life Cycle Mode=::NewEnumerator1, Loop Behavior=::NewEnumerator1, Loop Delay=$System.CrackDelay, UseLoopDelay=true)]
  particleSpawnScript [
   &FN_SetSpawnGroup()
   /SetVariables(Particles.VisToggle:NiagaraInt32=&MakeIntFromFloat(Float=&ClampFloat(Float=&RandomRangeFloat(Minimum=0, Maximum=3)))) {isSetParameters=true}
   &InitializeParticle(
    Color Maximum=<rgba 0.2523 3.4809 4 1>
    Color Minimum=<rgba 0.5645 0.706 3 1>
    Color Mode=::NewEnumerator2
    Lifetime Max=1.25
    Lifetime Min=0.5
    Lifetime Mode=::NewEnumerator1
    Mesh Scale Max=<0.25 0.5 0.5>
    Mesh Scale Min=<0.25 0.5 0.8>
    Mesh Scale Mode=::NewEnumerator2
    Sprite Size Mode=::NewEnumerator1
    Uniform Sprite Size Max=5
    Uniform Sprite Size Min=2
    UsePositionOffset=false
    Write Material Random=true
   )
   &AddVelocity(Module.Use Falloff Distance=false, Override Seed=false, Velocity Mode=::NewEnumerator1, Velocity Speed=100) {module=AddVelocity001}
   &SphereLocation(
    MaskSpawnGroup=false
    Non Uniform Scale=<0.25 1 1>
    Override Seed=false
    Sphere Radius=10
    Surface Only=false
    UseNonUniformScale=false
    UseOffset=false
    UseSphereOrientation=false
   )
   &AddVelocityFromPoint(
    Module.Use Falloff Distance=false
    Velocity Strength=&Multiply_Float(A=200, B=&RandomRangeFloat(Minimum=0.5, Maximum=1))
   )
   &InitialMeshOrientation(Orientation Vector=&RandomVector(Vector Scale=1), Use Orientation Vector=true, Use Rotation Vector=false)
   /SetVariables(Particles.MeshIndex=&RandomRangeInt(Minimum=0, Maximum=2)) {isSetParameters=true}
   &CameraOffset(Camera Offset Amount=-25)
  ]
  particleUpdateScript [
   &ParticleState()
   &UpdateMeshOrientation(
    Rotation Coordinate Space=::NewEnumerator3
    Rotation Rate=&FloatFromCurve(
     FloatCurve=DataInterface() {propertyValues {Curve {keys [
         {arriveTangent=-2.5411, interpMode=RCIM_Cubic, leaveTangent=-2.5411, tangentMode=RCTM_User, time=0, value=1}
         {arriveTangent=-0.173, interpMode=RCIM_Cubic, leaveTangent=-0.173, tangentMode=RCTM_User, time=1, value=0.1}
        ]}}}
     Scale Curve=&FloatFromCurve(
      FloatCurve=DataInterface() {propertyValues {Curve {keys [
          {interpMode=RCIM_Cubic, tangentMode=RCTM_User, time=0, value=-1}
          {arriveTangent=9.6958, interpMode=RCIM_Cubic, leaveTangent=9.6958, tangentMode=RCTM_User, time=0.5, value=0}
          {interpMode=RCIM_Cubic, tangentMode=RCTM_User, time=1, value=1}
         ]}}}
      CurveIndex=&RandomRangeFloat(Minimum=0, Maximum=1)
      Scale Curve=2
     )
    )
    Rotation Vector=&RandomRangeVector(Minimum=<0 1 0>, Maximum=<1 0 0>)
   )
   &Drag(Drag=5, Use Linear Drag=true, Use Rotational Drag=false)
   &SolveForcesAndVelocity(Clamp Velocity=false, Limit Acceleration=false)
   &ScaleColor(
    Scale Alpha=&FloatFromCurve(FloatCurve=DataInterface() {propertyValues {Curve {keys [
         {arriveTangent=-0.0405, interpMode=RCIM_Cubic, leaveTangent=-0.0405, tangentMode=RCTM_User, time=0.6, value=1}
         {interpMode=RCIM_Cubic, leaveTangent=-4.3194, time=1, value=0}
        ]}}})
    Scale RGB=&VectorFromFloat(Value=&Add_Float(
      A=1
      B=&FloatFromCurve(
       FloatCurve=DataInterface() {propertyValues {Curve {keys [
           {arriveTangent=-5.4733, interpMode=RCIM_Cubic, leaveTangent=-5.4733, tangentMode=RCTM_User, time=0, value=1}
           {interpMode=RCIM_Cubic, tangentMode=RCTM_User, time=0.5, value=0}
          ]}}}
       Scale Curve=20
      )
     ))
    ScaleA=true
    ScaleRGB=true
   )
   /SetVariables(Particles.MeshColor:LinearColor=&MakeLinearColorFromVectorAndFloat(
     "Vector (RGB)"=&VectorFromFloat(Value=&Add_Float(
       A=5
       B=&FloatFromCurve(
        FloatCurve=DataInterface() {propertyValues {Curve {keys [
            {arriveTangent=-3.7813, interpMode=RCIM_Cubic, leaveTangent=-3.7813, tangentMode=RCTM_User, time=0, value=1}
            {interpMode=RCIM_Cubic, tangentMode=RCTM_User, time=0.7, value=0}
           ]}}}
        Scale Curve=20
       )
      ))
     "Float (Alpha)"=&FloatFromCurve(
      FloatCurve=DataInterface() {propertyValues {Curve {keys [<at 0.7 1 RCIM_Cubic>, {arriveTangent=-5.6246, interpMode=RCIM_Cubic, leaveTangent=-5.6246, tangentMode=RCTM_User, time=1, value=0}]}}}
      Scale Curve=2.5
     )
    )) {isSetParameters=true}
   &ScaleMeshSize(Scale Factor=&VectorFromFloat(Value=&FloatFromCurve(FloatCurve=DataInterface() {propertyValues {Curve {keys [
          {arriveTangent=-0.0405, interpMode=RCIM_Cubic, leaveTangent=-0.0405, tangentMode=RCTM_User, time=0.6, value=1}
          {arriveTangent=-5.4415, interpMode=RCIM_Cubic, leaveTangent=-5.4415, tangentMode=RCTM_User, time=1, value=0}
         ]}}})))
  ]
  rendererClasses [&NiagaraMeshRendererProperties, &NiagaraSpriteRendererProperties]
  renderers [
   {rendererClass=&NiagaraMeshRendererProperties, rendererIndex=0}
   {rendererClass=&NiagaraSpriteRendererProperties, rendererIndex=1}
  ]
 }
 Glow {
  emitterUpdateScript [
   &EmitterState()
   &SpawnBurst_Instantaneous(Spawn Count=1, Use Spawn Probability=false)
  ]
  particleSpawnScript [
   &InitializeParticle(
    Color=<rgba 0.125 4.1062 8 1>
    Lifetime=$User.Lifetime
    Sprite Size Mode=::NewEnumerator3
    Uniform Sprite Size=150
    UsePositionOffset=false
    Write Material Random=true
   )
   /SetVariables(
    Particles.DynParams2:Vector4f=&MakeVector4(X=0.5, Y=1, Z=1, W=1)
    Particles.DynParams3:Vector4f=<1000 0 0 0>
   ) {isSetParameters=true}
   &CameraOffset(Camera Offset Amount=500)
  ]
  particleUpdateScript [
   &DynamicMaterialParameters(
    Index 0 Param 1=0.8
    Index 0 Param 2=&Lerp_Float(
     A=0.05
     B=1
     Alpha=&FloatFromCurve(FloatCurve=DataInterface() {propertyValues {Curve {keys [
          {arriveTangent=2.3085, interpMode=RCIM_Cubic, leaveTangent=2.3085, tangentMode=RCTM_User, time=0, value=0}
          {arriveTangent=1, interpMode=RCIM_Cubic, leaveTangent=1, time=0.15, value=0.55}
          {arriveTangentWeight=0.132, interpMode=RCIM_Cubic, leaveTangentWeight=0.1886, tangentMode=RCTM_User, time=0.5, value=0.5}
          {arriveTangent=3.3782, interpMode=RCIM_Cubic, leaveTangent=3.3782, tangentMode=RCTM_User, time=1, value=1}
         ]}}})
    )
    Module.Param0WriteEnabled=true
    Module.Param1WriteEnabled=true
    Module.Param2WriteEnabled=false
    Module.Param3WriteEnabled=false
   )
   &ParticleState()
   &ScaleSpriteSize(Uniform Curve Sprite Scale=DataInterface() {propertyValues {Curve {keys [
        <at 0 0.1 RCIM_Cubic>
        {arriveTangent=4.5, interpMode=RCIM_Cubic, leaveTangent=4.5, time=0.19, value=0.45}
        {arriveTangent=-2.0962, interpMode=RCIM_Cubic, leaveTangent=-2.0962, tangentMode=RCTM_User, time=0.2, value=1}
        {arriveTangent=-1.6667, interpMode=RCIM_Cubic, leaveTangent=-1.6667, time=0.4, value=0.6}
        {arriveTangent=-0.0822, interpMode=RCIM_Cubic, leaveTangent=-0.0822, tangentMode=RCTM_User, time=0.5, value=0.5}
        {arriveTangent=3.0844, time=0.8, value=1}
        {arriveTangent=4.7412, interpMode=RCIM_Cubic, leaveTangent=-5.2133, tangentMode=RCTM_Break, time=0.925, value=1}
        <at 1 0>
       ]}}}) {enabled=false}
   &ScaleColor(
    Scale Alpha=&FloatFromCurve(FloatCurve=DataInterface() {propertyValues {Curve {keys [
         {interpMode=RCIM_Cubic, tangentMode=RCTM_User, time=0, value=0}
         <at 0.05 1>
         <at 0.5 1>
         {arriveTangent=-2, interpMode=RCIM_Cubic, leaveTangent=-2, time=0.9, value=1}
         {tangentMode=RCTM_User, time=1, value=0}
        ]}}})
    ScaleA=true
    ScaleRGB=false
   )
   /Niagara/Modules/Update/Utility/LerpParticleAttributes(
    Lerp Alpha:NiagaraFloat=&FloatFromCurve(FloatCurve=DataInterface() {propertyValues {Curve {keys [
         {tangentMode=RCTM_User, time=0, value=0}
         {arriveTangent=0.45, interpMode=RCIM_Cubic, leaveTangent=0.45, tangentMode=RCTM_User, time=0.2, value=0.05}
         {arriveTangent=2.4919, interpMode=RCIM_Cubic, leaveTangent=2.4919, tangentMode=RCTM_User, time=0.9, value=0.65}
         {tangentMode=RCTM_User, time=1, value=1}
        ]}}})
    Lerp Position=true
    Position B:NiagaraPosition=$User.EndPosition
   )
   &CurlNoiseForce(Noise Frequency=0.5, Noise Strength=5000, Pan Noise=false) {enabled=false}
   &SolveForcesAndVelocity(Clamp Velocity=false, Limit Acceleration=false)
   &GenerateLocationEvent(Use Event Probability=false, UseEventDelay=false)
  ]
  rendererClasses [&NiagaraSpriteRendererProperties]
  renderers [
   {rendererClass=&NiagaraSpriteRendererProperties}
   {rendererClass=&NiagaraSpriteRendererProperties, rendererIndex=1}
  ]
 }
}
summary {
 systemName=NS_Rift_Teleport_PlayerGlow
 emitters [
  {emitterName=Glow, rendererClasses [&NiagaraSpriteRendererProperties]}
  {emitterName=Fragments, rendererClasses [&NiagaraMeshRendererProperties, &NiagaraSpriteRendererProperties]}
  {
   emitterName=Detail_Trail_Ribbon_Color
   rendererClasses [&NiagaraRibbonRendererProperties, &NiagaraSpriteRendererProperties]
  }
 ]
 userVariables [
  User.EndPosition {type=NiagaraPosition, description="", defaultValue {}}
  User.FireworkTertiaryColor {type=LinearColor, description="", defaultValue {struct=LinearColor, value=<rgba 1 1 1 1>}}
  User.Lifetime {type=NiagaraFloat, defaultValue=0.3, description=""}
 ]
}
