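# Niagara system for the intro rift teleport portal: a timed cyan-blue burst where a crack opens, glows and ends in a flash.
# A shared System.CrackDelay times the crack mesh, fragments, glints, glow burst and light shafts, while a glow builds and lights add illumination.
# A fixed later delay fires the closing glow burst and curl-noise residue dust.
# Curve-driven dynamic material parameters animate the crack, glows and shafts over particle life.
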NiagaraSystem /SparksEffects/Effects/Intro/Rift_Portal/Particles/NS_Rift_Teleport_Portal
dependencies {
 usedDataInterfaces [
  &NiagaraDataInterfaceCurve
  &NiagaraDataInterfaceMeshRendererInfo
  &NiagaraDataInterfaceVectorCurve
  &NiagaraDataInterfaceArrayFloat
 ]
 usedDynamicInputs []
 usedModules [
  &self:SystemSpawnScript.SystemScriptSource.SystemScriptGraph.NiagaraNodeAssignment_0.SetVariables_#1
  &SystemState
 ]
}
emitters {
 BurstEndResiduleDust {
  emitterUpdateScript [
   &EmitterState(Life Cycle Mode=::NewEnumerator1, Loop Behavior=::NewEnumerator1, Loop Delay=5.3, UseLoopDelay=true)
   &SpawnBurst_Instantaneous(Spawn Count=100, Use Spawn Probability=false, UseLoopCountLimit=false)
   &SpawnBurst_Instantaneous(Spawn Count=120, Spawn Time=0.06, Use Spawn Probability=false, UseLoopCountLimit=false) {module=SpawnBurst_Instantaneous001}
   &SpawnBurst_Instantaneous(Spawn Count=50, Spawn Time=0.15, Use Spawn Probability=false, UseLoopCountLimit=false) {module=SpawnBurst_Instantaneous002}
  ]
  particleSpawnScript [
   &FN_SetSpawnGroup()
   /SetVariables(Particles.VisToggle:NiagaraInt32=&MakeIntFromFloat(Float=&ClampFloat(Float=&RandomRangeFloat(Minimum=0, Maximum=3)))) {isSetParameters=true, module=SetVariables_#2}
   &InitializeParticle(
    Color Maximum=<rgba 0.2523 3.4809 4 1>
    Color Minimum=<rgba 0.5645 0.706 3 1>
    Color Mode=::NewEnumerator2
    Lifetime=&Multiply_Float(
     A=&SelectFloatFromArray(Direct Array Index=$Particles.VisToggle, Float Selection Array=DataInterface() {propertyValues {FloatData [3,4]}})
     B=&RandomRangeFloat(Minimum=0.25, Maximum=0.5)
    )
    Mesh Scale Max=<7 7 5>
    Mesh Scale Min=<4 4 5>
    Mesh Scale Mode=::NewEnumerator2
    Sprite Size Mode=::NewEnumerator1
    Uniform Sprite Size Max=25
    UsePositionOffset=false
    Write Material Random=true
   )
   &SphereLocation(
    MaskSpawnGroup=false
    Non Uniform Scale=<0.25 1 1>
    Override Seed=false
    Sphere Radius=200
    Surface Only=false
    UseNonUniformScale=true
    UseOffset=false
    UseSphereOrientation=false
   )
   &AddVelocityFromPoint(
    Module.Use Falloff Distance=false
    Velocity Strength=&Multiply_Float(A=1250, B=&RandomRangeFloat(Minimum=0.5, Maximum=1))
   )
   &AddVelocity(Velocity=&Multiply_VectorByFloat(Vector=<1 0 0>, Float=&RandomRangeFloat(Minimum=-500, Maximum=500)))
   &InitialMeshOrientation(Orientation Vector=&RandomVector(Vector Scale=1), Use Orientation Vector=true, Use Rotation Vector=false)
   /SetVariables(Particles.MeshIndex=&RandomRangeInt(Minimum=0, Maximum=2)) {isSetParameters=true, module=SetVariables_#3}
   &CameraOffset(Camera Offset Amount=-25)
  ]
  particleUpdateScript [
   &ParticleState()
   &UpdateMeshOrientation(
    Rotation Coordinate Space=::NewEnumerator3
    Rotation Rate=&FloatFromCurve(
     FloatCurve=DataInterface() {propertyValues {Curve {keys [
         {arriveTangent=-2.5411, interpMode=RCIM_Cubic, leaveTangent=-2.5411, tangentMode=RCTM_User, time=0, value=1}
         {arriveTangent=-0.173, interpMode=RCIM_Cubic, leaveTangent=-0.173, tangentMode=RCTM_User, time=1, value=0.1}
        ]}}}
     Scale Curve=&FloatFromCurve(
      FloatCurve=DataInterface() {propertyValues {Curve {keys [
          {interpMode=RCIM_Cubic, tangentMode=RCTM_User, time=0, value=-1}
          {arriveTangent=9.6958, interpMode=RCIM_Cubic, leaveTangent=9.6958, tangentMode=RCTM_User, time=0.5, value=0}
          {interpMode=RCIM_Cubic, tangentMode=RCTM_User, time=1, value=1}
         ]}}}
      CurveIndex=&RandomRangeFloat(Minimum=0, Maximum=1)
      Scale Curve=2
     )
    )
    Rotation Vector=&RandomRangeVector(Minimum=<0 1 0>, Maximum=<1 0 0>)
   )
   &Drag(
    Drag=&SelectFloatFromArray(Direct Array Index=$Particles.VisToggle, Float Selection Array=DataInterface() {propertyValues {FloatData [1.5,2]}})
    Use Linear Drag=true
    Use Rotational Drag=false
   )
   &VortexForce(UseOriginOffset=false, Vortex Force Amount=1500) {enabled=false}
   &CurlNoiseForce(Noise Frequency=0.5, Noise Strength=500, Pan Noise=true, Pan Noise Field=<0 0 0.25>)
   &SolveForcesAndVelocity(Clamp Velocity=false, Limit Acceleration=false)
   &ScaleColor(
    Scale Alpha=&FloatFromCurve(FloatCurve=DataInterface() {propertyValues {Curve {keys [
         {arriveTangent=-0.0405, interpMode=RCIM_Cubic, leaveTangent=-0.0405, tangentMode=RCTM_User, time=0.6, value=1}
         {interpMode=RCIM_Cubic, leaveTangent=-4.3194, time=1, value=0}
        ]}}})
    Scale RGB=&VectorFromFloat(Value=&Add_Float(
      A=1
      B=&FloatFromCurve(
       FloatCurve=DataInterface() {propertyValues {Curve {keys [
           {arriveTangent=-5.4733, interpMode=RCIM_Cubic, leaveTangent=-5.4733, tangentMode=RCTM_User, time=0, value=1}
           {interpMode=RCIM_Cubic, tangentMode=RCTM_User, time=0.5, value=0}
          ]}}}
       Scale Curve=20
      )
     ))
    ScaleA=true
    ScaleRGB=true
   )
   /SetVariables(Particles.MeshColor:LinearColor=&MakeLinearColorFromVectorAndFloat(
     "Vector (RGB)"=&VectorFromFloat(Value=&Add_Float(
       A=5
       B=&FloatFromCurve(
        FloatCurve=DataInterface() {propertyValues {Curve {keys [
            {arriveTangent=-3.7813, interpMode=RCIM_Cubic, leaveTangent=-3.7813, tangentMode=RCTM_User, time=0, value=1}
            {interpMode=RCIM_Cubic, tangentMode=RCTM_User, time=0.7, value=0}
           ]}}}
        Scale Curve=20
       )
      ))
     "Float (Alpha)"=&FloatFromCurve(
      FloatCurve=DataInterface() {propertyValues {Curve {keys [<at 0.7 1 RCIM_Cubic>, {arriveTangent=-5.6246, interpMode=RCIM_Cubic, leaveTangent=-5.6246, tangentMode=RCTM_User, time=1, value=0}]}}}
      Scale Curve=2.5
     )
    )) {isSetParameters=true, module=SetVariables_#4}
   &ScaleMeshSize(Scale Factor=&VectorFromFloat(Value=&FloatFromCurve(FloatCurve=DataInterface() {propertyValues {Curve {keys [
          {arriveTangent=-0.0405, interpMode=RCIM_Cubic, leaveTangent=-0.0405, tangentMode=RCTM_User, time=0.6, value=1}
          {arriveTangent=-5.4415, interpMode=RCIM_Cubic, leaveTangent=-5.4415, tangentMode=RCTM_User, time=1, value=0}
         ]}}})))
  ]
  rendererClasses [&NiagaraMeshRendererProperties, &NiagaraSpriteRendererProperties]
  renderers [
   {rendererClass=&NiagaraMeshRendererProperties, rendererIndex=0}
   {rendererClass=&NiagaraSpriteRendererProperties, rendererIndex=1}
  ]
 }
 Crack {
  emitterUpdateScript [
   &EmitterState(Life Cycle Mode=::NewEnumerator1, Loop Behavior=::NewEnumerator1, Loop Delay=$System.CrackDelay, UseLoopDelay=true)
   &SpawnBurst_Instantaneous(Use Spawn Probability=false, UseLoopCountLimit=false)
  ]
  particleSpawnScript [
   &InitializeParticle(
    Lifetime=4.3
    Mesh Scale Mode=::NewEnumerator3
    Mesh Uniform Scale=10
    Position Offset=<-200 0 0>
    UsePositionOffset=true
    Write Material Random=true
   ) {module=InitializeParticle001}
   &InitialMeshOrientation(Orientation Vector=<0 0 -1>, Use Orientation Vector=true, Use Rotation Vector=false)
   &CameraOffset(Camera Offset Amount=-250)
  ]
  particleUpdateScript [
   &ParticleState()
   &ScaleMeshSize(Scale Factor=&VectorFromFloat(Value=&FloatFromCurve(FloatCurve=DataInterface() {propertyValues {Curve {keys [
          <at 0 1>
          {arriveTangent=-1, interpMode=RCIM_Cubic, leaveTangent=-1, time=0.85, value=1}
          {arriveTangent=-10.8078, tangentMode=RCTM_User, time=1, value=0}
         ]}}}))) {enabled=false}
   &DynamicMaterialParameters(
    Index 0 Param 1=&FloatFromCurve(FloatCurve=DataInterface() {propertyValues {Curve {keys [
         {arriveTangent=10.8966, interpMode=RCIM_Cubic, time=0, value=0}
         {arriveTangent=0.3084, leaveTangent=0.3084, time=0.0671, value=0.3673}
         {arriveTangent=4.7601, interpMode=RCIM_Cubic, leaveTangent=4.7601, time=0.1489, value=0.6757}
         {arriveTangent=-0.0034, leaveTangent=-0.0034, time=0.2, value=1}
         {arriveTangent=-0.3471, leaveTangent=-0.3471, time=0.6021, value=0.9966}
         {arriveTangent=-0.0069, leaveTangent=-0.0069, time=0.6417, value=0.6711}
         {arriveTangent=-0.2804, leaveTangent=-0.2804, time=0.7336, value=0.6427}
         {arriveTangent=-0.0148, leaveTangent=-0.0148, time=0.7362, value=0.3838}
         {arriveTangent=-0.1832, leaveTangent=-0.1832, time=0.8231, value=0.3475}
         {arriveTangent=0.0012, leaveTangent=0.0012, time=0.8328, value=0.1805}
         {arriveTangent=-0.1656, leaveTangent=-0.1656, time=0.8946, value=0.1656}
         <at 1 0 RCIM_Cubic>
        ]}}})
    Index 0 Param 2=&FloatFromCurve(
     FloatCurve=DataInterface() {propertyValues {Curve {keys [
         <at 0.4 6 RCIM_Cubic>
         {arriveTangent=-9.9167, interpMode=RCIM_Cubic, leaveTangent=-9.9167, time=0.8876, value=4.0583}
         {interpMode=RCIM_Cubic, leaveTangent=-15.662, time=1, value=0.05}
        ]}}}
     Scale Curve=0.025
    )
    Index 0 Param 3=&FloatFromCurve(
     FloatCurve=DataInterface() {propertyValues {Curve {keys [
         {arriveTangent=2.0462, interpMode=RCIM_Cubic, time=0, value=0.15}
         {arriveTangent=2.8333, interpMode=RCIM_Cubic, leaveTangent=2.8333, time=0.0547, value=0.875}
         {arriveTangent=0.1893, interpMode=RCIM_Cubic, leaveTangent=0.1893, time=0.3, value=1}
         {arriveTangent=-0.1483, interpMode=RCIM_Cubic, leaveTangent=-0.1483, time=0.7149, value=1}
         {arriveTangent=-1.0604, interpMode=RCIM_Cubic, leaveTangent=-1.0604, time=0.7197, value=0.9377}
         {arriveTangent=-1.4069, interpMode=RCIM_Cubic, leaveTangent=-1.4069, time=0.8014, value=0.9083}
         {arriveTangent=-3.0576, interpMode=RCIM_Cubic, leaveTangent=-3.0576, time=0.8054, value=0.8172}
         {arriveTangent=-2.7631, interpMode=RCIM_Cubic, leaveTangent=-2.7631, time=0.8715, value=0.6939}
         {arriveTangent=-4.5478, interpMode=RCIM_Cubic, leaveTangent=-4.5478, time=0.8793, value=0.6131}
         {arriveTangent=-5.0784, interpMode=RCIM_Cubic, leaveTangent=-5.0784, time=0.9471, value=0.35}
         <at 1 0 RCIM_Cubic>
        ]}}}
     Scale Curve=0.05
    )
    Index 0 Param 4=&Lerp_Float(
     A=0
     B=0.6
     Alpha=&FloatFromCurve(FloatCurve=DataInterface() {propertyValues {Curve {keys [
          {arriveTangent=2.0462, interpMode=RCIM_Cubic, time=0, value=0.15}
          {arriveTangent=2.8333, interpMode=RCIM_Cubic, leaveTangent=2.8333, time=0.0547, value=0.875}
          {arriveTangent=0.1893, interpMode=RCIM_Cubic, leaveTangent=0.1893, time=0.3, value=1}
          {arriveTangent=-0.1483, interpMode=RCIM_Cubic, leaveTangent=-0.1483, time=0.7149, value=1}
          {arriveTangent=-1.0604, interpMode=RCIM_Cubic, leaveTangent=-1.0604, time=0.7197, value=0.9377}
          {arriveTangent=-1.4069, interpMode=RCIM_Cubic, leaveTangent=-1.4069, time=0.8014, value=0.9083}
          {arriveTangent=-3.0576, interpMode=RCIM_Cubic, leaveTangent=-3.0576, time=0.8054, value=0.8172}
          {arriveTangent=-2.7631, interpMode=RCIM_Cubic, leaveTangent=-2.7631, time=0.8715, value=0.6939}
          {arriveTangent=-4.5478, interpMode=RCIM_Cubic, leaveTangent=-4.5478, time=0.8793, value=0.6131}
          {arriveTangent=-5.0784, interpMode=RCIM_Cubic, leaveTangent=-5.0784, time=0.9471, value=0.35}
          <at 1 0 RCIM_Cubic>
         ]}}})
    )
    Module.Param0WriteEnabled=true
    Module.Param1WriteEnabled=true
    Module.Param2WriteEnabled=true
    Module.Param3WriteEnabled=true
   ) {enabled=false}
   &DynamicMaterialParameters(
    Index 0 Param 1=&FloatFromCurve(FloatCurve=DataInterface() {propertyValues {Curve {keys [
         {arriveTangent=10.8966, interpMode=RCIM_Cubic, time=0, value=0}
         {arriveTangent=0.875, interpMode=RCIM_Cubic, leaveTangent=0.875, time=0.3, value=0.9}
         {arriveTangent=-1.4411, interpMode=RCIM_Cubic, leaveTangent=-1.4411, tangentMode=RCTM_User, time=0.8, value=0.7}
         <at 1 0 RCIM_Cubic>
        ]}}})
    Index 0 Param 2=&FloatFromCurve(
     FloatCurve=DataInterface() {propertyValues {Curve {keys [
         <at 0.4 6 RCIM_Cubic>
         {arriveTangent=-9.9167, interpMode=RCIM_Cubic, leaveTangent=-9.9167, time=0.8876, value=4.0583}
         {arriveTangent=-37.3423, interpMode=RCIM_Cubic, leaveTangent=-37.3423, tangentMode=RCTM_User, time=1, value=0.05}
        ]}}}
     Scale Curve=0.025
    )
    Index 0 Param 3=&FloatFromCurve(
     FloatCurve=DataInterface() {propertyValues {Curve {keys [
         {arriveTangent=2.0462, interpMode=RCIM_Cubic, time=0, value=0.15}
         {arriveTangent=2.8333, interpMode=RCIM_Cubic, leaveTangent=2.8333, time=0.0547, value=0.875}
         {arriveTangent=0.1677, leaveTangent=0.1677, time=0.3, value=1}
         {arriveTangent=-0.3745, interpMode=RCIM_Cubic, leaveTangent=-0.3745, tangentMode=RCTM_User, time=0.75, value=1}
         {arriveTangent=-8.2121, interpMode=RCIM_Cubic, leaveTangent=-8.2121, tangentMode=RCTM_User, time=1, value=0}
        ]}}}
     Scale Curve=0.05
    )
    Index 0 Param 4=&Lerp_Float(
     A=0
     B=0.6
     Alpha=&FloatFromCurve(
      FloatCurve=DataInterface() {propertyValues {Curve {keys [
          {arriveTangent=9.0725, interpMode=RCIM_Cubic, leaveTangent=9.0725, tangentMode=RCTM_User, time=0, value=0}
          {arriveTangent=-0.5719, leaveTangent=-2.5249, tangentMode=RCTM_User, time=0.1408, value=0.8}
          {arriveTangent=-0.9311, interpMode=RCIM_Cubic, leaveTangent=-0.9311, time=0.95, value=0.8}
          {arriveTangent=-9.7998, interpMode=RCIM_Cubic, leaveTangent=-9.7998, tangentMode=RCTM_User, time=1, value=0}
         ]}}}
      Scale Curve=0.8
     )
    )
    Module.Param0WriteEnabled=true
    Module.Param1WriteEnabled=true
    Module.Param2WriteEnabled=true
    Module.Param3WriteEnabled=true
   ) {module=DynamicMaterialParameters001}
   &ScaleColor(
    Scale Alpha=&FloatFromCurve(FloatCurve=DataInterface() {propertyValues {Curve {keys [<at 0.85 1 RCIM_Cubic>, {arriveTangent=-22.1181, tangentMode=RCTM_User, time=1, value=0}]}}})
    Scale RGB=&VectorFromFloat(Value=&FloatFromCurve(FloatCurve=DataInterface() {propertyValues {Curve {keys [
          {arriveTangent=-62.5873, interpMode=RCIM_Cubic, leaveTangent=-62.5873, tangentMode=RCTM_User, time=0, value=2}
          {arriveTangentWeight=0.335, interpMode=RCIM_Cubic, leaveTangentWeight=0.335, tangentMode=RCTM_User, time=0.1, value=1}
          <at 0.2 2>
         ]}}}))
    ScaleA=true
    ScaleRGB=true
   )
  ]
  rendererClasses [&NiagaraMeshRendererProperties]
  renderers [
   {rendererClass=&NiagaraMeshRendererProperties, rendererIndex=0}
   {rendererClass=&NiagaraMeshRendererProperties, rendererIndex=1}
  ]
 }
 Fragments {
  emitterUpdateScript [
   &EmitterState(Life Cycle Mode=::NewEnumerator1, Loop Behavior=::NewEnumerator1, Loop Delay=$System.CrackDelay, UseLoopDelay=true)
   &SpawnBurst_Instantaneous(Spawn Count=100, Use Spawn Probability=false, UseLoopCountLimit=false)
   &SpawnBurst_Instantaneous(Spawn Count=120, Spawn Time=0.06, Use Spawn Probability=false, UseLoopCountLimit=false) {module=SpawnBurst_Instantaneous001}
   &SpawnBurst_Instantaneous(Spawn Count=50, Spawn Time=0.15, Use Spawn Probability=false, UseLoopCountLimit=false) {module=SpawnBurst_Instantaneous002}
  ]
  particleSpawnScript [
   &FN_SetSpawnGroup()
   /SetVariables(Particles.VisToggle:NiagaraInt32=&MakeIntFromFloat(Float=&ClampFloat(Float=&RandomRangeFloat(Minimum=0, Maximum=3)))) {isSetParameters=true, module=SetVariables_#2}
   &InitializeParticle(
    Color Maximum=<rgba 0.2523 3.4809 4 1>
    Color Minimum=<rgba 0.5645 0.706 3 1>
    Color Mode=::NewEnumerator2
    Lifetime Max=1.5
    Lifetime Min=0.5
    Lifetime Mode=::NewEnumerator1
    Mesh Scale Max=<4 7 5>
    Mesh Scale Min=<2 4 5>
    Mesh Scale Mode=::NewEnumerator2
    Sprite Size Mode=::NewEnumerator1
    Uniform Sprite Size Max=50
    Uniform Sprite Size Min=25
    UsePositionOffset=false
    Write Material Random=true
   )
   &SphereLocation(
    MaskSpawnGroup=false
    Non Uniform Scale=<0.25 1 1>
    Override Seed=false
    Sphere Radius=200
    Surface Only=false
    UseNonUniformScale=true
    UseOffset=false
    UseSphereOrientation=false
   )
   &AddVelocityFromPoint(
    Module.Use Falloff Distance=false
    Velocity Strength=&Multiply_Float(A=2500, B=&RandomRangeFloat(Minimum=0.5, Maximum=1))
   )
   &AddVelocity(Velocity=&Multiply_VectorByFloat(Vector=<1 0 0>, Float=&RandomRangeFloat(Minimum=-1500, Maximum=2000)))
   &InitialMeshOrientation(Orientation Vector=&RandomVector(Vector Scale=1), Use Orientation Vector=true, Use Rotation Vector=false)
   /SetVariables(Particles.MeshIndex=&RandomRangeInt(Minimum=0, Maximum=2)) {isSetParameters=true, module=SetVariables_#3}
   &CameraOffset(Camera Offset Amount=-25)
  ]
  particleUpdateScript [
   &ParticleState()
   &UpdateMeshOrientation(
    Rotation Coordinate Space=::NewEnumerator3
    Rotation Rate=&FloatFromCurve(
     FloatCurve=DataInterface() {propertyValues {Curve {keys [
         {arriveTangent=-2.5411, interpMode=RCIM_Cubic, leaveTangent=-2.5411, tangentMode=RCTM_User, time=0, value=1}
         {arriveTangent=-0.173, interpMode=RCIM_Cubic, leaveTangent=-0.173, tangentMode=RCTM_User, time=1, value=0.1}
        ]}}}
     Scale Curve=&FloatFromCurve(
      FloatCurve=DataInterface() {propertyValues {Curve {keys [
          {interpMode=RCIM_Cubic, tangentMode=RCTM_User, time=0, value=-1}
          {arriveTangent=9.6958, interpMode=RCIM_Cubic, leaveTangent=9.6958, tangentMode=RCTM_User, time=0.5, value=0}
          {interpMode=RCIM_Cubic, tangentMode=RCTM_User, time=1, value=1}
         ]}}}
      CurveIndex=&RandomRangeFloat(Minimum=0, Maximum=1)
      Scale Curve=2
     )
    )
    Rotation Vector=&RandomRangeVector(Minimum=<0 1 0>, Maximum=<1 0 0>)
   )
   &Drag(
    Drag=&SelectFloatFromArray(Direct Array Index=$Particles.VisToggle, Float Selection Array=DataInterface() {propertyValues {FloatData [1.5,2]}})
    Use Linear Drag=true
    Use Rotational Drag=false
   )
   &SolveForcesAndVelocity(Clamp Velocity=false, Limit Acceleration=false)
   &ScaleColor(
    Scale Alpha=&FloatFromCurve(FloatCurve=DataInterface() {propertyValues {Curve {keys [
         {arriveTangent=-0.0405, interpMode=RCIM_Cubic, leaveTangent=-0.0405, tangentMode=RCTM_User, time=0.6, value=1}
         {interpMode=RCIM_Cubic, leaveTangent=-4.3194, time=1, value=0}
        ]}}})
    Scale RGB=&VectorFromFloat(Value=&Add_Float(
      A=1
      B=&FloatFromCurve(
       FloatCurve=DataInterface() {propertyValues {Curve {keys [
           {arriveTangent=-5.4733, interpMode=RCIM_Cubic, leaveTangent=-5.4733, tangentMode=RCTM_User, time=0, value=1}
           {interpMode=RCIM_Cubic, tangentMode=RCTM_User, time=0.5, value=0}
          ]}}}
       Scale Curve=20
      )
     ))
    ScaleA=true
    ScaleRGB=true
   )
   /SetVariables(Particles.MeshColor:LinearColor=&MakeLinearColorFromVectorAndFloat(
     "Vector (RGB)"=&VectorFromFloat(Value=&Add_Float(
       A=5
       B=&FloatFromCurve(
        FloatCurve=DataInterface() {propertyValues {Curve {keys [
            {arriveTangent=-3.7813, interpMode=RCIM_Cubic, leaveTangent=-3.7813, tangentMode=RCTM_User, time=0, value=1}
            {interpMode=RCIM_Cubic, tangentMode=RCTM_User, time=0.7, value=0}
           ]}}}
        Scale Curve=20
       )
      ))
     "Float (Alpha)"=&FloatFromCurve(
      FloatCurve=DataInterface() {propertyValues {Curve {keys [<at 0.7 1 RCIM_Cubic>, {arriveTangent=-5.6246, interpMode=RCIM_Cubic, leaveTangent=-5.6246, tangentMode=RCTM_User, time=1, value=0}]}}}
      Scale Curve=2.5
     )
    )) {isSetParameters=true, module=SetVariables_#4}
   &ScaleMeshSize(Scale Factor=&VectorFromFloat(Value=&FloatFromCurve(FloatCurve=DataInterface() {propertyValues {Curve {keys [
          {arriveTangent=-0.0405, interpMode=RCIM_Cubic, leaveTangent=-0.0405, tangentMode=RCTM_User, time=0.6, value=1}
          {arriveTangent=-5.4415, interpMode=RCIM_Cubic, leaveTangent=-5.4415, tangentMode=RCTM_User, time=1, value=0}
         ]}}})))
  ]
  rendererClasses [&NiagaraMeshRendererProperties, &NiagaraSpriteRendererProperties]
  renderers [
   {rendererClass=&NiagaraMeshRendererProperties, rendererIndex=0}
   {rendererClass=&NiagaraSpriteRendererProperties, rendererIndex=1}
  ]
 }
 Glints {
  emitterUpdateScript [
   &EmitterState(
    Delay First Loop Only=true
    Enable Distance Culling=true
    Life Cycle Mode=::NewEnumerator1
    Loop Behavior=::NewEnumerator2
    Loop Count=3
    Loop Delay=$System.CrackDelay
    Loop Duration=&RandomRangeFloat(Minimum=0.8, Maximum=2, Evaluation Type=::NewEnumerator1)
    MaxDistance=3200
    Module.bUseMaxDistance=true
    Module.bUseMinDistance=false
    Recalculate Duration Each Loop=true
    Scalability Mode=::NewEnumerator1
    UseLoopDelay=true
   )
   &SpawnRate(SpawnRate=0, Use Spawn Probability=false)
   &SpawnBurst_Instantaneous(Spawn Count=15, Spawn Time=0.15, Use Spawn Probability=false)
   &SpawnBurst_Instantaneous(Spawn Count=30, Spawn Time=0.9, Use Spawn Probability=false) {module=SpawnBurst_Instantaneous001}
  ]
  particleSpawnScript [
   /Niagara/Modules/Spawn/Initialization/InitializeParticle(
    Lifetime=&RandomRangeFloat(Minimum=0.12, Maximum=1.5, Evaluation Type=::NewEnumerator1)
    Module.Write Color=true
    Module.Write Lifetime=true
    Module.Write Mass=true
    Module.Write Material Random=true
    Module.Write Position=false
    Module.Write Scale=false
    Module.Write SpriteRotation=false
    Module.Write SpriteSize=true
    Sprite Size=&RandomRangeVector2D(Minimum=<120 80>, Maximum=<200 140>, Evaluation Type=::NewEnumerator1)
   ) {enabled=false}
   /Niagara/Modules/Spawn/Initialization/V2/InitializeParticle(
    Lifetime=&RandomRangeFloat(Minimum=0.12, Maximum=1.5)
    Mass Mode=::NewEnumerator1
    Sprite Size=&RandomRangeVector2D(Minimum=<120 80>, Maximum=<200 140>)
    Sprite Size Mode=::NewEnumerator4
    UsePositionOffset=false
    Write Material Random=true
   ) {module=InitializeParticle001}
   &Color(Color=&MakeLinearColorFromVectorAndFloat("Vector (RGB)"=<0.2398 1.0892 1.5>, "Float (Alpha)"=&ClampFloat(Float=1)))
   &ShapeLocation(MaskSpawnGroup=false, Override Seed=false, Sphere Radius=500, UseOwnerScale=true)
  ]
  particleUpdateScript [
   &ParticleState()
   &ScaleSpriteSize(Scale Factor=&Vector2DFromFloat(Value=&FloatFromCurve(FloatCurve=DataInterface() {propertyValues {Curve {keys [
          <at 0 0 RCIM_Cubic>
          {arriveTangent=1.4286, interpMode=RCIM_Cubic, leaveTangent=1.4286, time=0.5754, value=0.3493}
          {arriveTangent=-0.8228, interpMode=RCIM_Cubic, leaveTangent=-0.8228, time=0.7, value=1}
          <at 1 0>
         ]}}})))
   &Drag(Drag=0.2, Use Linear Drag=true, Use Rotational Drag=false)
   &SolveForcesAndVelocity(Clamp Velocity=false, Limit Acceleration=false)
  ]
  rendererClasses [&NiagaraSpriteRendererProperties]
  renderers [{rendererClass=&NiagaraSpriteRendererProperties}]
 }
 Glow {
  emitterUpdateScript [
   &EmitterState()
   &SpawnBurst_Instantaneous(Spawn Count=1, Use Spawn Probability=false)
  ]
  particleSpawnScript [
   &InitializeParticle(
    Color=<rgba 0.0195 0.6416 1.25 1>
    Lifetime=5.6
    Sprite Size Mode=::NewEnumerator3
    Uniform Sprite Size=2500
    UsePositionOffset=false
    Write Material Random=true
   )
   /SetVariables(
    Particles.DynParams2:Vector4f=&MakeVector4(X=0.5, Y=1, Z=1, W=1)
    Particles.DynParams3:Vector4f=<1000 0 0 0>
   ) {isSetParameters=true}
   &CameraOffset(Camera Offset Amount=10)
  ]
  particleUpdateScript [
   &DynamicMaterialParameters(
    Index 0 Param 1=0.8
    Index 0 Param 2=&Lerp_Float(
     A=0.05
     B=1
     Alpha=&FloatFromCurve(FloatCurve=DataInterface() {propertyValues {Curve {keys [
          {arriveTangent=2.3085, interpMode=RCIM_Cubic, leaveTangent=2.3085, tangentMode=RCTM_User, time=0, value=0}
          {arriveTangent=1, interpMode=RCIM_Cubic, leaveTangent=1, time=0.15, value=0.55}
          {arriveTangentWeight=0.132, interpMode=RCIM_Cubic, leaveTangentWeight=0.1886, tangentMode=RCTM_User, time=0.5, value=0.5}
          {arriveTangent=3.3782, interpMode=RCIM_Cubic, leaveTangent=3.3782, tangentMode=RCTM_User, time=1, value=1}
         ]}}})
    )
    Module.Param0WriteEnabled=true
    Module.Param1WriteEnabled=true
    Module.Param2WriteEnabled=false
    Module.Param3WriteEnabled=false
   )
   &ParticleState()
   &ScaleSpriteSize(Uniform Curve Sprite Scale=DataInterface() {propertyValues {Curve {keys [
        <at 0 0.1 RCIM_Cubic>
        {arriveTangent=4.5, interpMode=RCIM_Cubic, leaveTangent=4.5, time=0.19, value=0.45}
        {arriveTangent=-2.0962, interpMode=RCIM_Cubic, leaveTangent=-2.0962, tangentMode=RCTM_User, time=0.2, value=1}
        {arriveTangent=-1.6667, interpMode=RCIM_Cubic, leaveTangent=-1.6667, time=0.4, value=0.6}
        {arriveTangent=-0.0822, interpMode=RCIM_Cubic, leaveTangent=-0.0822, tangentMode=RCTM_User, time=0.5, value=0.5}
        {arriveTangent=3.0844, time=0.8, value=1}
        {arriveTangent=4.7412, interpMode=RCIM_Cubic, leaveTangent=-5.2133, tangentMode=RCTM_Break, time=0.925, value=1}
        <at 1 0>
       ]}}})
   /SetVariables(Particles.SpriteSize2=&Multiply_Vector2(
     A=&Multiply_Vector2(A=$Particles.Initial.SpriteSize, B=<2 0.1>)
     B=&Vector2DFromFloat(Value=&FloatFromCurve(FloatCurve=DataInterface() {propertyValues {Curve {keys [
           <at 0 0 RCIM_Cubic>
           {arriveTangent=2.5542, interpMode=RCIM_Cubic, leaveTangent=2.5542, tangentMode=RCTM_User, time=0.19, value=0.5}
           {arriveTangent=-5.0959, interpMode=RCIM_Cubic, leaveTangent=-5.0959, tangentMode=RCTM_User, time=0.2, value=1.5}
           {arriveTangentWeight=0.1736, interpMode=RCIM_Cubic, leaveTangentWeight=0.0868, tangentMode=RCTM_User, time=0.6, value=0.6}
           {arriveTangent=1.9949, interpMode=RCIM_Cubic, leaveTangent=1.9949, tangentMode=RCTM_User, time=1, value=1}
          ]}}}))
    )) {isSetParameters=true}
   &ScaleColor(
    Scale Alpha=&FloatFromCurve(FloatCurve=DataInterface() {propertyValues {Curve {keys [
         {arriveTangent=3.4226, interpMode=RCIM_Cubic, time=0, value=0}
         {arriveTangent=10.4625, interpMode=RCIM_Cubic, leaveTangent=10.4625, tangentMode=RCTM_User, time=0.17, value=0.5}
         {arriveTangent=-10.1737, interpMode=RCIM_Cubic, leaveTangent=-10.1737, tangentMode=RCTM_User, time=0.2, value=1.5}
         {arriveTangent=-2.4, interpMode=RCIM_Cubic, leaveTangent=-2.4, time=0.28, value=0.65}
         {arriveTangentWeight=0.1965, interpMode=RCIM_Cubic, leaveTangentWeight=0.1179, tangentMode=RCTM_User, time=0.7, value=0.3}
         {interpMode=RCIM_Cubic, tangentMode=RCTM_User, time=1, value=0.8}
        ]}}})
    ScaleA=true
    ScaleRGB=false
   )
   /SetVariables(Particles.LightColor=&MultiplyLinearColor(Linear Color=$Particles.Color, Scale Factor=&MakeVector4_FromFloat(Float=0))) {isSetParameters=true}
   /SetVariables(Particles.LightRadius=&FloatFromCurve(
     FloatCurve=DataInterface() {propertyValues {Curve {keys [
         {interpMode=RCIM_Cubic, tangentMode=RCTM_User, time=0, value=0.25}
         {arriveTangent=0.25, interpMode=RCIM_Cubic, leaveTangent=0.25, time=0.2, value=1}
         {arriveTangent=-1.5747, interpMode=RCIM_Cubic, leaveTangent=-1.5747, tangentMode=RCTM_User, time=1, value=0.5}
        ]}}}
     Scale Curve=100
    )) {isSetParameters=true}
   /SetVariables(Particles.Color2=&MakeLinearColorFromVectorAndFloat(
     "Vector (RGB)"=<0 0 0>
     "Float (Alpha)"=&FloatFromCurve(FloatCurve=DataInterface() {propertyValues {Curve {keys [
          {arriveTangent=3.4226, interpMode=RCIM_Cubic, time=0, value=0}
          {
           arriveTangent=-0.9498
           arriveTangentWeight=0.5255
           interpMode=RCIM_Cubic
           leaveTangent=-0.9498
           leaveTangentWeight=2.9779
           tangentMode=RCTM_User
           time=0.2
           value=0.9
          }
          {arriveTangentWeight=0.2749, interpMode=RCIM_Cubic, leaveTangent=1.2882, leaveTangentWeight=0.15, time=1, value=0.75}
         ]}}})
    )) {isSetParameters=true}
  ]
  rendererClasses [&NiagaraSpriteRendererProperties, &NiagaraLightRendererProperties]
  renderers [
   {rendererClass=&NiagaraSpriteRendererProperties}
   {rendererClass=&NiagaraLightRendererProperties, rendererIndex=1}
   {rendererClass=&NiagaraSpriteRendererProperties, rendererIndex=2}
   {rendererClass=&NiagaraSpriteRendererProperties, rendererIndex=3}
  ]
 }
 GlowBurst {
  emitterUpdateScript [
   &EmitterState(Life Cycle Mode=::NewEnumerator1, Loop Behavior=::NewEnumerator1, Loop Delay=$System.CrackDelay, UseLoopDelay=true)
   &SpawnBurst_Instantaneous(Spawn Count=1, Use Spawn Probability=false)
  ]
  particleSpawnScript [
   &InitializeParticle(
    Color=<rgba 0.0856 1.1221 1.5 1>
    Lifetime=0.5
    Mesh Scale Mode=::NewEnumerator3
    Mesh Uniform Scale=30
    Sprite Size Mode=::NewEnumerator3
    Uniform Sprite Size=12000
    UsePositionOffset=false
    Write Material Random=true
   ) {module=InitializeParticle001}
   &InitialMeshOrientation(Orientation Axis=<0 0 1>, Use Orientation Vector=true, Use Rotation Vector=false) {module=InitialMeshOrientation001}
   &CameraOffset(Camera Offset Amount=-250)
  ]
  particleUpdateScript [
   &ParticleState()
   /SetVariables(Particles.MeshDynParams:Vector4f=&MakeVector4(
     X=&FloatFromCurve(FloatCurve=DataInterface() {propertyValues {Curve {keys [
          {arriveTangent=1.7005, interpMode=RCIM_Cubic, leaveTangent=1.7005, tangentMode=RCTM_User, time=0, value=0.4}
          {arriveTangent=0.0334, interpMode=RCIM_Cubic, leaveTangent=0.0334, tangentMode=RCTM_User, time=1, value=1}
         ]}}})
     Y=0
     Z=0
     W=0
    )) {isSetParameters=true, module=SetVariables_#5}
   &ScaleSpriteSize(Scale Factor=&Vector2DFromFloat(Value=&FloatFromCurve(FloatCurve=DataInterface() {propertyValues {Curve {keys [
          {arriveTangent=1.2491, interpMode=RCIM_Cubic, leaveTangent=1.2491, tangentMode=RCTM_User, time=0, value=0.1}
          {arriveTangent=0.3527, interpMode=RCIM_Cubic, leaveTangent=0.3527, tangentMode=RCTM_User, time=1, value=1}
         ]}}})))
   &DynamicMaterialParameters(
    Index 0 Param 1=1
    Index 0 Param 2=&FloatFromCurve(FloatCurve=DataInterface() {propertyValues {Curve {keys [
         {arriveTangent=-1.3736, time=0, value=1}
         {arriveTangent=-1.3273, interpMode=RCIM_Cubic, leaveTangent=-1.3273, tangentMode=RCTM_User, time=1, value=0.1}
        ]}}})
    Module.Param0WriteEnabled=true
    Module.Param1WriteEnabled=true
    Module.Param2WriteEnabled=false
    Module.Param3WriteEnabled=false
   )
   &ScaleColor(
    Scale Alpha=&FloatFromCurve(FloatCurve=DataInterface() {propertyValues {Curve {keys [
         {arriveTangent=37.5771, interpMode=RCIM_Cubic, leaveTangent=37.5771, tangentMode=RCTM_User, time=0, value=0}
         {arriveTangent=-5.419, interpMode=RCIM_Cubic, leaveTangent=-5.419, tangentMode=RCTM_User, time=0.05, value=1}
         {arriveTangent=-1.5245, interpMode=RCIM_Cubic, leaveTangent=-1.5245, tangentMode=RCTM_User, time=0.2, value=0.4}
         <at 1 0>
        ]}}})
    ScaleA=true
    ScaleRGB=false
   )
   /SetVariables(Particles.LightColor=&MultiplyLinearColor(Linear Color=$Particles.Color, Scale Factor=&MakeVector4_FromFloat(Float=15))) {isSetParameters=true, module=SetVariables_#6}
   &ScaleMeshSize(Scale Factor=&VectorFromFloat(Value=&FloatFromCurve(FloatCurve=DataInterface() {propertyValues {Curve {keys [
          {arriveTangent=1.9831, interpMode=RCIM_Cubic, leaveTangent=1.9831, tangentMode=RCTM_User, time=0, value=0.1}
          {arriveTangent=0.1996, interpMode=RCIM_Cubic, leaveTangent=0.1996, tangentMode=RCTM_User, time=1, value=1}
         ]}}})))
  ]
  rendererClasses [
   &NiagaraLightRendererProperties
   &NiagaraSpriteRendererProperties
   &NiagaraMeshRendererProperties
  ]
  renderers [
   {rendererClass=&NiagaraLightRendererProperties, rendererIndex=0}
   {rendererClass=&NiagaraSpriteRendererProperties, rendererIndex=1}
   {rendererClass=&NiagaraMeshRendererProperties, rendererIndex=2}
  ]
 }
 GlowBurstEnd {
  emitterUpdateScript [
   &EmitterState(Life Cycle Mode=::NewEnumerator1, Loop Behavior=::NewEnumerator1, Loop Delay=5.3, UseLoopDelay=true)
   &SpawnBurst_Instantaneous(Spawn Count=1, Use Spawn Probability=false)
  ]
  particleSpawnScript [
   &InitializeParticle(
    Color=<rgba 0.5708 7.4805 10 1>
    Lifetime=0.35
    Mesh Scale Mode=::NewEnumerator3
    Mesh Uniform Scale=10
    Sprite Size Mode=::NewEnumerator3
    Uniform Sprite Size=2500
    UsePositionOffset=false
    Write Material Random=true
   ) {module=InitializeParticle001}
   &InitialMeshOrientation(Orientation Axis=<0 0 1>, Use Orientation Vector=true, Use Rotation Vector=false) {module=InitialMeshOrientation001}
   &CameraOffset(Camera Offset Amount=50)
  ]
  particleUpdateScript [
   &ParticleState()
   /SetVariables(Particles.MeshDynParams:Vector4f=&MakeVector4(
     X=&FloatFromCurve(FloatCurve=DataInterface() {propertyValues {Curve {keys [
          {arriveTangent=1.7005, interpMode=RCIM_Cubic, leaveTangent=1.7005, tangentMode=RCTM_User, time=0, value=0.4}
          {arriveTangent=0.0334, interpMode=RCIM_Cubic, leaveTangent=0.0334, tangentMode=RCTM_User, time=1, value=1}
         ]}}})
     Y=0
     Z=0
     W=0
    )) {isSetParameters=true, module=SetVariables_#5}
   &ScaleSpriteSize(Scale Factor=&Vector2DFromFloat(Value=&FloatFromCurve(FloatCurve=DataInterface() {propertyValues {Curve {keys [
          {arriveTangent=1.2491, interpMode=RCIM_Cubic, leaveTangent=1.2491, tangentMode=RCTM_User, time=0, value=0.1}
          {arriveTangent=0.3527, interpMode=RCIM_Cubic, leaveTangent=0.3527, tangentMode=RCTM_User, time=1, value=1}
         ]}}})))
   &DynamicMaterialParameters(
    Index 0 Param 1=0.9
    Index 0 Param 2=&FloatFromCurve(FloatCurve=DataInterface() {propertyValues {Curve {keys [
         {arriveTangent=-2.7033, interpMode=RCIM_Cubic, leaveTangent=-2.7033, tangentMode=RCTM_User, time=0, value=1}
         {arriveTangent=-0.1381, interpMode=RCIM_Cubic, leaveTangent=-0.1381, tangentMode=RCTM_User, time=1, value=0.1}
        ]}}})
    Module.Param0WriteEnabled=true
    Module.Param1WriteEnabled=true
    Module.Param2WriteEnabled=false
    Module.Param3WriteEnabled=false
   )
   &ScaleColor(
    Scale Alpha=&FloatFromCurve(FloatCurve=DataInterface() {propertyValues {Curve {keys [
         {arriveTangent=37.5771, interpMode=RCIM_Cubic, leaveTangent=37.5771, tangentMode=RCTM_User, time=0, value=0}
         {arriveTangent=-5.419, interpMode=RCIM_Cubic, leaveTangent=-5.419, tangentMode=RCTM_User, time=0.05, value=1}
         {arriveTangent=-0.9659, interpMode=RCIM_Cubic, leaveTangent=-0.9659, tangentMode=RCTM_User, time=0.25, value=0.25}
         <at 1 0>
        ]}}})
    ScaleA=true
    ScaleRGB=false
   )
   /SetVariables(Particles.LightColor=&MultiplyLinearColor(Linear Color=$Particles.Color, Scale Factor=&MakeVector4_FromFloat(Float=15))) {isSetParameters=true, module=SetVariables_#6}
   &ScaleMeshSize(Scale Factor=&VectorFromFloat(Value=&FloatFromCurve(FloatCurve=DataInterface() {propertyValues {Curve {keys [
          {arriveTangent=1.9831, interpMode=RCIM_Cubic, leaveTangent=1.9831, tangentMode=RCTM_User, time=0, value=0.1}
          {arriveTangent=0.1996, interpMode=RCIM_Cubic, leaveTangent=0.1996, tangentMode=RCTM_User, time=1, value=1}
         ]}}})))
  ]
  rendererClasses [
   &NiagaraLightRendererProperties
   &NiagaraSpriteRendererProperties
   &NiagaraMeshRendererProperties
  ]
  renderers [
   {rendererClass=&NiagaraLightRendererProperties, rendererIndex=0}
   {rendererClass=&NiagaraSpriteRendererProperties, rendererIndex=1}
   {rendererClass=&NiagaraMeshRendererProperties, rendererIndex=2}
  ]
 }
 LightShaftsBurst {
  emitterUpdateScript [
   &EmitterState(Life Cycle Mode=::NewEnumerator1, Loop Behavior=::NewEnumerator1, Loop Delay=$System.CrackDelay, Loop Duration=3, UseLoopDelay=true)
   &SpawnBurst_Instantaneous(Use Spawn Probability=false, UseLoopCountLimit=false)
  ]
  particleSpawnScript [
   &InitializeParticle(
    Color=<rgba 1.2748 5.156 10 1>
    Lifetime=1.5
    Mesh Scale Mode=::NewEnumerator3
    Mesh Uniform Scale=5
    Position Offset=<-180 0 0>
    UsePositionOffset=true
    Write Material Random=true
   )
   &InitialMeshOrientation(Orientation Vector=<0 0 1>, Use Orientation Vector=true, Use Rotation Vector=false)
   &CameraOffset(Camera Offset Amount=-250) {enabled=false}
  ]
  particleUpdateScript [
   &ParticleState()
   &DynamicMaterialParameters(
    Index 0 Param 1=&FloatFromCurve(
     FloatCurve=DataInterface() {propertyValues {Curve {keys [{interpMode=RCIM_Cubic, tangentMode=RCTM_User, time=0, value=1}, {interpMode=RCIM_Cubic, tangentMode=RCTM_User, time=1, value=0.25}]}}}
     Scale Curve=4
    )
    Index 0 Param 2=&Lerp_Float(
     A=30
     B=10
     Alpha=&FloatFromCurve(FloatCurve=DataInterface() {propertyValues {Curve {keys [
          {arriveTangent=4.7763, interpMode=RCIM_Cubic, leaveTangent=4.7763, tangentMode=RCTM_User, time=0, value=0}
          {arriveTangent=0.4097, interpMode=RCIM_Cubic, leaveTangent=0.4097, tangentMode=RCTM_User, time=0.1, value=0.9}
          <at 1 1>
         ]}}})
    )
    Index 0 Param 3=&FloatFromCurve(
     FloatCurve=DataInterface() {propertyValues {Curve {keys [
         {arriveTangent=20.3138, interpMode=RCIM_Cubic, leaveTangent=20.3138, tangentMode=RCTM_User, time=0, value=0.5}
         {arriveTangent=0.5, interpMode=RCIM_Cubic, leaveTangent=0.5, time=0.05, value=1}
         {interpMode=RCIM_Cubic, tangentMode=RCTM_User, time=1, value=0.25}
        ]}}}
     Scale Curve=1.5
    )
    Index 0 Param 4=500
    Module.Param0WriteEnabled=true
    Module.Param1WriteEnabled=true
    Module.Param2WriteEnabled=true
    Module.Param3WriteEnabled=true
   )
   &ScaleColor(
    Scale RGB=&VectorFromFloat(Value=&FloatFromCurve(FloatCurve=DataInterface() {propertyValues {Curve {keys [{interpMode=RCIM_Cubic, tangentMode=RCTM_User, time=0, value=1}, {interpMode=RCIM_Cubic, tangentMode=RCTM_User, time=1, value=0}]}}}))
    ScaleA=true
    ScaleRGB=true
   )
  ]
  rendererClasses [&NiagaraMeshRendererProperties]
  renderers [{rendererClass=&NiagaraMeshRendererProperties, rendererIndex=0}]
 }
 Light_CharactersFront {
  emitterUpdateScript [
   &EmitterState(Life Cycle Mode=::NewEnumerator1, Loop Behavior=::NewEnumerator1, Loop Delay=$User.InitialStartDelay, Loop Duration=6, UseLoopDelay=false)
   &SpawnBurst_Instantaneous(Spawn Count=1, Use Spawn Probability=false)
  ]
  particleSpawnScript [
   /Niagara/Modules/Spawn/Initialization/InitializeParticle(
    Color=<rgba 0.147 0.5957 1 1>
    Lifetime=&RandomRangeFloat(Minimum=2.7, Maximum=3.2, Evaluation Type=::NewEnumerator1)
    Module.Write Color=true
    Module.Write Lifetime=true
    Module.Write Mass=true
    Module.Write Material Random=true
    Module.Write Position=false
    Module.Write Scale=false
    Module.Write SpriteRotation=false
    Module.Write SpriteSize=true
    Sprite Size=<500 500>
   ) {enabled=false}
   /Niagara/Modules/Spawn/Initialization/V2/InitializeParticle(
    Color=<rgba 2.0677 2.7891 3 1>
    Lifetime=5.25
    Mass Mode=::NewEnumerator1
    Sprite Size=<500 500>
    Sprite Size Mode=::NewEnumerator4
    UsePositionOffset=false
    Write Material Random=true
   ) {module=InitializeParticle001}
   /SetVariables(Particles.LightBrightness=1500) {isSetParameters=true}
  ]
  particleUpdateScript [
   /SetVariables(Particles.LightColorInit:LinearColor=$Particles.Color) {enabled=false, isSetParameters=true}
   /SetVariables(Particles.LightColor=$Particles.Color) {isSetParameters=true}
   /SetVariables(Particles.LightExponent=12) {isSetParameters=true}
   /SetVariables(Particles.LightRadius=2500) {isSetParameters=true}
   &ParticleState()
   &ScaleColor(
    Scale Alpha=&FloatFromCurve(FloatCurve=DataInterface() {propertyValues {Curve {keys [{arriveTangent=-0.0064, time=0, value=1}, {arriveTangent=-1, interpMode=RCIM_Cubic, leaveTangent=-1, time=0.75, value=1}, <at 1 0>]}}})
    ScaleA=true
    ScaleRGB=true
   )
   /SetVariables(Particles.LightIntensityRamp=5) {isSetParameters=true}
   &CameraOffset(Camera Offset Amount=10)
  ]
  rendererClasses [&NiagaraSpriteRendererProperties, &NiagaraLightRendererProperties]
  renderers [
   {rendererClass=&NiagaraSpriteRendererProperties}
   {rendererClass=&NiagaraLightRendererProperties, rendererIndex=1}
  ]
 }
 PreCrackGlow {
  bEnabled=false
  emitterUpdateScript [
   &EmitterState()
   &SpawnBurst_Instantaneous(Use Spawn Probability=false, UseLoopCountLimit=false)
  ]
  particleSpawnScript [
   &InitializeParticle(
    Color=<rgba 1.1042 1.5161 2 1>
    Lifetime=$System.CrackDelay
    Mesh Scale Mode=::NewEnumerator3
    Mesh Uniform Scale=10
    Position Offset=<-200 0 0>
    UsePositionOffset=true
    Write Material Random=true
   ) {module=InitializeParticle001}
   &InitialMeshOrientation(Orientation Vector=<0 0 -1>, Use Orientation Vector=true, Use Rotation Vector=false)
  ]
  particleUpdateScript [
   &ParticleState()
   &ScaleColor(
    Scale Alpha=&FloatFromCurve(FloatCurve=DataInterface() {propertyValues {Curve {keys [<at 0 0 RCIM_Cubic>, {arriveTangent=1.4286, interpMode=RCIM_Cubic, leaveTangent=1.4286, time=0.2228, value=0.673}, <at 0.7 1 RCIM_Cubic>]}}})
    Scale RGB=&VectorFromCurve(VectorCurve=DataInterface() {propertyValues {
       XCurve {keys [<at 0 0.3 RCIM_Cubic>, {arriveTangent=0.3244, interpMode=RCIM_Cubic, leaveTangent=0.3244, time=0.8, value=0.5792}, <at 1 0.6244 RCIM_Cubic>]}
       YCurve {keys [
         <at 0 0.1333 RCIM_Cubic>
         {arriveTangent=0.5819, interpMode=RCIM_Cubic, leaveTangent=0.5819, tangentMode=RCTM_User, time=0.8, value=0.4141}
         {interpMode=RCIM_Cubic, leaveTangent=53.0681, time=1, value=5}
        ]}
       ZCurve {keys [
         <at 0 0.2178 RCIM_Cubic>
         {arriveTangent=2.7196, interpMode=RCIM_Cubic, leaveTangent=2.7196, tangentMode=RCTM_User, time=0.8, value=1.25}
         {interpMode=RCIM_Cubic, leaveTangent=39.7353, time=1, value=4.3508}
        ]}
      }})
    ScaleA=true
    ScaleRGB=true
   )
   /SetVariables(Particles.Color2=&MakeLinearColorFromVectorAndFloat(
     "Vector (RGB)"=&VectorFromFloat(Value=&FloatFromCurve(FloatCurve=DataInterface() {propertyValues {Curve {keys [<at 0.3 2 RCIM_Cubic>, {interpMode=RCIM_Cubic, leaveTangent=8.9946, time=1, value=5}]}}}))
     "Float (Alpha)"=1
    )) {isSetParameters=true}
   &DynamicMaterialParameters(
    Index 0 Param 1=&FloatFromCurve(FloatCurve=DataInterface() {propertyValues {Curve {keys [<at 0 1 RCIM_Cubic>, {arriveTangent=-2.5599, interpMode=RCIM_Cubic, leaveTangent=-2.5599, tangentMode=RCTM_User, time=0.98, value=0}]}}})
    Index 0 Param 2=&FloatFromCurve(
     FloatCurve=DataInterface() {propertyValues {Curve {keys [
         <at 0 0.25 RCIM_Cubic>
         {arriveTangent=0.9943, interpMode=RCIM_Cubic, leaveTangent=0.9943, tangentMode=RCTM_User, time=0.94, value=1}
         {interpMode=RCIM_Cubic, leaveTangent=-12.1231, time=1, value=0.6}
        ]}}}
     Scale Curve=0.6
    )
    Index 0 Param 3=&FloatFromCurve(FloatCurve=DataInterface() {propertyValues {Curve {keys [<at 0 0 RCIM_Cubic>,<at 1 0>]}}})
    Module.Param0WriteEnabled=true
    Module.Param1WriteEnabled=true
    Module.Param2WriteEnabled=true
    Module.Param3WriteEnabled=true
   )
   /SetVariables(Particles.DynParams2:Vector4f=&MakeVector4(
     X=&FloatFromCurve(FloatCurve=DataInterface() {propertyValues {Curve {keys [
          {arriveTangent=0.0047, time=0, value=0}
          {arriveTangent=0.0117, leaveTangent=0.0117, time=0.0669, value=0.0047}
          {arriveTangent=0.0018, leaveTangent=0.0018, time=0.0733, value=0.0164}
          {arriveTangent=0.0254, leaveTangent=0.0254, time=0.1844, value=0.0182}
          {arriveTangent=0.0022, leaveTangent=0.0022, time=0.1873, value=0.0435}
          {arriveTangent=0.0502, leaveTangent=0.0502, time=0.3152, value=0.0457}
          {arriveTangent=-0.0006, leaveTangent=-0.0006, time=0.3233, value=0.0959}
          {arriveTangent=0.0994, leaveTangent=0.0994, time=0.4744, value=0.0954}
          {arriveTangent=0.0025, leaveTangent=0.0025, time=0.4844, value=0.1948}
          {arriveTangent=0.1188, leaveTangent=0.1188, time=0.6747, value=0.1973}
          {arriveTangent=2.4678, interpMode=RCIM_Cubic, leaveTangent=2.4678, time=0.7018, value=0.3161}
          {interpMode=RCIM_Cubic, leaveTangent=2.8447, time=1, value=1}
         ]}}})
     Y=&FloatFromCurve(
      FloatCurve=DataInterface() {propertyValues {Curve {keys [<at 0 0.4 RCIM_Cubic>, {interpMode=RCIM_Cubic, leaveTangent=7.4823, time=1, value=3}]}}}
      Scale Curve=0.25
     )
     Z=1
     W=&FloatFromCurve(FloatCurve=DataInterface() {propertyValues {Curve {keys [<at 0 0>,<at 1 0>]}}})
    )) {isSetParameters=true}
  ]
  rendererClasses [&NiagaraMeshRendererProperties]
  renderers [
   {rendererClass=&NiagaraMeshRendererProperties, rendererIndex=0}
   {rendererClass=&NiagaraMeshRendererProperties, rendererIndex=1}
  ]
 }
 PreCrackLightShafts {
  emitterUpdateScript [
   &EmitterState(Life Cycle Mode=::NewEnumerator1, Loop Behavior=::NewEnumerator1, Loop Duration=$System.CrackDelay, UseLoopDelay=false)
   &SpawnBurst_Instantaneous(Use Spawn Probability=false, UseLoopCountLimit=false)
  ]
  particleSpawnScript [
   &InitializeParticle(
    Color=<rgba 0.255 1.0312 2 1>
    Lifetime=$System.CrackDelay
    Mesh Scale Mode=::NewEnumerator3
    Position Offset=<-100 0 0>
    UsePositionOffset=true
    Write Material Random=true
   )
   &InitialMeshOrientation(Orientation Vector=<0 0 1>, Use Orientation Vector=true, Use Rotation Vector=false)
  ]
  particleUpdateScript [
   &ParticleState()
   &DynamicMaterialParameters(
    Index 0 Param 1=&FloatFromCurve(
     FloatCurve=DataInterface() {propertyValues {Curve {keys [
         {interpMode=RCIM_Cubic, tangentMode=RCTM_User, time=0, value=0.25}
         {arriveTangent=1.0397, interpMode=RCIM_Cubic, leaveTangent=1.0397, tangentMode=RCTM_User, time=1, value=1}
        ]}}}
     Scale Curve=4
    )
    Index 0 Param 2=&Lerp_Float(
     A=200
     B=30
     Alpha=&FloatFromCurve(FloatCurve=DataInterface() {propertyValues {Curve {keys [
          {arriveTangent=0.0583, interpMode=RCIM_Cubic, leaveTangent=0.0583, tangentMode=RCTM_User, time=0, value=0}
          {interpMode=RCIM_Cubic, leaveTangent=1.7299, time=1, value=1}
         ]}}})
    )
    Index 0 Param 3=&FloatFromCurve(
     FloatCurve=DataInterface() {propertyValues {Curve {keys [
         {interpMode=RCIM_Cubic, tangentMode=RCTM_User, time=0, value=0}
         {arriveTangent=0.5, interpMode=RCIM_Cubic, leaveTangent=0.5, time=0.6, value=1}
         {interpMode=RCIM_Cubic, tangentMode=RCTM_User, time=1, value=0.75}
        ]}}}
     Scale Curve=2
    )
    Index 0 Param 4=500
    Module.Param0WriteEnabled=true
    Module.Param1WriteEnabled=true
    Module.Param2WriteEnabled=true
    Module.Param3WriteEnabled=true
   )
   &ScaleColor(
    Scale RGB=&VectorFromFloat(Value=&FloatFromCurve(FloatCurve=DataInterface() {propertyValues {Curve {keys [
          {arriveTangent=2.8782, interpMode=RCIM_Cubic, leaveTangent=2.8782, tangentMode=RCTM_User, time=0, value=0}
          {arriveTangent=0.5705, interpMode=RCIM_Cubic, leaveTangent=0.5705, tangentMode=RCTM_User, time=0.4, value=0.4}
          {arriveTangent=1.7332, interpMode=RCIM_Cubic, leaveTangent=1.7332, tangentMode=RCTM_User, time=1, value=1}
         ]}}}))
    ScaleA=true
    ScaleRGB=true
   )
   &ScaleMeshSize(Scale Factor=&VectorFromFloat(Value=&FloatFromCurve(FloatCurve=DataInterface() {propertyValues {Curve {keys [
          {arriveTangent=1.5937, interpMode=RCIM_Cubic, leaveTangent=1.5937, tangentMode=RCTM_User, time=0, value=0}
          {arriveTangent=0.5591, interpMode=RCIM_Cubic, leaveTangent=0.5591, tangentMode=RCTM_User, time=1, value=1}
         ]}}})))
  ]
  rendererClasses [&NiagaraMeshRendererProperties]
  renderers [{rendererClass=&NiagaraMeshRendererProperties, rendererIndex=0}]
 }
 RoundFlare {
  emitterUpdateScript [
   &EmitterState(
    Enable Distance Culling=true
    Life Cycle Mode=::NewEnumerator1
    Loop Behavior=::NewEnumerator1
    Loop Delay=$System.CrackDelay
    Loop Duration=&RandomRangeFloat(Minimum=1, Maximum=2.7)
    MaxDistance=10000
    Module.bUseMaxDistance=true
    Module.bUseMinDistance=false
    Scalability Mode=::NewEnumerator1
    UseLoopDelay=true
   )
   &SpawnBurst_Instantaneous(Spawn Count=1, Use Spawn Probability=false)
  ]
  particleSpawnScript [
   /Niagara/Modules/Spawn/Initialization/InitializeParticle(
    Color=<rgba 0.2453 0.9334 1.25 1>
    Lifetime=&RandomRangeFloat(Minimum=2.7, Maximum=3.2, Evaluation Type=::NewEnumerator1)
    Module.Write Color=true
    Module.Write Lifetime=true
    Module.Write Mass=true
    Module.Write Material Random=true
    Module.Write Position=false
    Module.Write Scale=false
    Module.Write SpriteRotation=false
    Module.Write SpriteSize=true
    Sprite Size=<120 120>
   ) {enabled=false}
   /Niagara/Modules/Spawn/Initialization/V2/InitializeParticle(
    Color=<rgba 0.1759 1.1379 1.25 1>
    Lifetime=4.3
    Mass Mode=::NewEnumerator1
    Sprite Size=<450 450>
    Sprite Size Mode=::NewEnumerator4
    UsePositionOffset=false
    Write Material Random=true
   ) {module=InitializeParticle001}
   &DynamicMaterialParameters(
    Index 0 Param 1=&RandomRangeFloat(Minimum=0.016, Maximum=0.5, Evaluation Type=::NewEnumerator1)
    Index 0 Param 2=1
    Index 0 Param 3=1
    Index 0 Param 4=1
    Module.Param0WriteEnabled=true
    Module.Param1WriteEnabled=true
    Module.Param2WriteEnabled=true
    Module.Param3WriteEnabled=true
   )
  ]
  particleUpdateScript [
   &ParticleState()
   &ScaleColor(
    Scale Alpha=&FloatFromCurve(FloatCurve=DataInterface() {propertyValues {Curve {keys [
         <at 0 0 RCIM_Cubic>
         {arriveTangent=1.3333, interpMode=RCIM_Cubic, leaveTangent=1.3333, time=0.25, value=1}
         {arriveTangent=-1.3333, interpMode=RCIM_Cubic, leaveTangent=-1.3333, time=0.75, value=1}
         <at 1 0>
        ]}}})
    ScaleA=true
    ScaleRGB=true
   )
   &DynamicMaterialParameters(
    Index 0 Param 1=&RandomRangeFloat(Minimum=0.016, Maximum=0.5)
    Index 0 Param 2=1
    Index 0 Param 3=1
    Index 0 Param 4=1
    Module.Param0WriteEnabled=true
    Module.Param1WriteEnabled=true
    Module.Param2WriteEnabled=true
    Module.Param3WriteEnabled=true
   ) {module=DynamicMaterialParameters001}
   &CameraOffset(Camera Offset Amount=-150)
   &ScaleSpriteSize(Scale Factor=&Vector2DFromFloat(Value=&FloatFromCurve(FloatCurve=DataInterface() {propertyValues {Curve {keys [<at 0 1>,<at 10000 3>]}}}, CurveIndex=$Engine.Owner.LODDistance)))
  ]
  rendererClasses [&NiagaraSpriteRendererProperties]
  renderers [{rendererClass=&NiagaraSpriteRendererProperties}]
 }
 SharpFlare {
  emitterUpdateScript [
   &EmitterState(
    Enable Distance Culling=true
    Life Cycle Mode=::NewEnumerator1
    Loop Behavior=::NewEnumerator1
    Loop Delay=$System.CrackDelay
    Loop Duration=&RandomRangeFloat(Minimum=2.4, Maximum=2.7, Evaluation Type=::NewEnumerator1)
    MaxDistance=10000
    Module.bUseMaxDistance=false
    Module.bUseMinDistance=false
    Scalability Mode=::NewEnumerator1
    UseLoopDelay=true
   )
   &SpawnBurst_Instantaneous(Spawn Count=1, Use Spawn Probability=false)
  ]
  particleSpawnScript [
   /Niagara/Modules/Spawn/Initialization/InitializeParticle(
    Lifetime=2
    Module.Write Color=true
    Module.Write Lifetime=true
    Module.Write Mass=true
    Module.Write Material Random=true
    Module.Write Position=false
    Module.Write Scale=false
    Module.Write SpriteRotation=false
    Module.Write SpriteSize=true
    Sprite Size=<500 150>
   ) {enabled=false}
   /Niagara/Modules/Spawn/Initialization/V2/InitializeParticle(
    Color=<rgba 1.15 1.15 1.15 1>
    Lifetime=4.3
    Mass Mode=::NewEnumerator1
    Sprite Size=<350 150>
    Sprite Size Mode=::NewEnumerator4
    UsePositionOffset=false
    Write Material Random=true
   ) {module=InitializeParticle001}
   &ShapeLocation(MaskSpawnGroup=false, Offset=$System.Offset, Offset Mode=::NewEnumerator0, Override Seed=false, Sphere Radius=0, UseOwnerScale=true)
   &Color(Color=&MakeLinearColorFromVectorAndFloat("Vector (RGB)"=<0.2453 0.9334 1.25>, "Float (Alpha)"=&ClampFloat(Float=1)))
  ]
  particleUpdateScript [
   &ParticleState()
   &ScaleSpriteSize(
    Uniform Curve Scale=6
    Uniform Curve Sprite Scale=DataInterface() {propertyValues {Curve {keys [
        {arriveTangent=21.6769, interpMode=RCIM_Cubic, leaveTangent=21.6769, tangentMode=RCTM_User, time=0, value=0.2}
        <at 0.1 1>
        {
         arriveTangent=-0.0783
         arriveTangentWeight=0.2899
         interpMode=RCIM_Cubic
         leaveTangent=-0.0783
         leaveTangentWeight=0.1115
         tangentMode=RCTM_User
         time=0.25
         value=1
        }
        {arriveTangent=-3.7848, interpMode=RCIM_Cubic, leaveTangent=-3.7848, tangentMode=RCTM_User, time=1, value=0.2}
       ]}}}
   )
   &ScaleColor(
    Scale Alpha=&FloatFromCurve(FloatCurve=DataInterface() {propertyValues {Curve {keys [
         {arriveTangent=21.6769, interpMode=RCIM_Cubic, leaveTangent=21.6769, tangentMode=RCTM_User, time=0, value=0}
         <at 0.1 1>
         {arriveTangent=-1.1111, interpMode=RCIM_Cubic, leaveTangent=-1.1111, time=0.75, value=1}
         <at 1 0>
        ]}}})
    ScaleA=true
    ScaleRGB=false
   )
   &CameraOffset(Camera Offset Amount=-175)
  ]
  rendererClasses [&NiagaraSpriteRendererProperties]
  renderers [{rendererClass=&NiagaraSpriteRendererProperties}]
 }
}
summary {
 systemName=NS_Rift_Teleport_Portal
 emitters [
  {emitterName=Crack, rendererClasses [&NiagaraMeshRendererProperties]}
  {bEnabled=false, emitterName=PreCrackGlow, rendererClasses [&NiagaraMeshRendererProperties]}
  {
   emitterName=GlowBurst
   rendererClasses [
    &NiagaraLightRendererProperties
    &NiagaraSpriteRendererProperties
    &NiagaraMeshRendererProperties
   ]
  }
  {emitterName=Fragments, rendererClasses [&NiagaraMeshRendererProperties, &NiagaraSpriteRendererProperties]}
  {emitterName=PreCrackLightShafts, rendererClasses [&NiagaraMeshRendererProperties]}
  {emitterName=LightShaftsBurst, rendererClasses [&NiagaraMeshRendererProperties]}
  {emitterName=Glow, rendererClasses [&NiagaraSpriteRendererProperties, &NiagaraLightRendererProperties]}
  {
   emitterName=GlowBurstEnd
   rendererClasses [
    &NiagaraLightRendererProperties
    &NiagaraSpriteRendererProperties
    &NiagaraMeshRendererProperties
   ]
  }
  {
   emitterName=BurstEndResiduleDust
   rendererClasses [&NiagaraMeshRendererProperties, &NiagaraSpriteRendererProperties]
  }
  {emitterName=SharpFlare, rendererClasses [&NiagaraSpriteRendererProperties]}
  {emitterName=RoundFlare, rendererClasses [&NiagaraSpriteRendererProperties]}
  {emitterName=Glints, rendererClasses [&NiagaraSpriteRendererProperties]}
  {
   emitterName=Light_CharactersFront
   rendererClasses [&NiagaraSpriteRendererProperties, &NiagaraLightRendererProperties]
  }
 ]
 userVariables [
  User.InitialStartDelay {type=NiagaraFloat, defaultValue=2.75, description=""}
  User.Open {type=NiagaraFloat, defaultValue=0, description=""}
 ]
}
